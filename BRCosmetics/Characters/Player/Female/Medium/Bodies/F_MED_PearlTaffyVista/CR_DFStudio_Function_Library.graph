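# Control Rig function library for a female character body, providing DistanceRemap and DistanceRemap_RelativeToSelf, with defaults pairing a thigh twist bone and skirt bone.
# Both measure joint offset in a parent space: one ignores a chosen axis, the other keeps only that axis against the end joint's initial transform.
# A cross-product component signs and scales the distance, remapped to a returned alpha.
# Optional debug draws the offset and prints values.
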ControlRigBlueprint /BRCosmetics/Characters/Player/Female/Medium/Bodies/F_MED_PearlTaffyVista/CR_DFStudio_Function_Library
graphs [
 New Function_ContainedGraph {nodes [
   RigVMFunctionEntryNode()
   > RigVMFunction_Sequence = RigVMAggregateNode()
   n1 = RigVMVariableNode() {variable=ParentSpace}
   n2 = &RigUnit_GetRelativeTransformForItem(bChildInitial=False, bParentInitial=False, Child=RigVMVariableNode() {variable=StartJoint}.Value, Parent=@n1.Value)
   n3 = RigVMDispatchNode(Struct=@n2.RelativeTransform.Translation) {template="DISPATCH_RigVMDispatch_BreakStruct(in Struct,out Elements)"}
   n4 = &RigUnit_GetRelativeTransformForItem(bChildInitial=False, bParentInitial=False, Child=RigVMVariableNode() {variable=EndJoint}.Value, Parent=@n1.Value)
   n5 = RigVMDispatchNode(Struct=@n4.RelativeTransform.Translation) {template="DISPATCH_RigVMDispatch_BreakStruct(in Struct,out Elements)"}
   n6 = RigVMDispatchNode(
    Values=((X=0,Y=0,Z=0),(X=0,Y=0,Z=0),((X=0,Y=0,Z=0),(X=0,Y=0,Z=0),X=0,Y=0,Z=0),((X=0,Y=0,Z=0),(X=0,Y=0,Z=0),X=0,Y=0,Z=0))
    Index=RigVMDispatchNode(Value=RigVMVariableNode() {variable=AxisToAvoid}.Value) {template="DISPATCH_RigVMDispatch_CastEnumToInt(in Value,out Result)"}.Result
    Values.1=RigVMDispatchNode(Elements=(X=0,Y=0,Z=0), Elements.Y=@n3.Elements.Y, Elements.Z=@n3.Elements.Z, Elements.X=@n5.Elements.X) {template="DISPATCH_RigVMDispatch_MakeStruct(in Elements,out Struct)"}.Struct
    Values.2=RigVMDispatchNode(Elements=(X=0,Y=0,Z=0), Elements.X=@n3.Elements.X, Elements.Z=@n3.Elements.Z, Elements.Y=@n5.Elements.Y) {template="DISPATCH_RigVMDispatch_MakeStruct(in Elements,out Struct)"}.Struct
    Values.3=RigVMDispatchNode(Elements=(X=0,Y=0,Z=0), Elements.X=@n3.Elements.X, Elements.Y=@n3.Elements.Y, Elements.Z=@n5.Elements.Z) {template="DISPATCH_RigVMDispatch_MakeStruct(in Elements,out Struct)"}.Struct
    Values.0=RigVMDispatchNode(Elements=(X=0,Y=0,Z=0), Elements.X=@n3.Elements.X, Elements.Y=@n3.Elements.Y, Elements.Z=@n3.Elements.Z) {template="DISPATCH_RigVMDispatch_MakeStruct(in Elements,out Struct)"}.Struct
   ) {template="DISPATCH_RigVMDispatch_SelectInt32(in Index,in Values,out Result)"}
   Subtract = /Script/RigVM.RigVMFunction_MathVectorSub(B=@n6.Result, A=@n4.RelativeTransform.Translation) {is=struct RigVMUnitNode, template="Subtract::Execute(in A,in B,out Result)"}
   Unit = /Script/RigVM.RigVMFunction_MathVectorUnit(Value=@Subtract.Result) {is=struct RigVMUnitNode, template="Unit::Execute(in Value,out Result)"}
   n7 = &RigVMFunction_MathVectorUnit(Value=@n2.RelativeTransform.Translation) {template="Unit::Execute(in Value,out Result)"}
   RigVMFunction_MathVectorCross = /Script/RigVM.RigVMFunction_MathVectorCross(B=@Unit.Result, A=@n7.Result) {is=struct RigVMUnitNode, template="Cross::Execute(in A,in B,out Result)"}
   n8 = RigVMDispatchNode(
    Values=(0,0,0,0)
    Index=RigVMDispatchNode(Value=RigVMVariableNode() {variable=AxisToTrack}.Value) {template="DISPATCH_RigVMDispatch_CastEnumToInt(in Value,out Result)"}.Result
    Values.1=@RigVMFunction_MathVectorCross.Result.X
    Values.2=@RigVMFunction_MathVectorCross.Result.Y
    Values.3=@RigVMFunction_MathVectorCross.Result.Z
   ) {template="DISPATCH_RigVMDispatch_SelectInt32(in Index,in Values,out Result)"}
   Less = /Script/RigVM.RigVMFunction_MathDoubleLess(B=0, A=@n8.Result) {is=struct RigVMUnitNode, template="Less::Execute(in A,in B,out Result)"}
   Cast = &RigVMFunction_MathBoolToFloat(Value=@Less.Result)
   Absolute = &RigVMFunction_MathDoubleAbs(Value=@n8.Result)
   Multiply = &RigVMFunction_MathDoubleMul(
    A=&RigVMFunction_MathDoubleRemap(SourceMaximum=0, SourceMinimum=1, TargetMaximum=1, TargetMinimum=-1, Value=@Cast.Result).Result
    B=@Absolute.Result
   )
   RigVMFunction_MathVectorLength = /Script/RigVM.RigVMFunction_MathVectorLength(Value=@Subtract.Result) {is=struct RigVMUnitNode, template="Length::Execute(in Value,out Result)"}
   n9 = &RigVMFunction_MathDoubleMul(A=@Multiply.Result, B=@RigVMFunction_MathVectorLength.Result)
   RigVMVariableNode(
    Value=&RigVMFunction_MathDoubleRemap(
     Value=@n9.Result
     TargetMinimum=RigVMVariableNode() {variable=TargetMin}.Value
     TargetMaximum=RigVMVariableNode() {variable=TargetMax}.Value
     SourceMinimum=RigVMVariableNode() {variable=DistanceMin}.Value
     SourceMaximum=RigVMVariableNode() {variable=DistanceMax}.Value
    ).Result
    ExecuteContext=@RigVMFunction_Sequence.A
   ) {variable=FinalAlpha}
   RigVMFunction_ControlFlowBranch = &RigVMFunction_ControlFlowBranch(ExecuteContext=@RigVMFunction_Sequence.B, Condition=RigVMVariableNode() {variable=Enable Debug}.Value)
   /Script/RigVM.RigVMFunction_DebugLineNoSpace(
    Color=(R=0,G=1,B=0,A=1)
    DebugDrawSettings=(DepthPriority=SDPG_Foreground,Lifetime=-1)
    Thickness=2
    bEnabled=true
    WorldOffset=&RigUnit_GetTransform(bInitial=False, Item=RigVMVariableNode() {variable=ParentSpace}.Value).Transform
    ExecutePin=@RigVMFunction_ControlFlowBranch.True
    B=@n4.RelativeTransform.Translation
    A=@n6.Result
   ) {is=struct RigVMUnitNode}
   To_String = RigVMDispatchNode(Value=RigVMVariableNode() {variable=AxisToTrack}.Value) {template="DISPATCH_RigDispatch_ToString(in Value,out Result)"}
   Concat = /Script/RigVM.RigVMFunction_StringConcat(B=": ", A=@To_String.Result) {is=struct RigVMUnitNode, template="Concat::Execute(in A,in B,out Result)"}
   > RigVMDispatchNode(Enabled=True, ScreenColor=(R=1,G=1,B=1,A=1), ScreenDuration=0.0, Prefix=@Concat.Result, Value=@n9.Result) {template="DISPATCH_RigVMDispatch_Print(in Prefix,in Value,in Enabled,in ScreenDuration,in ScreenColor)"}
   RigVMFunction_MathVectorDot = /Script/RigVM.RigVMFunction_MathVectorDot(B=@Unit.Result, A=@n7.Result) {is=struct RigVMUnitNode}
   > RigVMDispatchNode(Enabled=True, Prefix=Dot:, ScreenColor=(R=1,G=1,B=1,A=1), ScreenDuration=0.0, Value=@RigVMFunction_MathVectorDot.Result) {template="DISPATCH_RigVMDispatch_Print(in Prefix,in Value,in Enabled,in ScreenDuration,in ScreenColor)"}
   RigVMFunctionReturnNode(ExecuteContext=@RigVMFunction_Sequence.D, Value=RigVMVariableNode() {variable=FinalAlpha}.Value)
  ]}
 New Function_ContainedGraph {nodes [
   RigVMFunctionEntryNode()
   > RigVMFunction_Sequence = RigVMAggregateNode()
   n1 = RigVMVariableNode() {variable=ParentSpace}
   n2 = &RigUnit_GetRelativeTransformForItem(bChildInitial=False, bParentInitial=False, Child=RigVMVariableNode() {variable=StartJoint}.Value, Parent=@n1.Value)
   n3 = RigVMDispatchNode(Struct=@n2.RelativeTransform.Translation) {template="DISPATCH_RigVMDispatch_BreakStruct(in Struct,out Elements)"}
   n4 = &RigUnit_GetRelativeTransformForItem(bChildInitial=true, bParentInitial=False, Child=RigVMVariableNode() {variable=EndJoint}.Value, Parent=@n1.Value)
   n5 = RigVMDispatchNode(Struct=@n4.RelativeTransform.Translation) {template="DISPATCH_RigVMDispatch_BreakStruct(in Struct,out Elements)"}
   n6 = RigVMDispatchNode(
    Values=((X=0,Y=0,Z=0),(X=0,Y=0,Z=0),((X=0,Y=0,Z=0),(X=0,Y=0,Z=0),X=0,Y=0,Z=0),((X=0,Y=0,Z=0),(X=0,Y=0,Z=0),X=0,Y=0,Z=0))
    Index=RigVMDispatchNode(Value=RigVMVariableNode() {variable=AxisToNotEqualized}.Value) {template="DISPATCH_RigVMDispatch_CastEnumToInt(in Value,out Result)"}.Result
    Values.1=RigVMDispatchNode(Elements=(X=0,Y=0,Z=0), Elements.X=@n3.Elements.X, Elements.Y=@n5.Elements.Y, Elements.Z=@n5.Elements.Z) {template="DISPATCH_RigVMDispatch_MakeStruct(in Elements,out Struct)"}.Struct
    Values.2=RigVMDispatchNode(Elements=(X=0,Y=0,Z=0), Elements.Y=@n3.Elements.Y, Elements.X=@n5.Elements.X, Elements.Z=@n5.Elements.Z) {template="DISPATCH_RigVMDispatch_MakeStruct(in Elements,out Struct)"}.Struct
    Values.3=RigVMDispatchNode(Elements=(X=0,Y=0,Z=0), Elements.Z=@n3.Elements.Z, Elements.X=@n5.Elements.X, Elements.Y=@n5.Elements.Y) {template="DISPATCH_RigVMDispatch_MakeStruct(in Elements,out Struct)"}.Struct
    Values.0=RigVMDispatchNode(Elements=(X=0,Y=0,Z=0), Elements.X=@n3.Elements.X, Elements.Y=@n3.Elements.Y, Elements.Z=@n3.Elements.Z) {template="DISPATCH_RigVMDispatch_MakeStruct(in Elements,out Struct)"}.Struct
   ) {template="DISPATCH_RigVMDispatch_SelectInt32(in Index,in Values,out Result)"}
   Subtract = &RigVMFunction_MathVectorSub(B=@n6.Result, A=@n4.RelativeTransform.Translation) {template="Subtract::Execute(in A,in B,out Result)"}
   Unit = &RigVMFunction_MathVectorUnit(Value=@Subtract.Result) {template="Unit::Execute(in Value,out Result)"}
   n7 = &RigVMFunction_MathVectorUnit(Value=@n2.RelativeTransform.Translation) {template="Unit::Execute(in Value,out Result)"}
   RigVMFunction_MathVectorCross = &RigVMFunction_MathVectorCross(B=@Unit.Result, A=@n7.Result) {template="Cross::Execute(in A,in B,out Result)"}
   n8 = RigVMDispatchNode(
    Values=(0,0,0,0)
    Index=RigVMDispatchNode(Value=RigVMVariableNode() {variable=AxisToTrack}.Value) {template="DISPATCH_RigVMDispatch_CastEnumToInt(in Value,out Result)"}.Result
    Values.1=@RigVMFunction_MathVectorCross.Result.X
    Values.2=@RigVMFunction_MathVectorCross.Result.Y
    Values.3=@RigVMFunction_MathVectorCross.Result.Z
   ) {template="DISPATCH_RigVMDispatch_SelectInt32(in Index,in Values,out Result)"}
   Less = &RigVMFunction_MathDoubleLess(B=0, A=@n8.Result) {template="Less::Execute(in A,in B,out Result)"}
   Cast = &RigVMFunction_MathBoolToFloat(Value=@Less.Result)
   Absolute = &RigVMFunction_MathDoubleAbs(Value=@n8.Result)
   Multiply = &RigVMFunction_MathDoubleMul(
    A=&RigVMFunction_MathDoubleRemap(SourceMaximum=0, SourceMinimum=1, TargetMaximum=1, TargetMinimum=-1, Value=@Cast.Result).Result
    B=@Absolute.Result
   )
   RigVMFunction_MathVectorLength = &RigVMFunction_MathVectorLength(Value=@Subtract.Result) {template="Length::Execute(in Value,out Result)"}
   n9 = &RigVMFunction_MathDoubleMul(A=@Multiply.Result, B=@RigVMFunction_MathVectorLength.Result)
   RigVMVariableNode(
    Value=&RigVMFunction_MathDoubleRemap(
     Value=@n9.Result
     TargetMinimum=RigVMVariableNode() {variable=TargetMin}.Value
     TargetMaximum=RigVMVariableNode() {variable=TargetMax}.Value
     SourceMinimum=RigVMVariableNode() {variable=DistanceMin}.Value
     SourceMaximum=RigVMVariableNode() {variable=DistanceMax}.Value
    ).Result
    ExecuteContext=@RigVMFunction_Sequence.A
   ) {variable=FinalAlpha}
   RigVMFunction_ControlFlowBranch = &RigVMFunction_ControlFlowBranch(ExecuteContext=@RigVMFunction_Sequence.C, Condition=RigVMVariableNode() {variable=Enable Debug}.Value)
   &RigVMFunction_DebugLineNoSpace(
    Color=(R=0,G=1,B=0,A=1)
    DebugDrawSettings=(DepthPriority=SDPG_Foreground,Lifetime=-1)
    Thickness=3
    bEnabled=true
    WorldOffset=&RigUnit_GetTransform(bInitial=False, Item=RigVMVariableNode() {variable=ParentSpace}.Value).Transform
    ExecutePin=@RigVMFunction_ControlFlowBranch.True
    B=@n4.RelativeTransform.Translation
    A=@n6.Result
   )
   To_String = RigVMDispatchNode(Value=RigVMVariableNode() {variable=AxisToTrack}.Value) {template="DISPATCH_RigDispatch_ToString(in Value,out Result)"}
   Concat = &RigVMFunction_StringConcat(B=": ", A=@To_String.Result) {template="Concat::Execute(in A,in B,out Result)"}
   > RigVMDispatchNode(Enabled=True, ScreenColor=(R=1,G=1,B=1,A=1), ScreenDuration=0.0, Prefix=@Concat.Result, Value=@n9.Result) {template="DISPATCH_RigVMDispatch_Print(in Prefix,in Value,in Enabled,in ScreenDuration,in ScreenColor)"}
   RigVMFunction_MathVectorDot = &RigVMFunction_MathVectorDot(B=@Unit.Result, A=@n7.Result)
   > RigVMDispatchNode(Enabled=True, Prefix=Dot:, ScreenColor=(R=1,G=1,B=1,A=1), ScreenDuration=0.0, Value=@RigVMFunction_MathVectorDot.Result) {template="DISPATCH_RigVMDispatch_Print(in Prefix,in Value,in Enabled,in ScreenDuration,in ScreenColor)"}
   RigVMFunctionReturnNode(ExecuteContext=@RigVMFunction_Sequence.D, Value=RigVMVariableNode() {variable=FinalAlpha}.Value)
  ]}
 RigVMFunctionLibrary {nodes [
   DistanceRemap = RigVMCollapseNode(
    AxisToAvoid=Z
    AxisToTrack=X
    EndJoint=(Type=Bone,Name="dyn_skirt_fr_end_l")
    ExecuteContext=()
    ParentSpace=(Type=None,Name="None")
    StartJoint=(Type=Bone,Name="thigh_twist_01_l")
    TargetMax=1
    TargetMin=0
   )
   DistanceRemap_RelativeToSelf = RigVMCollapseNode(
    AxisToNotEqualized=Z
    AxisToTrack=X
    EndJoint=(Type=Bone,Name="dyn_skirt_fr_end_l")
    ExecuteContext=()
    ParentSpace=(Type=None,Name="None")
    StartJoint=(Type=Bone,Name="thigh_twist_01_l")
    TargetMax=1
    TargetMin=0
   )
  ]}
 RigVMFunction_Sequence_1_ContainedGraph {nodes [
   n1 = RigVMFunctionEntryNode()
   RigVMFunction_Sequence = &RigVMFunction_Sequence(ExecuteContext=@n1.ExecuteContext)
   n2 = &RigVMFunction_Sequence(ExecuteContext=@RigVMFunction_Sequence.B)
   n3 = &RigVMFunction_Sequence(ExecuteContext=@n2.B)
   RigVMFunctionReturnNode(A=@RigVMFunction_Sequence.A, B=@n2.A, C=@n3.A, D=@n3.B)
  ]}
 RigVMFunction_Sequence_1_ContainedGraph {nodes [
   n1 = RigVMFunctionEntryNode()
   RigVMFunction_Sequence = &RigVMFunction_Sequence(ExecuteContext=@n1.ExecuteContext)
   n2 = &RigVMFunction_Sequence(ExecuteContext=@RigVMFunction_Sequence.B)
   n3 = &RigVMFunction_Sequence(ExecuteContext=@n2.B)
   RigVMFunctionReturnNode(A=@RigVMFunction_Sequence.A, B=@n2.A, C=@n3.A, D=@n3.B)
  ]}
 RigVMModel {nodes [RigUnit_BeginExecution = &RigUnit_BeginExecution()]}
]
