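# Generic audio-reactive effect material function for cosmetic character materials, modifying incoming material attributes.
# Amplitude from shared audio functions (or debug sine waves) combines with Fresnel, TechArtMask-style channel masks and AO-weighted gradients into a grayscale effect.
# The effect adds colored emissive from a parameter or map, while gradient overrides fade base color, opacity and subsurface toward zero.
# Amplitude, grayscale, colors and mask are also output; static switches select sources.
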MaterialFunction /BRCosmetics/Global/Materials/Character_Parent_2023/AudioReactivity_Parent_2026/MF_AudioReactivity_Generic
graph {nodes [
  n1 = GetMaterialAttributes(None=NamedRerouteUsage()) {attributeGetTypes [
    #1
    #2
    #3
    #4
   ]}
  n2 = SetMaterialAttributes!(
   MaterialAttributes=@n1.MaterialAttributes
   BaseColor=NamedRerouteUsage()
   EmissiveColor=NamedRerouteUsage()
   Opacity=NamedRerouteUsage()
   SubsurfaceColor=NamedRerouteUsage()
  ) {attributeSetTypes [
    #1
    #2
    #3
    #4
   ]}
  FunctionOutput(None=@n2)
  FunctionOutput(None=NamedRerouteUsage()) {outputName=Effect Grayscale, sortPriority=2}
  FunctionOutput(None=NamedRerouteUsage()) {outputName=Amplitude, sortPriority=1}
  FunctionOutput(None=NamedRerouteUsage()) {outputName="Color(s)", sortPriority=3}
  FunctionOutput(None=NamedRerouteUsage()) {outputName=TechArtMask, sortPriority=5}
  NamedRerouteDeclaration(None=FunctionInput() {inputType=FunctionInput_MaterialAttributes}) {name=Material In, nodeColor=<rgba 1 0.7031 0 1>}
  NamedRerouteDeclaration(None=@n1.BaseColor) {name=BaseColor In, nodeColor=<rgba 0.3828 0 1 1>}
  n3 = Multiply(
   A=Power(
    Base=NamedRerouteUsage()
    Exp=ScalarParameter() {defaultValue=2, group=Audio Reactivity|Generic|Overrides, parameterName=Gradient Base Color Override - Exp}
   )
   B=ScalarParameter() {defaultValue=2, group=Audio Reactivity|Generic|Overrides, parameterName=Gradient Base Color Override - Int}
  ) {constB=3}
  NamedRerouteDeclaration(None=LinearInterpolate(A=NamedRerouteUsage(), Alpha=Saturate(None=@n3)) {constB=0}) {name=BaseColor Out, nodeColor=<rgba 0 0.8359 1 1>}
  n4 = FunctionInput() {bUsePreviewValueAsDefault=true, inputName=AO, inputType=FunctionInput_Scalar, sortPriority=2, previewValue=<1 0 0 1>}
  NamedRerouteDeclaration(None=@n4) {name=Ambient Occlusion, nodeColor=<rgba 0.0078 1 0 1>}
  n5 = ScalarParameter() {group=Audio Reactivity|Generic, parameterName=ReactivityColorMin}
  n6 = ScalarParameter() {defaultValue=1.5, group=Audio Reactivity|Generic, parameterName=ReactiveFresnelExpMin}
  n7 = ScalarParameter() {defaultValue=1, group=Audio Reactivity|Generic, parameterName=ReactiveFresnelExpMax}
  n8 = NamedRerouteUsage()
  n9 = Fresnel(ExponentIn=Multiply(
    A=LinearInterpolate(A=@n6, B=@n7, Alpha=@n8)
    B=ScalarParameter() {defaultValue=1, group=Audio Reactivity|Generic, parameterName=ReactiveFresnelExpMultiplier}
   )) {baseReflectFraction=0, exponent=1}
  NamedRerouteDeclaration(None=Multiply(A=LinearInterpolate(A=@n5, Alpha=@n9), B=ScalarParameter() {defaultValue=1, group=Audio Reactivity|Generic, parameterName=FresnelIntensity})) {name=Fresnel Gradient, nodeColor=<rgba 1 0.3281 0 1>}
  n10 = Power(
   Base=NamedRerouteUsage()
   Exp=LinearInterpolate(
    A=ScalarParameter() {defaultValue=1, group=Audio Reactivity|Generic, parameterName=ReactivGradientExpMin}
    B=ScalarParameter() {defaultValue=1, group=Audio Reactivity|Generic, parameterName=ReactiveGradientExpMax}
    Alpha=NamedRerouteUsage()
   )
  )
  n11 = ScalarParameter() {defaultValue=1, group=Audio Reactivity|Generic, parameterName=ReactivityGradientIntensity}
  n12 = ScalarParameter() {group=Audio Reactivity|Generic, parameterName=ReactiveGradientAOIntensity}
  n13 = NamedRerouteUsage()
  NamedRerouteDeclaration(None=Multiply(A=Multiply(A=@n10, B=@n11), B=LinearInterpolate(A=@n12, Alpha=@n13))) {name=Reactivity Map Gradient, nodeColor=<rgba 0.0078 0 1 1>}
  n14 = NamedRerouteUsage()
  n15 = ScalarParameter() {defaultValue=1, group=Audio Reactivity|Generic|Overrides, parameterName=Gradient Subsurface Override - Exp}
  n16 = Saturate(None=Multiply(
    A=Power(Base=@n14, Exp=@n15)
    B=ScalarParameter() {defaultValue=1, group=Audio Reactivity|Generic|Overrides, parameterName=Gradient Subsurface Override - Int}
   ) {constB=3})
  NamedRerouteDeclaration(None=LinearInterpolate(A=NamedRerouteUsage(), Alpha=@n16) {constB=0}) {name=OpacityOut, nodeColor=<rgba 1 0.8438 0 1>}
  NamedRerouteDeclaration(None=@n1.Opacity) {name=OpacityIn, nodeColor=<rgba 0 1 0.2031 1>}
  NamedRerouteDeclaration(None=@n1.SubsurfaceColor) {name=SSSColor In, nodeColor=<rgba 0.0547 1 0 1>}
  NamedRerouteDeclaration(None=LinearInterpolate(A=NamedRerouteUsage(), Alpha=@n16) {constB=0}) {name=SSSColor Out, nodeColor=<rgba 0 1 0.3438 1>}
  NamedRerouteDeclaration(None=@n1.EmissiveColor) {name=Emissive In, nodeColor=<rgba 0 1 0.0625 1>}
  n17 = ChannelMaskParameter(None=NamedRerouteUsage()) {group=Audio Reactivity|Generic, maskChannel=Green, parameterName=ReactivityGradientChannel, defaultValue=<rgba 0 1 0 0>}
  NamedRerouteDeclaration(None=Power(Base=@n17.R, Exp=ScalarParameter() {defaultValue=1, group=Audio Reactivity|Generic, parameterName=ReactivityGradientExponent})) {name=ReactivityGradientMap, nodeColor=<rgba 1 0 0.2578 1>}
  NamedRerouteDeclaration(None=StaticSwitchParameter(
    True=TextureSampleParameter2D() {group=Audio Reactivity|Generic, parameterName=CustomTechArtMask, texture=/Engine/EngineResources/WhiteSquareTexture}.RGBA
    False=FunctionInput() {inputName=TechArtMask, inputType=FunctionInput_Vector4, sortPriority=1}
   ) {group=Audio Reactivity|Generic, parameterName=UseCustomTechArtMask}) {name=TechArtMask, nodeColor=<rgba 1 0.0459 0 1>}
  n18 = NamedRerouteUsage()
  n19 = NamedRerouteUsage()
  NamedRerouteDeclaration(None=Add(A=NamedRerouteUsage(), B=Multiply(A=@n18, B=@n19))) {name=Emissive Out, nodeColor=<rgba 0.8984 1 0 1>}
  n20 = MaterialFunctionCall(Speed=Constant() {r=1.5}) {materialFunction=&DebugTimeSine}
  n21 = MaterialFunctionCall() {materialFunction=&DebugTimeSine}
  n22 = Constant() {r=1.25}
  n23 = Saturate(None=Subtract(
    A=Add(A=@n20.Result, B=@n21.Result)
    B=MaterialFunctionCall(Speed=@n22) {materialFunction=&DebugTimeSine}.Result
   ))
  n24 = MaterialFunctionCall() {materialFunction=/Game/Packages/Fortress_SharedMaterials/Base_Material_Functions/BaseCharacterShaderFunctions/Audio/MF_AudioReactivity}
  n25 = Power(
   Base=MaterialFunctionCall!(
    InAmplitude=FunctionInput(Preview=@n24.Amplitude) {bUsePreviewValueAsDefault=true, inputName=InAmplitude, inputType=FunctionInput_Scalar, sortPriority=5}
    InSpectrum
    CustomFakeSpectrum
    CustomFakeAmplitude
   ) {materialFunction=/BRCosmetics/Global/MaterialFunctions/AudioReactivity/MFI_AudioReactivity_FrontendLoop}.Amplitude
   Exp=ScalarParameter() {defaultValue=4, group=Audio Reactivity|Generic, parameterName=AmplitudeExponent}
  )
  n26 = ScalarParameter() {
   controlType=Enumeration
   defaultValue=1
   enumeration=/Game/Global/Materials/Enums/E_Mat_OnOff
   group=Audio Reactivity|Generic
   parameterName=AudioReactivity On
  }
  NamedRerouteDeclaration(None=Multiply(A=StaticSwitchParameter(True=@n23, False=@n25) {group=Audio Reactivity|Generic, parameterName=DebugAudioReactivity}, B=@n26)) {name=Amplitude, nodeColor=<rgba 1 0.9375 0 1>}
  n27 = ChannelMaskParameter(None=NamedRerouteUsage()) {group=Audio Reactivity|Generic, maskChannel=Green, parameterName=ReactivityMaskChannel, defaultValue=<rgba 0 1 0 0>}
  NamedRerouteDeclaration(None=Max(A=@n27.R, B=NamedRerouteUsage())) {name=ReactivityMask, nodeColor=<rgba 0 1 0.1328 1>}
  n28 = NamedRerouteUsage()
  n29 = TextureSampleParameter2D(UVs=StaticSwitchParameter(True=NamedRerouteUsage(), False=TextureCoordinate()) {group=Audio Reactivity|Generic, parameterName=UseAmplitudeForColorUVs}) {group=Audio Reactivity|Generic, parameterName=ReactivityColorMap, texture=&WhiteSquareTexture}
  NamedRerouteDeclaration(None=StaticSwitchParameter(
    True=StaticSwitchParameter(True=@n28, False=@n29.RGB) {group=Audio Reactivity|Generic, parameterName=UseEmissiveMapForcolors}
    False=VectorParameter() {group=Audio Reactivity|Generic, parameterName=ReactivityColor, defaultValue=<rgba 1 1 1 1>}.RGB
   ) {group=Audio Reactivity|Generic, parameterName=UseColorTextureMap}) {name=ReactivityColors, nodeColor=<rgba 0.6875 0 1 1>}
  n30 = Saturate(None=Max(A=NamedRerouteUsage(), B=NamedRerouteUsage()))
  n31 = NamedRerouteUsage()
  NamedRerouteDeclaration(None=Multiply(A=Multiply(A=@n30, B=@n31), B=NamedRerouteUsage())) {name=ReactivityGrayscale, nodeColor=<rgba 1 0 0.5156 1>}
 ]}
