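# Parent character material for the PearlTaffyVista female body, adding animated, audio-reactive glitter and stripe effects between the shared 2023 PreFX and PostFX functions.
# A generic audio reactivity function supplies amplitude, color and a channel-selected gradient mask.
# Dense glitter combines two sparkle samples, one parallax-offset, with camera-distance tiling and Fresnel falloff.
# It merges with a large glitter layer and procedural sine stripes, then adds to Base Color and Emissive.
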Material /BRCosmetics/Characters/Player/Female/Medium/Bodies/F_MED_PearlTaffyVista/Materials/M_PearlTaffyVista_Parent
graph {
 properties {
  bAutomaticallySetUsageInEditor=false
  bUseMaterialAttributes=true
  bUsedWithMeshDeformer=true
  bUsedWithMorphTargets=true
  bUsedWithSkeletalMesh=true
  pixelDepthOffsetMode=PDOM_Legacy
  previewMesh=/BRCosmetics/Characters/Player/Female/Medium/Bodies/F_MED_PearlTaffyVista/Meshes/F_MED_PearlTaffyVista
  shadingModel=MSM_FromMaterialExpression
  usedShadingModels=Default Lit | Subsurface | Clear Coat | Subsurface Profile | Cloth
 }
 nodes [
  n1 = MaterialFunctionCall() {materialFunction=&MF_CharacterParent_2023_PreFX}
  MaterialFunctionCall(Material=NamedRerouteUsage(), M=@n1.M, Pre Skinned Local Position=@n1.M, Distance Blend=@n1.M, Fresnel=@n1.M) {materialFunction=&MF_CharacterParent_2023_PostFX}
  n2 = NamedRerouteUsage()
  n3 = NamedRerouteUsage()
  n4 = Multiply(
   A=Power(Base=Multiply(A=@n2, B=@n3), Exp=ScalarParameter() {defaultValue=0.9, parameterName=Dense Sparkle Exp}) {constExponent=4}
   B=ScalarParameter() {defaultValue=3, parameterName=Dense Sparkle Intensity}
  ) {constB=4}
  n5 = OneMinus(None=Fresnel() {baseReflectFraction=0, exponent=1})
  n6 = ScalarParameter() {defaultValue=0.5, group=PearlTaffyVista|Glitter, parameterName=Dense GlitterFresnelExp}
  NamedRerouteDeclaration(None=Multiply(A=Saturate(None=@n4), B=Power(Base=@n5, Exp=@n6) {constExponent=3})) {name=DenseGlitter, nodeColor=<rgba 0.7344 0 1 1>}
  n7 = MaterialFunctionCall!(
   In=NamedRerouteDeclaration(None=@n1.Result) {name=Material In, nodeColor=<rgba 0.0312 0 1 1>}
   TechArtMask=AppendVector(A=@n1.Tech Art Mask, B=@n1.Tech Art Mask)
   AO=ComponentMask(None=@n1.M) {r=true}
   InAmplitude
  ) {materialFunction=/BRCosmetics/Global/Materials/Character_Parent_2023/AudioReactivity_Parent_2026/MF_AudioReactivity_Generic}
  NamedRerouteDeclaration(None=@n7.TechArtMask) {name=FXMask In, nodeColor=<rgba 1 0 0.7969 1>}
  n8 = NamedRerouteUsage()
  n9 = Multiply(
   A=Add(B=Time())
   B=ScalarParameter() {defaultValue=2, group=PearlTaffyVista|Glitter, parameterName=Dense Sparkle Time Scale, sliderMax=2, sortPriority=3}
  ) {constB=0.5}
  n10 = MaterialFunctionCall(UV Channel=Frac(None=Add(A=@n8, B=@n9))) {materialFunction=&LinearGradient}
  NamedRerouteDeclaration(None=ConstantBiasScale(None=Sine(None=@n10.UGradient))) {name=GlitterAnimatedMask, nodeColor=<rgba 0.1016 0 1 1>}
  n11 = VertexInterpolator(VS=LocalPosition() {localOrigin=InstancePreSkinning}.XYZ)
  n12 = Multiply(A=ComponentMask(None=@n11.PS) {r=true}) {constB=0.01}
  n13 = Sine(None=Add(A=Multiply(A=@n12) {constB=5}, B=Time()))
  n14 = ConstantBiasScale(None=@n13)
  n15 = Multiply(A=@n14, B=@n14)
  n16 = Abs(None=@n12)
  n17 = Divide(A=@n11.PS, B=ScalarParameter() {defaultValue=10, group=PearlTaffyVista, parameterName=AnimatedStripes - Tiling, sliderMax=200}) {constB=30}
  n18 = Add(A=Multiply(A=ComponentMask(None=@n17) {b=true, r=true}) {constB=-1}) {constB=0.5}
  n19 = ScalarParameter() {defaultValue=0.85, group=PearlTaffyVista, parameterName=AnimatedStripes - Rotation, sliderMax=1, sliderMin=-1}
  n20 = Add(A=Multiply(A=Power(Base=@n16)) {constB=-20}, B=ComponentMask(None=Rotator(Coordinate=@n18, Time=@n19) {speed=3.1415}) {g=true})
  n21 = Time()
  n22 = Multiply(A=@n13, B=NamedRerouteUsage())
  n23 = Sine(None=Add(A=Add(A=@n20, B=@n21), B=Multiply(A=@n22) {constB=0.15}))
  NamedRerouteDeclaration(None=Multiply(A=LinearInterpolate(Alpha=@n15) {constA=0.9}, B=ConstantBiasScale(None=@n23))) {name=Animated Stripes, nodeColor=<rgba 0.3594 0 1 1>}
  n24 = GetMaterialAttributes(None=NamedRerouteUsage()) {attributeGetTypes [#1, #2]}
  n25 = MaterialFunctionCall!(In=NamedRerouteUsage(), SpeedMultiplier, GlitterPower, GlitterIntensity, GlitterColor, SkinFX_Mask) {
   desc=Large Glitter
   materialFunction=&/MFI_PearlTaffyVista_Glitter
  }
  n26 = NamedRerouteUsage()
  n27 = Multiply(A=GetMaterialAttributes(None=@n25.Result) {attributeGetTypes [#2]}.Emissive Color, B=@n26)
  n28 = Power(Base=@n26) {constExponent=3}
  n29 = Multiply(A=NamedRerouteUsage(), B=@n28)
  n30 = Add(
   A=Multiply(A=Max(A=@n27, B=@n29), B=ScalarParameter() {defaultValue=7.5, group=PearlTaffyVista, parameterName=DenseGlitterIntensity}) {constB=10}
   B=Multiply(A=@n28, B=ScalarParameter() {defaultValue=0.2, group=PearlTaffyVista, parameterName=AnimatedStripesIntensity}) {constB=0.2}
  )
  n31 = Power(Base=NamedRerouteUsage(), Exp=ScalarParameter() {defaultValue=1, group=PearlTaffyVista, parameterName=EffectGradientMaskPow}) {constExponent=0.5}
  n32 = Multiply(A=Multiply(A=Multiply(A=@n30, B=@n31), B=NamedRerouteUsage()), B=NamedRerouteUsage())
  NamedRerouteDeclaration(None=SetMaterialAttributes!(MaterialAttributes=@n24.MaterialAttributes, Base Color=Add(A=@n24.Base Color, B=@n32), Emissive Color=Add(A=@n24.Emissive Color, B=@n32)) {attributeSetTypes [#1, #2]}) {name=Material Out, nodeColor=<rgba 0.2891 1 0 1>}
  NamedRerouteDeclaration(None=@n7.Result) {name=Audio Reactivity In, nodeColor=<rgba 1 0 0.6094 1>}
  n33 = ChannelMaskParameter(None=NamedRerouteUsage()) {group=PearlTaffyVista, maskChannel=Green, parameterName=Reactivity Gradient, defaultValue=<rgba 0 1 0 0>}
  NamedRerouteDeclaration(None=@n33.R) {name=FXGradient, nodeColor=<rgba 0 1 0.8828 1>}
  NamedRerouteDeclaration(None=@n7.Amplitude) {name=Amplitude, nodeColor=<rgba 0.4531 1 0 1>}
  NamedRerouteDeclaration(None=@n7."Color(s)") {name=ReactivityColor, nodeColor=<rgba 0 0.0391 1 1>}
  n34 = TextureCoordinate()
  n35 = ScalarParameter() {defaultValue=16, group=PearlTaffyVista|Glitter, parameterName=Dense Sparkle Detail Tiling, sliderMax=64, sliderMin=8, sortPriority=2}
  n36 = ScalarParameter() {defaultValue=0.75, group=PearlTaffyVista|Glitter, parameterName=Dense Glitter Tiling Close}
  n37 = ScalarParameter() {defaultValue=0.75, group=PearlTaffyVista|Glitter, parameterName=Dense Glitter Tiling Far}
  n38 = MaterialFunctionCall(
   Fade Length=ScalarParameter() {defaultValue=2, group=PearlTaffyVista|Glitter, parameterName=Dense Glitter Fade Length}
   Fade Offset=ScalarParameter() {defaultValue=100, group=PearlTaffyVista|Glitter, parameterName=Dense Glitter Fade Offset}
  ) {materialFunction=&CameraDepthFade}
  n39 = NamedRerouteDeclaration(None=Multiply(A=Multiply(A=@n34, B=@n35) {constB=3}, B=LinearInterpolate(A=@n36, B=@n37, Alpha=@n38.Result))) {name=UVs, nodeColor=<rgba 1 0.5859 0 1>}
  n40 = TextureSampleParameter2D(UVs=@n39) {
   group=Glitter
   parameterName=Sparkle Mask
   samplerSource=SSM_Wrap_WorldGroupSettings
   samplerType=SAMPLERTYPE_Masks
   sortPriority=0
   texture=/Game/Weapons/WeaponSkins/Textures/T_SparkleMask
  }
  n41 = Constant2Vector() {g=0.3256, r=0.1235}
  n42 = TextureSampleParameter2D(UVs=BumpOffset(Coordinate=Add(A=@n39, B=@n41), Height=ScalarParameter() {defaultValue=-2, group=PearlTaffyVista|Glitter, parameterName=Dense Glitter Depth})) {
   group=Glitter
   parameterName=Sparkle Mask
   samplerSource=SSM_Wrap_WorldGroupSettings
   samplerType=SAMPLERTYPE_Masks
   sortPriority=0
   texture=&T_SparkleMask
  }
  NamedRerouteDeclaration(None=Multiply(A=@n40.R, B=@n42.R)) {name=Glitter, nodeColor=<rgba 1 0 0.7734 1>}
  NamedRerouteDeclaration(None=@n42.G) {name=GlitterMask, nodeColor=<rgba 0 1 0.9062 1>}
 ]
}
