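# Control Rig for the F_MED_PearlTaffyVista body that adds corrective deformation to its skirt, belts, bag, darts and feet.
# Setup creates pelvis-parented nulls from listed bones; each solve runs FortDeformRig, then a five-branch sequence.
# Skirt master bones blend weighted parent constraints with correctives driven by spherical pose readers, thigh attributes, pelvis height and idle/slide states.
# Belts and bag follow weighted constraints and joint-rotation correctives, while darts ride a spline-placed, aimed null.
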ControlRigBlueprint /BRCosmetics/Characters/Player/Female/Medium/Bodies/F_MED_PearlTaffyVista/Meshes/F_MED_PearlTaffyVista_CtrlRig
graphs [
 RigVMFunctionLibrary {links [], nodes []}
 RigVMFunction_Sequence_1_ContainedGraph {nodes [
   n1 = RigVMFunctionEntryNode()
   RigVMFunction_Sequence = &RigVMFunction_Sequence(ExecuteContext=@n1.ExecuteContext)
   n2 = &RigVMFunction_Sequence(ExecuteContext=@RigVMFunction_Sequence.B)
   n3 = &RigVMFunction_Sequence(ExecuteContext=@n2.B)
   n4 = &RigVMFunction_Sequence(ExecuteContext=@n3.B)
   RigVMFunctionReturnNode(A=@RigVMFunction_Sequence.A, B=@n2.A, C=@n3.A, D=@n4.A, E=@n4.B)
  ]}
 RigVMModel {nodes [
   n1 = &RigUnit_PrepareForExecution()
   ItemArray = RigVMDispatchNode(Value="((Type=Bone,Name=\"dyn_skirt_bk_end_r\"),(Type=Bone,Name=\"dyn_skirt_bk_end\"),(Type=Bone,Name=\"dyn_skirt_fr_end\"),(Type=Bone,Name=\"dyn_skirt_bk_end_l\"),(Type=Bone,Name=\"dyn_skirt_side_end_l\"),(Type=Bone,Name=\"dyn_skirt_fr_end_l\"),(Type=Bone,Name=\"dyn_skirt_fr ...(truncated)") {template="DISPATCH_RigVMDispatch_Constant(io Value)"}
   For_Each = RigVMDispatchNode(Array=@ItemArray.Value, ExecuteContext=@n1.ExecutePin) {template="DISPATCH_RigVMDispatch_ArrayIterator(in Array,out Element,out Index,out Count,out Ratio)"}
   Concat = /Script/RigVM.RigVMFunction_NameConcat(B=_null, A=@For_Each.Element.Name) {is=struct RigVMUnitNode, template="Concat::Execute(in A,in B,out Result)"}
   n2 = &RigUnit_GetRelativeTransformForItem(Parent=(Type=Bone,Name="pelvis"), bChildInitial=False, bParentInitial=False, Child=@For_Each.Element)
   > &RigUnit_HierarchyAddNull(
    Parent=(Type=Bone,Name="pelvis")
    Space=LocalSpace
    Transform=(Rotation=(X=0,Y=0,Z=0,W=1),Translation=(X=0,Y=0,Z=0),Scale3D=(X=1,Y=1,Z=1))
    Name=@Concat.Result
    Transform.Scale3D=@n2.RelativeTransform.Scale3D
    Transform.Translation=@n2.RelativeTransform.Translation
   )
   RigUnit_BeginExecution = &RigUnit_BeginExecution()
   /Script/FortniteGame.RigUnit_FortDeformRig(
    Bool=(bUpperArm=true,bElbow=true,bWrist=true,bGroin=true,bGlute=true,bKnee=true,bPec=true,bBicep=False,bLat=False)
    Curve="(bLUpperArmUpAlpha=False,bRUpperArmUpAlpha=False,bLUpperArmFrAlpha=False,bRUpperArmFrAlpha=False,bLUpperArmBkAlpha=false,bRUpperArmBkAlpha=false,bLWristUpAlpha=False,bRWristUpAlpha=False,bLWristDnAlpha=False,bRWristDnAlpha=False,bLKneeAlpha=False,bRKneeAlp ...(truncated)"
    Debug=(bUpperArmDebug=False,bWristDebug=False,bLatDebug=False,bGroinGluteDebug=False)
    Value="(UpperArmUp=(Rotation=(X=0.0,Y=-0.0,Z=0.0,W=1.0),Translation=(X=0.0,Y=0.0,Z=8.0),Scale3D=(X=0.70,Y=1.0,Z=1.0)),UpperArmFwdFr=(Rotation=(X=0.0,Y=-0.0,Z=0.0,W=1.0),Translation=(X=0.0,Y=10.0,Z=0.0),Scale3D=(X=0.50,Y=1.0,Z=1.0)),UpperArmFwdBk=(Rotation=(X=0.0, ...(truncated)"
    ExecutePin=@RigUnit_BeginExecution.ExecutePin
   ) {is=struct RigVMUnitNode}
   > RigVMFunction_Sequence = RigVMAggregateNode()
   ItemArray_2 = RigVMDispatchNode(Value="((Type=Null,Name=\"dyn_skirt_bk_end_r_null\"),(Type=Null,Name=\"dyn_skirt_bk_end_null\"),(Type=Null,Name=\"dyn_skirt_fr_end_null\"),(Type=Null,Name=\"dyn_skirt_bk_end_l_null\"),(Type=Null,Name=\"dyn_skirt_side_end_l_null\"),(Type=Null,Name=\"dyn_skirt_fr_end_l_null\") ...(truncated)") {template="DISPATCH_RigVMDispatch_Constant(io Value)"}
   For_Each_2 = RigVMDispatchNode(Array=@ItemArray_2.Value, ExecuteContext=@RigVMFunction_Sequence.A) {template="DISPATCH_RigVMDispatch_ArrayIterator(in Array,out Element,out Index,out Count,out Ratio)"}
   > &RigUnit_ParentConstraint(
    AdvancedSettings=(InterpolationType=Average,RotationOrderForFilter=XZY)
    Filter=(TranslationFilter=(bX=true,bY=true,bZ=true),RotationFilter=(bX=false,bY=false,bZ=false),ScaleFilter=(bX=true,bY=true,bZ=true))
    Parents=((Item=(Type=Bone,Name="None"),Weight=1))
    Weight=1
    bMaintainOffset=true
    Parents.0.Item.Name=/Script/RigVM.RigVMFunction_NameTruncate(Count=5, FromEnd=True, Name=@For_Each_2.Element.Name) {is=struct RigVMUnitNode, template="Chop::Execute(in Name,in Count,in FromEnd,out Remainder,out Chopped)"}.Remainder
    Child=@For_Each_2.Element
   )
   &RigUnit_ParentConstraint(
    AdvancedSettings=(InterpolationType=Average,RotationOrderForFilter=XZY)
    Child=(Type=Bone,Name="MSTR_skirt_fr_l")
    Filter=(TranslationFilter=(bX=true,bY=true,bZ=true),RotationFilter=(bX=true,bY=true,bZ=true),ScaleFilter=(bX=true,bY=true,bZ=true))
    Parents=((Item=(Type=Bone,Name="thigh_twist_01_l"),Weight=0.9),(Item=(Type=Bone,Name="deform_groin_l"),Weight=0.05),(Item=(Type=Bone,Name="pelvis"),Weight=0.05))
    Weight=1
    bMaintainOffset=true
    ExecutePin=@RigVMFunction_Sequence.B
   )
   > DistanceRemap = RigVMFunctionReferenceNode(
    AxisToAvoid=Z
    AxisToTrack=X
    DistanceMax=8.8
    DistanceMin=-8.8
    Enable Debug=false
    EndJoint=(Type=Null,Name="thigh_twist_01_r_null")
    ParentSpace=(Type=Bone,Name="pelvis")
    StartJoint=(Type=Null,Name="thigh_twist_01_l_null")
    TargetMax=1
    TargetMin=-1
   )
   > &RigUnit_ParentConstraint(
    AdvancedSettings=(InterpolationType=Average,RotationOrderForFilter=XZY)
    Child=(Type=Bone,Name="MSTR_skirt_fr")
    Filter=(TranslationFilter=(bX=true,bY=true,bZ=true),RotationFilter=(bX=true,bY=true,bZ=true),ScaleFilter=(bX=true,bY=True,bZ=True))
    Parents=((Item=(Type=Bone,Name="thigh_twist_01_l"),Weight=0.9),(Item=(Type=Bone,Name="thigh_twist_01_r"),Weight=0.5),(Item=(Type=Bone,Name="deform_groin_l"),Weight=0.3),(Item=(Type=Bone,Name="deform_groin_r"),Weight=0.3),(Item=(Type=Bone,Name="pelvis"),Weight=0))
    Weight=1
    bMaintainOffset=true
    Parents.1.Weight=&RigVMFunction_MathDoubleRemap(SourceMaximum=1, SourceMinimum=0, TargetMaximum=1, TargetMinimum=0.3, Value=@DistanceRemap.Value).Result
    Parents.0.Weight=&RigVMFunction_MathDoubleRemap(SourceMaximum=0, SourceMinimum=-1, TargetMaximum=0.3, TargetMinimum=1, Value=@DistanceRemap.Value).Result
   )
   > &RigUnit_ParentConstraint(
    AdvancedSettings=(InterpolationType=Average,RotationOrderForFilter=XZY)
    Child=(Type=Bone,Name="MSTR_skirt_fr_r")
    Filter=(TranslationFilter=(bX=true,bY=true,bZ=true),RotationFilter=(bX=true,bY=True,bZ=True),ScaleFilter=(bX=True,bY=True,bZ=True))
    Parents=((Item=(Type=Bone,Name="thigh_twist_01_r"),Weight=0.9),(Item=(Type=Bone,Name="deform_groin_r"),Weight=0.1),(Item=(Type=Bone,Name="pelvis"),Weight=0.05))
    Weight=1
   )
   > &RigUnit_ParentConstraint(
    AdvancedSettings=(InterpolationType=Average,RotationOrderForFilter=XZY)
    Child=(Type=Bone,Name="MSTR_skirt_side_l")
    Filter=(TranslationFilter=(bX=true,bY=true,bZ=true),RotationFilter=(bX=true,bY=true,bZ=true),ScaleFilter=(bX=True,bY=True,bZ=True))
    Parents=((Item=(Type=Bone,Name="thigh_twist_01_l"),Weight=0.7),(Item=(Type=Bone,Name="pelvis"),Weight=0),(Item=(Type=Bone,Name="deform_groin_l"),Weight=0.05))
    Weight=1
   )
   > &RigUnit_ParentConstraint(
    AdvancedSettings=(InterpolationType=Average,RotationOrderForFilter=XZY)
    Child=(Type=Bone,Name="MSTR_skirt_side_r")
    Filter=(TranslationFilter=(bX=true,bY=true,bZ=true),RotationFilter=(bX=true,bY=true,bZ=True),ScaleFilter=(bX=True,bY=True,bZ=True))
    Parents=((Item=(Type=Bone,Name="thigh_twist_01_r"),Weight=0.7),(Item=(Type=Bone,Name="pelvis"),Weight=0),(Item=(Type=Bone,Name="deform_groin_r"),Weight=0.05))
    Weight=1
    bMaintainOffset=true
   )
   > n3 = &RigUnit_SphericalPoseReader(
    ActiveRegionScaleFactors=(PositiveWidth=1,NegativeWidth=1,PositiveHeight=1,NegativeHeight=1)
    ActiveRegionSize=0.03
    Debug=(bDrawDebug=true,bDraw2D=False,bDrawLocalAxes=False,DebugScale=25,DebugSegments=20,DebugThickness=0.25)
    DriverAxis=(X=-1,Y=0,Z=0)
    DriverItem=(Type=Bone,Name="thigh_l")
    FalloffRegionScaleFactors=(PositiveWidth=0.3,NegativeWidth=1,PositiveHeight=1,NegativeHeight=0.4)
    FalloffSize=0.4
    FlipHeightScaling=False
    FlipWidthScaling=False
    OptionalParentItem=(Type=Bone,Name="None")
    RotationOffset=(X=140,Y=0,Z=0)
   )
   > &RigUnit_ParentConstraint(
    AdvancedSettings=(InterpolationType=Average,RotationOrderForFilter=XZY)
    Child=(Type=Bone,Name="MSTR_skirt_bk_l")
    Filter=(TranslationFilter=(bX=true,bY=true,bZ=true),RotationFilter=(bX=true,bY=true,bZ=true),ScaleFilter=(bX=True,bY=True,bZ=True))
    Parents=((Item=(Type=Bone,Name="thigh_twist_01_l"),Weight=0.3),(Item=(Type=Bone,Name="pelvis"),Weight=2),(Item=(Type=Bone,Name="deform_glute_l"),Weight=0.3))
    Weight=1
   )
   > &RigUnit_ModifyTransforms(
    ItemToModify="((Item=(Type=Bone,Name=\"MSTR_skirt_side_l\"),Transform=(Rotation=(X=0.0008,Y=0.0158,Z=0.0521,W=-0.9985),Translation=(X=4.7727,Y=-3.245,Z=0.1521),Scale3D=(X=0.8452,Y=1,Z=1))),(Item=(Type=Bone,Name=\"MSTR_skirt_bk_l\"),Transform=(Rotation=(X=-0.0389,Y=-0.2234,Z ...(truncated)"
    WeightMinimum=0
    Weight=&RigVMFunction_MathDoubleRemap(SourceMaximum=1, SourceMinimum=0, TargetMaximum=1, TargetMinimum=0, Value=@n3.OutputParam).Result
   )
   > &RigUnit_ParentConstraint(
    AdvancedSettings=(InterpolationType=Average,RotationOrderForFilter=XZY)
    Child=(Type=Bone,Name="MSTR_skirt_bk_r")
    Filter=(TranslationFilter=(bX=true,bY=true,bZ=true),RotationFilter=(bX=true,bY=true,bZ=True),ScaleFilter=(bX=True,bY=True,bZ=True))
    Parents=((Item=(Type=Bone,Name="thigh_twist_01_r"),Weight=0.3),(Item=(Type=Bone,Name="pelvis"),Weight=2),(Item=(Type=Bone,Name="deform_glute_r"),Weight=0.3))
    Weight=1
   )
   > n4 = &RigUnit_SphericalPoseReader(
    ActiveRegionScaleFactors=(PositiveWidth=1,NegativeWidth=1,PositiveHeight=1,NegativeHeight=1)
    ActiveRegionSize=0.03
    Debug=(bDrawDebug=false,bDraw2D=False,bDrawLocalAxes=False,DebugScale=25,DebugSegments=20,DebugThickness=0.25)
    DriverAxis=(X=1,Y=0,Z=0)
    DriverItem=(Type=Bone,Name="thigh_r")
    FalloffRegionScaleFactors=(PositiveWidth=0.3,NegativeWidth=1,PositiveHeight=1,NegativeHeight=0.5)
    FalloffSize=0.4
    FlipHeightScaling=false
    FlipWidthScaling=true
    OptionalParentItem=(Type=Bone,Name="None")
    RotationOffset=(X=-40,Y=0,Z=0)
   )
   > &RigUnit_ModifyTransforms(
    ItemToModify="((Item=(Type=Bone,Name=\"MSTR_skirt_side_r\"),Transform=(Rotation=(X=0.0007,Y=0.0158,Z=-0.0436,W=0.9989),Translation=(X=4.7,Y=-3.245,Z=-0.1521),Scale3D=(X=0.8452,Y=1,Z=1))),(Item=(Type=Bone,Name=\"MSTR_skirt_bk_r\"),Transform=(Rotation=(X=0.0389,Y=0.2234,Z=-0. ...(truncated)"
    WeightMinimum=0
    Weight=&RigVMFunction_MathDoubleRemap(SourceMaximum=1, SourceMinimum=0, TargetMaximum=1, TargetMinimum=0, Value=@n4.OutputParam).Result
   )
   > &RigUnit_ParentConstraint(
    AdvancedSettings=(InterpolationType=Average,RotationOrderForFilter=XZY)
    Child=(Type=Bone,Name="MSTR_skirt_bk")
    Filter=(TranslationFilter=(bX=false,bY=false,bZ=true),RotationFilter=(bX=true,bY=true,bZ=true),ScaleFilter=(bX=True,bY=True,bZ=True))
    Parents=((Item=(Type=Bone,Name="MSTR_skirt_bk_l"),Weight=0.1),(Item=(Type=Bone,Name="pelvis"),Weight=0.8),(Item=(Type=Bone,Name="MSTR_skirt_bk_r"),Weight=0.1))
    Weight=1
   )
   Get_Animation_Attribute_2 = RigVMDispatchNode(BoneName=None, Default=0, Name=deformRig_l_thigh_up_alpha) {template="DISPATCH_RigDispatch_GetAnimAttribute(in Name,in BoneName,in Default,out Value,out Found)"}
   > &RigUnit_ModifyTransforms(
    ItemToModify="((Item=(Type=Bone,Name=\"MSTR_skirt_fr_l\"),Transform=(Rotation=(X=0,Y=0,Z=-0.0175,W=0.9998),Translation=(X=7,Y=-1,Z=0),Scale3D=(X=0.8,Y=1,Z=1))),(Item=(Type=Bone,Name=\"MSTR_skirt_side_l\"),Transform=(Rotation=(X=0,Y=0,Z=0.0872,W=0.9962),Translation=(X=0,Y=-1 ...(truncated)"
    WeightMinimum=0
    Weight=@Get_Animation_Attribute_2.Value
   )
   Get_Animation_Attribute_3 = RigVMDispatchNode(BoneName=None, Default=0, Name=deformRig_r_thigh_up_alpha) {template="DISPATCH_RigDispatch_GetAnimAttribute(in Name,in BoneName,in Default,out Value,out Found)"}
   > &RigUnit_ModifyTransforms(
    ItemToModify="((Item=(Type=Bone,Name=\"MSTR_skirt_fr_r\"),Transform=(Rotation=(X=0,Y=0,Z=-0.0262,W=0.9997),Translation=(X=7,Y=-0.5,Z=2),Scale3D=(X=0.8,Y=1,Z=1))),(Item=(Type=Bone,Name=\"MSTR_skirt_side_r\"),Transform=(Rotation=(X=0,Y=0,Z=0.0872,W=0.9962),Translation=(X=0,Y= ...(truncated)"
    WeightMinimum=0
    Weight=@Get_Animation_Attribute_3.Value
   )
   n5 = &RigUnit_GetRelativeTransformForItem(Child=(Type=Bone,Name="pelvis"), Parent=(Type=Bone,Name="root"), bChildInitial=False, bParentInitial=False)
   > &RigUnit_ModifyTransforms(
    ItemToModify="((Item=(Type=Bone,Name=\"MSTR_skirt_bk\"),Transform=(Rotation=(X=0,Y=0,Z=0.0872,W=0.9962),Translation=(X=0,Y=0,Z=0),Scale3D=(X=1,Y=1,Z=1))),(Item=(Type=Bone,Name=\"MSTR_skirt_bk_l\"),Transform=(Rotation=(X=0,Y=0,Z=0.0436,W=0.999),Translation=(X=0,Y=0,Z=0),Scal ...(truncated)"
    WeightMinimum=0
    Weight=&RigVMFunction_MathDoubleMul(
     A=&RigVMFunction_MathDoubleRemap(SourceMaximum=70, SourceMinimum=90, TargetMaximum=1, TargetMinimum=0, Value=@n5.RelativeTransform.Translation.Z).Result
     B=&RigVMFunction_MathDoubleRemap(SourceMaximum=50, SourceMinimum=65, TargetMaximum=0, TargetMinimum=1, Value=@n5.RelativeTransform.Translation.Z).Result
    ).Result
   )
   > n6 = &RigUnit_SphericalPoseReader(
    ActiveRegionScaleFactors=(PositiveWidth=1.5,NegativeWidth=1.5,PositiveHeight=1.5,NegativeHeight=1)
    ActiveRegionSize=0.03
    Debug=(bDrawDebug=false,bDraw2D=False,bDrawLocalAxes=False,DebugScale=25,DebugSegments=20,DebugThickness=0.25)
    DriverAxis=(X=-1,Y=0,Z=0)
    DriverItem=(Type=Bone,Name="thigh_l")
    FalloffRegionScaleFactors=(PositiveWidth=1,NegativeWidth=1,PositiveHeight=1,NegativeHeight=0.6)
    FalloffSize=0.2
    FlipHeightScaling=False
    FlipWidthScaling=False
    OptionalParentItem=(Type=Bone,Name="pelvis")
    RotationOffset=(X=110,Y=0,Z=85)
   )
   Cast = &RigVMFunction_MathBoolToFloat(Value=RigVMVariableNode() {variable=isFrontEndIdle}.Value)
   AlphaInterp = &RigVMFunction_AlphaInterp(InterpSpeedDecreasing=5, InterpSpeedIncreasing=5, Scale=1, bClampResult=False, Value=@Cast.Result)
   > &RigUnit_ModifyTransforms(
    ItemToModify="((Item=(Type=Bone,Name=\"MSTR_skirt_side_l\"),Transform=(Rotation=(X=0,Y=0,Z=0.0436,W=0.999),Translation=(X=0,Y=0,Z=0),Scale3D=(X=0.95,Y=1,Z=1))),(Item=(Type=Bone,Name=\"MSTR_skirt_fr_l\"),Transform=(Rotation=(X=0,Y=0.0175,Z=0,W=0.9998),Translation=(X=0,Y=0,Z= ...(truncated)"
    WeightMinimum=0
    Weight=&RigVMFunction_MathDoubleMul(B=@AlphaInterp.Result, A=@n6.OutputParam).Result
   )
   > n7 = &RigUnit_SphericalPoseReader(
    ActiveRegionScaleFactors=(PositiveWidth=1.5,NegativeWidth=1.5,PositiveHeight=1.5,NegativeHeight=1)
    ActiveRegionSize=0.03
    Debug=(bDrawDebug=false,bDraw2D=False,bDrawLocalAxes=False,DebugScale=25,DebugSegments=20,DebugThickness=0.25)
    DriverAxis=(X=1,Y=0,Z=0)
    DriverItem=(Type=Bone,Name="thigh_r")
    FalloffRegionScaleFactors=(PositiveWidth=1,NegativeWidth=1,PositiveHeight=1,NegativeHeight=0.6)
    FalloffSize=0.2
    FlipHeightScaling=false
    FlipWidthScaling=false
    OptionalParentItem=(Type=Bone,Name="pelvis")
    RotationOffset=(X=-70,Y=0,Z=90)
   )
   > &RigUnit_ModifyTransforms(
    ItemToModify="((Item=(Type=Bone,Name=\"MSTR_skirt_side_r\"),Transform=(Rotation=(X=0,Y=0,Z=0.0436,W=0.999),Translation=(X=0,Y=0,Z=0),Scale3D=(X=0.95,Y=1,Z=1))),(Item=(Type=Bone,Name=\"MSTR_skirt_fr_r\"),Transform=(Rotation=(X=0,Y=0,Z=0,W=1),Translation=(X=0,Y=0,Z=0),Scale3D ...(truncated)"
    WeightMinimum=0
    Weight=&RigVMFunction_MathDoubleMul(B=@AlphaInterp.Result, A=@n7.OutputParam).Result
   )
   RigVMFunction_MathQuaternionToEuler = &RigVMFunction_MathQuaternionToEuler(
    RotationOrder=ZYX
    Value=&RigUnit_GetTransform(Item=(Type=Bone,Name="foot_l"), Space=LocalSpace, bInitial=False).Transform.Rotation
   )
   Remap = &RigVMFunction_MathDoubleRemap(SourceMaximum=80, SourceMinimum=0, TargetMaximum=1, TargetMinimum=0, Value=@RigVMFunction_MathQuaternionToEuler.Result.Z)
   &RigUnit_ModifyTransforms(
    ItemToModify=((Item=(Type=Bone,Name="def_corr_foot_l"),Transform=(Rotation=(X=0,Y=0,Z=-0.5,W=0.866),Translation=(X=2,Y=-3,Z=0),Scale3D=(X=1,Y=1,Z=1))))
    WeightMinimum=0
    Weight=@Remap.Result
    ExecutePin=@RigVMFunction_Sequence.C
   )
   n8 = &RigUnit_GetTransform(Item=(Type=Bone,Name="foot_r"), Space=LocalSpace, bInitial=False)
   > &RigUnit_ModifyTransforms(
    ItemToModify=((Item=(Type=Bone,Name="def_corr_foot_r"),Transform=(Rotation=(X=0,Y=0,Z=-0.5,W=0.866),Translation=(X=2,Y=-3,Z=0),Scale3D=(X=1,Y=1,Z=1))))
    WeightMinimum=0
    Weight=&RigVMFunction_MathDoubleRemap(
     SourceMaximum=80
     SourceMinimum=0
     TargetMaximum=1
     TargetMinimum=0
     Value=&RigVMFunction_MathQuaternionToEuler(RotationOrder=ZYX, Value=@n8.Transform.Rotation).Result.Z
    ).Result
   )
   n9 = &RigVMFunction_MathBoolToFloat(Value=RigVMVariableNode() {variable=isSliding}.Value)
   n10 = &RigVMFunction_MathDoubleRemap(
    SourceMaximum=140
    SourceMinimum=110
    TargetMaximum=1
    TargetMinimum=0
    Value=&RigVMFunction_MathQuaternionToEuler(
     RotationOrder=ZYX
     Value=&RigUnit_GetTransform(Item=(Type=Bone,Name="calf_r"), Space=LocalSpace, bInitial=False).Transform.Rotation
    ).Result.Z
   )
   n11 = &RigVMFunction_MathDoubleRemap(
    SourceMaximum=30
    SourceMinimum=45
    TargetMaximum=1
    TargetMinimum=0
    Value=&RigUnit_GetTransform(Item=(Type=Bone,Name="pelvis"), Space=LocalSpace, bInitial=False).Transform.Translation.Z
   )
   n12 = &RigVMFunction_AlphaInterp(
    InterpSpeedDecreasing=10
    InterpSpeedIncreasing=20
    Scale=1
    bClampResult=False
    bInterpResult=True
    Value=&RigVMFunction_MathDoubleMul(A=@n9.Result, B=&RigVMFunction_MathDoubleMul(A=@n10.Result, B=@n11.Result).Result).Result
   )
   > &RigUnit_ModifyTransforms(
    ItemToModify="((Item=(Type=Bone,Name=\"dyn_skirt_bk_02_r\"),Transform=(Rotation=(X=0,Y=0,Z=0.2588,W=0.9659),Translation=(X=-1,Y=-1,Z=0),Scale3D=(X=0.8,Y=1,Z=1))),(Item=(Type=Bone,Name=\"dyn_skirt_bk_02\"),Transform=(Rotation=(X=0.017,Y=-0.1294,Z=0.1294,W=0.983),Translation= ...(truncated)"
    WeightMinimum=0
    Weight=@n12.Result
   )
   n13 = &RigVMFunction_MathDoubleRemap(
    SourceMaximum=140
    SourceMinimum=110
    TargetMaximum=1
    TargetMinimum=0
    Value=&RigVMFunction_MathQuaternionToEuler(
     RotationOrder=ZYX
     Value=&RigUnit_GetTransform(Item=(Type=Bone,Name="calf_l"), Space=LocalSpace, bInitial=False).Transform.Rotation
    ).Result.Z
   )
   n14 = &RigVMFunction_AlphaInterp(
    InterpSpeedDecreasing=10
    InterpSpeedIncreasing=20
    Scale=1
    bClampResult=False
    bInterpResult=True
    Value=&RigVMFunction_MathDoubleMul(A=@n9.Result, B=&RigVMFunction_MathDoubleMul(A=@n13.Result, B=@n11.Result).Result).Result
   )
   > &RigUnit_ModifyTransforms(
    ItemToModify="((Item=(Type=Bone,Name=\"dyn_skirt_bk_02\"),Transform=(Rotation=(X=-0.017,Y=0.1294,Z=0.1294,W=0.983),Translation=(X=1,Y=0,Z=0),Scale3D=(X=0.8,Y=1,Z=1))),(Item=(Type=Bone,Name=\"dyn_skirt_bk_02_l\"),Transform=(Rotation=(X=0,Y=0,Z=0.2588,W=0.9659),Translation=(X ...(truncated)"
    WeightMinimum=0
    Weight=@n14.Result
   )
   > n15 = &RigUnit_SphericalPoseReader(
    ActiveRegionScaleFactors=(PositiveWidth=1,NegativeWidth=3,PositiveHeight=2,NegativeHeight=2)
    ActiveRegionSize=0.03
    Debug=(bDrawDebug=false,bDraw2D=False,bDrawLocalAxes=False,DebugScale=25,DebugSegments=20,DebugThickness=0.25)
    DriverAxis=(X=1,Y=0,Z=0)
    DriverItem=(Type=Bone,Name="thigh_r")
    FalloffRegionScaleFactors=(PositiveWidth=1,NegativeWidth=1,PositiveHeight=0.5,NegativeHeight=0.5)
    FalloffSize=0.2
    FlipHeightScaling=False
    FlipWidthScaling=False
    OptionalParentItem=(Type=Bone,Name="None")
    RotationOffset=(X=-107,Y=-10,Z=5)
   )
   > n16 = &RigUnit_SphericalPoseReader(
    ActiveRegionScaleFactors=(PositiveWidth=1,NegativeWidth=1,PositiveHeight=1,NegativeHeight=1)
    ActiveRegionSize=0.05
    Debug=(bDrawDebug=false,bDraw2D=False,bDrawLocalAxes=False,DebugScale=25,DebugSegments=20,DebugThickness=0.25)
    DriverAxis=(X=-1,Y=0,Z=0)
    DriverItem=(Type=Bone,Name="thigh_l")
    FalloffRegionScaleFactors=(PositiveWidth=0.7,NegativeWidth=0.7,PositiveHeight=0.5,NegativeHeight=0.5)
    FalloffSize=0.2
    FlipHeightScaling=False
    FlipWidthScaling=False
    OptionalParentItem=(Type=Bone,Name="None")
    RotationOffset=(X=67,Y=-30,Z=60)
   )
   n17 = &RigUnit_GetRelativeTransformForItem(Child=(Type=Bone,Name="pelvis"), Parent=(Type=Bone,Name="root"), bChildInitial=False, bParentInitial=False)
   > &RigUnit_ModifyTransforms(
    ItemToModify="((Item=(Type=Bone,Name=\"MSTR_skirt_fr\"),Transform=(Rotation=(X=0.1766,Y=0.0782,Z=-0.0579,W=0.9795),Translation=(X=1,Y=0,Z=0),Scale3D=(X=1,Y=1,Z=1))),(Item=(Type=Bone,Name=\"MSTR_skirt_fr_r\"),Transform=(Rotation=(X=-0.0019,Y=0.0436,Z=0.0436,W=0.9981),Transla ...(truncated)"
    WeightMinimum=0
    Weight=&RigVMFunction_MathDoubleMul(
     A=&RigVMFunction_MathDoubleMul(A=@n15.OutputParam, B=@n16.OutputParam).Result
     B=&RigVMFunction_MathDoubleRemap(SourceMaximum=65, SourceMinimum=98, TargetMaximum=1, TargetMinimum=0, Value=@n17.RelativeTransform.Translation.Z).Result
    ).Result
   )
   n18 = RigVMVariableNode() {variable=isSkydiving}
   > &RigUnit_ModifyTransforms(
    ItemToModify="((Item=(Type=Bone,Name=\"upperarm_twist_02_r\"),Transform=(Rotation=(X=0,Y=0,Z=0,W=1),Translation=(X=0,Y=0,Z=0),Scale3D=(X=2,Y=2.1,Z=2))),(Item=(Type=Bone,Name=\"upperarm_twist_01_r\"),Transform=(Rotation=(X=0,Y=0,Z=0,W=1),Translation=(X=0,Y=0,Z=0),Scale3D=(X= ...(truncated)"
    WeightMinimum=0
    Weight=@n18.Value
   )
   > &RigUnit_ModifyTransforms(
    ItemToModify="((Item=(Type=Bone,Name=\"upperarm_twist_02_l\"),Transform=(Rotation=(X=0,Y=0,Z=0,W=1),Translation=(X=0,Y=0,Z=0),Scale3D=(X=2,Y=2.1,Z=2))),(Item=(Type=Bone,Name=\"upperarm_twist_01_l\"),Transform=(Rotation=(X=0,Y=0,Z=0,W=1),Translation=(X=0,Y=0,Z=0),Scale3D=(X= ...(truncated)"
    WeightMinimum=0
    Weight=@n18.Value
   )
   &RigUnit_ParentConstraint(
    AdvancedSettings=(InterpolationType=Average,RotationOrderForFilter=XZY)
    Child=(Type=Bone,Name="def_belt_fr_r")
    Filter=(TranslationFilter=(bX=true,bY=false,bZ=false),RotationFilter=(bX=true,bY=true,bZ=true),ScaleFilter=(bX=True,bY=True,bZ=True))
    Parents=((Item=(Type=Bone,Name="dyn_skirt_fr_01_r"),Weight=0.5),(Item=(Type=Bone,Name="deform_groin_r"),Weight=0.2),(Item=(Type=Bone,Name="thigh_twist_01_r"),Weight=1))
    Weight=1
    ExecutePin=@RigVMFunction_Sequence.D
   )
   > &RigUnit_ParentConstraint(
    AdvancedSettings=(InterpolationType=Average,RotationOrderForFilter=XZY)
    Child=(Type=Bone,Name="def_belt_fr_l")
    Filter=(TranslationFilter=(bX=true,bY=false,bZ=false),RotationFilter=(bX=true,bY=true,bZ=true),ScaleFilter=(bX=True,bY=True,bZ=True))
    Parents=((Item=(Type=Bone,Name="dyn_skirt_fr_01_l"),Weight=0.2),(Item=(Type=Bone,Name="deform_groin_l"),Weight=0.5),(Item=(Type=Bone,Name="thigh_twist_01_l"),Weight=1))
    Weight=1
    bMaintainOffset=true
   )
   > &RigUnit_ParentConstraint(
    AdvancedSettings=(InterpolationType=Average,RotationOrderForFilter=XZY)
    Child=(Type=Bone,Name="def_belt_fr")
    Filter=(TranslationFilter=(bX=true,bY=true,bZ=true),RotationFilter=(bX=true,bY=true,bZ=true),ScaleFilter=(bX=True,bY=True,bZ=True))
    Parents=((Item=(Type=Bone,Name="def_belt_fr_l"),Weight=1),(Item=(Type=Bone,Name="def_belt_fr_r"),Weight=1),(Item=(Type=Bone,Name="dyn_skirt_fr_01"),Weight=0))
    Weight=1
   )
   > &RigUnit_ParentConstraint(
    AdvancedSettings=(InterpolationType=Average,RotationOrderForFilter=XZY)
    Child=(Type=Bone,Name="def_belt_fr_rear_r")
    Filter=(TranslationFilter=(bX=true,bY=true,bZ=true),RotationFilter=(bX=True,bY=True,bZ=True),ScaleFilter=(bX=True,bY=True,bZ=True))
    Parents=((Item=(Type=Bone,Name="dyn_skirt_fr_01_r"),Weight=0.5),(Item=(Type=Bone,Name="deform_groin_r"),Weight=0.5),(Item=(Type=Bone,Name="pelvis"),Weight=0.2))
    Weight=1
    bMaintainOffset=true
   )
   > &RigUnit_ParentConstraint(
    AdvancedSettings=(InterpolationType=Average,RotationOrderForFilter=XZY)
    Child=(Type=Bone,Name="def_belt_side_l")
    Filter=(TranslationFilter=(bX=true,bY=true,bZ=true),RotationFilter=(bX=true,bY=true,bZ=true),ScaleFilter=(bX=True,bY=True,bZ=True))
    Parents="((Item=(Type=Bone,Name=\"dyn_skirt_side_01_l\"),Weight=0.2),(Item=(Type=Bone,Name=\"deform_groin_l\"),Weight=0.3),(Item=(Type=Bone,Name=\"pelvis\"),Weight=1),(Item=(Type=Bone,Name=\"dyn_skirt_fr_01_l\"),Weight=0),(Item=(Type=Bone,Name=\"thigh_twist_01_l\"),Weight=0) ...(truncated)"
    Weight=1
    bMaintainOffset=true
   )
   > Get Delta Ref_1 = RigVMFunctionReferenceNode(Item=(Type=Bone,Name="thigh_r"))
   > &RigUnit_ModifyTransforms(
    ItemToModify="((Item=(Type=Bone,Name=\"def_belt_fr_r\"),Transform=(Rotation=(X=-0.0625,Y=0.1431,Z=-0.1603,W=0.9746),Translation=(X=0,Y=3,Z=-0.5),Scale3D=(X=1,Y=1,Z=1))),(Item=(Type=Bone,Name=\"dart_01\"),Transform=(Rotation=(X=0,Y=0,Z=0,W=1),Translation=(X=0,Y=0,Z=0),Scale3 ...(truncated)"
    WeightMinimum=0
    Weight=&RigVMFunction_MathDoubleRemap(SourceMaximum=-110, SourceMinimum=0, TargetMaximum=1, TargetMinimum=0, Value=@"Get Delta Ref_1".Result.Rotation.Yaw).Result
   )
   > Get Delta Ref = RigVMFunctionReferenceNode(Item=(Type=Bone,Name="thigh_l"))
   > &RigUnit_ModifyTransforms(
    ItemToModify="((Item=(Type=Bone,Name=\"def_belt_fr_l\"),Transform=(Rotation=(X=-0.1305,Y=0,Z=0,W=0.9914),Translation=(X=0,Y=3,Z=3),Scale3D=(X=1,Y=1,Z=1))),(Item=(Type=Bone,Name=\"def_belt_side_l\"),Transform=(Rotation=(X=0,Y=0,Z=0,W=1),Translation=(X=1,Y=1,Z=-1),Scale3D=(X= ...(truncated)"
    WeightMinimum=0
    Weight=&RigVMFunction_MathDoubleRemap(SourceMaximum=-130, SourceMinimum=0, TargetMaximum=1, TargetMinimum=0, Value=@"Get Delta Ref".Result.Rotation.Yaw).Result
   )
   > &RigUnit_ParentConstraint(
    AdvancedSettings=(RotationOrderForFilter=XZY)
    Child=(Type=Bone,Name="dyn_bag_01")
    Filter=(TranslationFilter=(bX=true,bY=true,bZ=true),RotationFilter=(bX=true,bY=false,bZ=false),ScaleFilter=(bX=True,bY=True,bZ=True))
    Parents="((Item=(Type=Bone,Name=\"spine_01\"),Weight=0.0139),(Item=(Type=Bone,Name=\"dyn_skirt_side_02_l\"),Weight=0.0263),(Item=(Type=Bone,Name=\"dyn_skirt_fr_01_l\"),Weight=0.0277),(Item=(Type=Bone,Name=\"spine_02\"),Weight=0.0523),(Item=(Type=Bone,Name=\"thigh_twist_01_l ...(truncated)"
    Weight=1
   )
   > &RigUnit_ParentConstraint(
    AdvancedSettings=(RotationOrderForFilter=XZY)
    Child=(Type=Bone,Name="dyn_bag_02")
    Filter=(TranslationFilter=(bX=true,bY=true,bZ=true),RotationFilter=(bX=true,bY=true,bZ=false),ScaleFilter=(bX=True,bY=True,bZ=True))
    Parents="((Item=(Type=Bone,Name=\"dyn_skirt_side_03_l\"),Weight=0.0208),(Item=(Type=Bone,Name=\"dyn_skirt_fr_02_l\"),Weight=0.0404),(Item=(Type=Bone,Name=\"dyn_skirt_fr_01_l\"),Weight=0.0585),(Item=(Type=Bone,Name=\"thigh_twist_01_l\"),Weight=0.0713),(Item=(Type=Bone,Name= ...(truncated)"
    Weight=1
   )
   > &RigUnit_ParentConstraint(
    AdvancedSettings=(RotationOrderForFilter=XZY)
    Child=(Type=Control,Name="def_belt_fr_r_ctrl")
    Parents="((Item=(Type=Bone,Name=\"spine_01\"),Weight=0.0215),(Item=(Type=Bone,Name=\"dyn_skirt_fr_02_r\"),Weight=0.022),(Item=(Type=Bone,Name=\"thigh_twist_01_r\"),Weight=0.0262),(Item=(Type=Bone,Name=\"dyn_skirt_fr_01\"),Weight=0.0495),(Item=(Type=Bone,Name=\"dyn_skirt_fr_ ...(truncated)"
    Weight=1
   )
   > &RigUnit_ParentConstraint(
    AdvancedSettings=(RotationOrderForFilter=XZY)
    Child=(Type=Bone,Name="def_belt_fr_rear_r")
    Parents="((Item=(Type=Bone,Name=\"dyn_skirt_fr_02_r\"),Weight=0.0238),(Item=(Type=Bone,Name=\"dyn_skirt_side_02_r\"),Weight=0.0254),(Item=(Type=Bone,Name=\"spine_02\"),Weight=0.0362),(Item=(Type=Bone,Name=\"dyn_skirt_fr_01_r\"),Weight=0.0565),(Item=(Type=Bone,Name=\"thigh_t ...(truncated)"
    Weight=1
   )
   > &RigUnit_ParentConstraint(
    AdvancedSettings=(RotationOrderForFilter=XZY)
    Child=(Type=Control,Name="MSTR_skirt_side_r_ctrl")
    Parents="((Item=(Type=Bone,Name=\"dyn_skirt_bk_01_r\"),Weight=0.0069),(Item=(Type=Bone,Name=\"dyn_skirt_side_02_r\"),Weight=0.0419),(Item=(Type=Bone,Name=\"spine_02\"),Weight=0.0422),(Item=(Type=Bone,Name=\"spine_01\"),Weight=0.0601),(Item=(Type=Bone,Name=\"deform_groin_r\") ...(truncated)"
    Weight=1
   )
   > &RigUnit_ParentConstraint(
    AdvancedSettings=(RotationOrderForFilter=XZY)
    Child=(Type=Control,Name="MSTR_skirt_bk_r_ctrl")
    Parents="((Item=(Type=Bone,Name=\"deform_glute_r\"),Weight=0.0221),(Item=(Type=Bone,Name=\"dyn_skirt_side_02_r\"),Weight=0.0294),(Item=(Type=Bone,Name=\"dyn_skirt_bk_02_r\"),Weight=0.0309),(Item=(Type=Bone,Name=\"spine_02\"),Weight=0.043),(Item=(Type=Bone,Name=\"dyn_skirt_s ...(truncated)"
    Weight=1
   )
   > &RigUnit_ParentConstraint(
    AdvancedSettings=(InterpolationType=Average,RotationOrderForFilter=XZY)
    Child=(Type=Bone,Name="dart_01")
    Filter=(TranslationFilter=(bX=True,bY=True,bZ=True),RotationFilter=(bX=true,bY=true,bZ=true),ScaleFilter=(bX=True,bY=True,bZ=True))
    Parents=((Item=(Type=Null,Name="dart"),Weight=1))
    Weight=1
   )
   > &RigUnit_ParentConstraint(
    AdvancedSettings=(InterpolationType=Average,RotationOrderForFilter=XZY)
    Child=(Type=Bone,Name="dart_02")
    Filter=(TranslationFilter=(bX=True,bY=True,bZ=True),RotationFilter=(bX=true,bY=true,bZ=true),ScaleFilter=(bX=True,bY=True,bZ=True))
    Parents=((Item=(Type=Null,Name="dart"),Weight=1))
    Weight=1
   )
   > &RigUnit_ParentConstraint(
    AdvancedSettings=(InterpolationType=Average,RotationOrderForFilter=XZY)
    Child=(Type=Bone,Name="dart_03")
    Filter=(TranslationFilter=(bX=True,bY=True,bZ=True),RotationFilter=(bX=true,bY=true,bZ=true),ScaleFilter=(bX=True,bY=True,bZ=True))
    Parents=((Item=(Type=Null,Name="dart"),Weight=1))
    Weight=1
   )
   Get_Animation_Attribute_4 = RigVMDispatchNode(BoneName=None, Default=0, Name=deformRig_r_thigh_up_alpha) {template="DISPATCH_RigDispatch_GetAnimAttribute(in Name,in BoneName,in Default,out Value,out Found)"}
   > &RigUnit_ModifyTransforms(
    ItemToModify=((Item=(Type=Null,Name="dart"),Transform=(Rotation=(X=0,Y=0,Z=0,W=1),Translation=(X=0,Y=-2,Z=-1),Scale3D=(X=0.7,Y=0.7,Z=0.7))))
    WeightMinimum=0
    Weight=@Get_Animation_Attribute_4.Value
   )
   ItemArray_8 = RigVMDispatchNode(Value=((Type=Control,Name="def_belt_fr_r_ctrl"),(Type=Control,Name="def_belt_fr_rear_r_ctrl"),(Type=Control,Name="MSTR_skirt_side_r_ctrl"),(Type=Control,Name="MSTR_skirt_bk_r_ctrl"))) {template="DISPATCH_RigVMDispatch_Constant(io Value)"}
   > SplineFromItems_2 = RigVMFunctionReferenceNode(Samples Per Segment=16, Spline Mode=Hermite, Items=@ItemArray_8.Value)
   n19 = &RigVMFunction_MathTransformMakeRelative(
    Parent=&RigUnit_GetTransform(Item=(Type=Bone,Name="pelvis"), bInitial=False).Transform
    Global=/Script/ControlRigSpline.RigUnit_TransformFromControlRigSpline2(
     PrimaryAxis=(X=1,Y=0,Z=0)
     SecondaryAxis=(X=0,Y=1,Z=0)
     U=&RigVMFunction_MathDoubleRemap(SourceMaximum=1, SourceMinimum=0.2, TargetMaximum=0.4, TargetMinimum=0, Value=@Get_Animation_Attribute_4.Value).Result
     Spline=@SplineFromItems_2.Spline
    ) {is=struct RigVMUnitNode}.Transform
   )
   > &RigUnit_ModifyTransforms(
    ItemToModify=((Item=(Type=Null,Name="dart"),Transform=(Rotation=(X=0,Y=0,Z=0,W=1),Translation=(X=0,Y=0,Z=0),Scale3D=(X=0.9,Y=0.9,Z=0.9))))
    Mode=OverrideLocal
    Weight=0.7
    WeightMinimum=0
    ItemToModify.0.Transform.Rotation=@n19.Local.Rotation
    ItemToModify.0.Transform.Translation=@n19.Local.Translation
   )
   > AimAt = RigVMFunctionReferenceNode(
    Active=true
    Axis=(X=0,Y=0,Z=1)
    Debug Draw=true
    Up Vector Axis=(X=0,Y=-1,Z=0)
    Up Vector Object=(Type=Bone,Name="spine_01")
    object=(Type=Null,Name="dart")
    target=(Type=Null,Name="pelvis_null")
   )
   > Get Delta Ref_2 = RigVMFunctionReferenceNode(Item=(Type=Bone,Name="thigh_l"))
   > &RigUnit_ModifyTransforms(
    ItemToModify="((Item=(Type=Bone,Name=\"dyn_bag_01\"),Transform=(Rotation=(X=0.0435,Y=-0.0038,Z=-0.0871,W=0.9952),Translation=(X=0,Y=0.5,Z=2),Scale3D=(X=1,Y=1,Z=1))),(Item=(Type=Bone,Name=\"dyn_bag_02\"),Transform=(Rotation=(X=0.0435,Y=0.0871,Z=-0.0038,W=0.9952),Translation= ...(truncated)"
    WeightMinimum=0
    Weight=&RigVMFunction_MathDoubleRemap(SourceMaximum=-90, SourceMinimum=0, TargetMaximum=1, TargetMinimum=0, Value=@"Get Delta Ref_2".Result.Rotation.Yaw).Result
   )
   > &RigUnit_ModifyTransforms(
    ItemToModify="((Item=(Type=Bone,Name=\"dyn_bag_01\"),Transform=(Rotation=(X=-0.0762,Y=-0.4172,Z=-0.0973,W=0.9004),Translation=(X=2,Y=0,Z=2),Scale3D=(X=1,Y=1,Z=1))),(Item=(Type=Bone,Name=\"dyn_bag_02\"),Transform=(Rotation=(X=0.1722,Y=0.1285,Z=-0.0227,W=0.9764),Translation=( ...(truncated)"
    WeightMinimum=0
    Weight=&RigVMFunction_MathDoubleRemap(SourceMaximum=100, SourceMinimum=30, TargetMaximum=1, TargetMinimum=0, Value=@"Get Delta Ref_2".Result.Rotation.Pitch).Result
   )
   &RigUnit_ParentConstraint(
    AdvancedSettings=(RotationOrderForFilter=XZY)
    Child=(Type=Bone,Name="def_belt_back_corr_r")
    Parents="((Item=(Type=Bone,Name=\"spine_02\"),Weight=0),(Item=(Type=Bone,Name=\"spine_03\"),Weight=0),(Item=(Type=Bone,Name=\"deform_upperarm_bk_r\"),Weight=0),(Item=(Type=Bone,Name=\"upperarm_twist_01_r\"),Weight=0),(Item=(Type=Bone,Name=\"clavicle_r\"),Weight=0.1),(Item=(T ...(truncated)"
    Weight=1
    ExecutePin=@RigVMFunction_Sequence.E
   )
   > &RigUnit_ParentConstraint(
    AdvancedSettings=(RotationOrderForFilter=XZY)
    Child=(Type=Bone,Name="def_belt_back_corr_l")
    Filter=(TranslationFilter=(bX=true,bY=true,bZ=True),RotationFilter=(bX=True,bY=True,bZ=True),ScaleFilter=(bX=True,bY=True,bZ=True))
    Parents="((Item=(Type=Bone,Name=\"spine_02\"),Weight=0),(Item=(Type=Bone,Name=\"spine_03\"),Weight=0),(Item=(Type=Bone,Name=\"deform_upperarm_bk_l\"),Weight=0),(Item=(Type=Bone,Name=\"upperarm_twist_01_l\"),Weight=0),(Item=(Type=Bone,Name=\"clavicle_l\"),Weight=0.1),(Item=(T ...(truncated)"
    Weight=1
   )
   > Get Delta Ref_5 = RigVMFunctionReferenceNode(Item=(Type=Bone,Name="clavicle_l"))
   > &RigUnit_ModifyTransforms(
    ItemToModify=((Item=(Type=Bone,Name="def_belt_back_corr_l"),Transform=(Rotation=(X=0.13,Y=-0.0114,Z=-0.0864,W=0.9877),Translation=(X=0,Y=0.6,Z=-1),Scale3D=(X=1,Y=1,Z=1))))
    WeightMinimum=0
    Weight=&RigVMFunction_MathDoubleRemap(SourceMaximum=-25, SourceMinimum=0, TargetMaximum=1, TargetMinimum=0, Value=@"Get Delta Ref_5".Result.Rotation.Yaw).Result
   )
   > &RigUnit_ModifyTransforms(
    ItemToModify=((Item=(Type=Bone,Name="def_belt_back_corr_l"),Transform=(Rotation=(X=0,Y=0,Z=0,W=1),Translation=(X=0,Y=-0.3,Z=-0.5),Scale3D=(X=1,Y=1,Z=1))))
    WeightMinimum=0
    Weight=&RigVMFunction_MathDoubleRemap(SourceMaximum=30, SourceMinimum=0, TargetMaximum=1, TargetMinimum=0, Value=@"Get Delta Ref_5".Result.Rotation.Pitch).Result
   )
   > Get Delta Ref_6 = RigVMFunctionReferenceNode(Item=(Type=Bone,Name="clavicle_r"))
   > &RigUnit_ModifyTransforms(
    ItemToModify=((Item=(Type=Bone,Name="def_belt_back_corr_r"),Transform=(Rotation=(X=-0.13,Y=0.0114,Z=-0.0864,W=0.9877),Translation=(X=0,Y=0.6,Z=1),Scale3D=(X=1,Y=1,Z=0))))
    WeightMinimum=0
    Weight=&RigVMFunction_MathDoubleRemap(SourceMaximum=-25, SourceMinimum=0, TargetMaximum=1, TargetMinimum=0, Value=@"Get Delta Ref_6".Result.Rotation.Yaw).Result
   )
   > &RigUnit_ModifyTransforms(
    ItemToModify=((Item=(Type=Bone,Name="def_belt_back_corr_r"),Transform=(Rotation=(X=0,Y=0,Z=0,W=1),Translation=(X=0,Y=-0.3,Z=0.5),Scale3D=(X=1,Y=1,Z=1))))
    WeightMinimum=0
    Weight=&RigVMFunction_MathDoubleRemap(SourceMaximum=30, SourceMinimum=0, TargetMaximum=1, TargetMinimum=0, Value=@"Get Delta Ref_6".Result.Rotation.Pitch).Result
   )
   RigUnit_GetTransform = &RigUnit_GetTransform(Item=(Type=Bone,Name="pelvis"), bInitial=False)
   Add = /Script/RigVM.RigVMFunction_MathVectorAdd(
    B=(X=0,Y=4,Z=-3)
    B.Z=&RigVMFunction_MathDoubleRemap(SourceMaximum=1, SourceMinimum=0, TargetMaximum=-2, TargetMinimum=2, Value=@Get_Animation_Attribute_4.Value).Result
    B.Y=&RigVMFunction_MathDoubleRemap(SourceMaximum=1, SourceMinimum=0, TargetMaximum=4, TargetMinimum=-3, Value=@Get_Animation_Attribute_4.Value).Result
    A=@RigUnit_GetTransform.Transform.Translation
   ) {is=struct RigVMUnitNode, template="Add::Execute(in A,in B,out Result)"}
   &RigUnit_ModifyTransforms(
    ExecutePin=()
    ItemToModify=((Item=(Type=Bone,Name="thigh_l"),Transform=(Rotation=(X=0,Y=0,Z=0,W=1),Translation=(X=0,Y=0,Z=0),Scale3D=(X=1,Y=1,Z=1))),(Item=(Type=Bone,Name="thigh_r"),Transform=(Rotation=(X=0,Y=0,Z=0,W=1),Translation=(X=0,Y=0,Z=0),Scale3D=(X=1,Y=1,Z=1))))
    Weight=1
    WeightMinimum=0
   )
  ]}
]
hierarchy [
 root {loc=<0 0 0>, rot=<0 0 0 1>}
 pelvis {parent=root, loc=<0 -0.5301 98.0571>, rot=<0.0127 0.707 -0.0127 -0.707>}
 spine_01 {parent=pelvis, loc=<6.1208 -0.9328 0>, rot=<0 0 0.034 -0.9994>}
 spine_02 {parent=spine_01, loc=<6.5997 -0.2095 0>, rot=<0 0 0.0238 -0.9997>}
 spine_03 {parent=spine_02, loc=<6.4914 -0.1237 0>, rot=<0 0 0.0149 -0.9999>}
 spine_04 {parent=spine_03, loc=<6.9598 0.0794 0>, rot=<0 0 -0.0132 -0.9999>}
 spine_05 {parent=spine_04, loc=<6.9987 0 0>, rot=<0 0 -0.0015 -1>}
 neck_01 {parent=spine_05, loc=<11.3138 0.1529 0>, rot=<0 0 -0.2003 -0.9797>}
 neck_02 {parent=neck_01, loc=<4.5627 0 0>, rot=<0 0 0.0504 -0.9987>}
 head {parent=neck_02, loc=<5.3914 0 0>, rot=<0 0 0.115 -0.9934>}
 clavicle_l {parent=spine_05, loc=<5.4831 -0.1053 -3.2946>, rot=<0.0465 -0.7289 -0.009 -0.6829>}
 upperarm_l {parent=clavicle_l, loc=<9.8658 -1.2136 -0.4105>, rot=<0.0164 -0.3245 -0.0269 -0.9454>}
 lowerarm_l {parent=upperarm_l, loc=<26.465 -0.8345 0.259>, rot=<0.0036 0.0161 -0.2298 -0.9731>}
 hand_l {parent=lowerarm_l, loc=<22.7355 -0.3203 -0.6105>, rot=<-0.6784 -0.0223 0.0735 -0.7307>}
 thumb_01_l {parent=hand_l, loc=<1.3118 -0.3191 -2.459>, rot=<0.6546 -0.4042 0.0664 -0.6354>}
 thumb_02_l {parent=thumb_01_l, loc=<5.9336 -0.0448 0.0401>, rot=<-0.0037 0.0213 0.298 -0.9543>}
 thumb_03_l {parent=thumb_02_l, loc=<2.2223 0.3682 -0.0907>, rot=<0.0118 -0.013 0.0066 -0.9998>}
 index_metacarpal_l {parent=hand_l, loc=<1.6535 0.0183 -2.96>, rot=<0.0158 -0.0124 -0.0173 -0.9996>}
 index_01_l {parent=index_metacarpal_l, loc=<8.3008 0.0357 -0.0524>, rot=<-0.0044 0.0485 0.0924 -0.9945>}
 index_02_l {parent=index_01_l, loc=<3.3587 -0.3004 0.0452>, rot=<-0.002 0.0027 0.3225 -0.9466>}
 index_03_l {parent=index_02_l, loc=<2.5578 -0.123 -0.0252>, rot=<-0.002 0.0067 0.2558 -0.9667>}
 middle_metacarpal_l {parent=hand_l, loc=<2.1856 -0.1725 -0.5984>, rot=<0.006 0.0168 -0.0329 -0.9993>}
 middle_01_l {parent=middle_metacarpal_l, loc=<7.3275 -0.2447 0.0129>, rot=<-0.0208 0.0042 0.0858 -0.9961>}
 middle_02_l {parent=middle_01_l, loc=<4.0123 0.1401 0.1175>, rot=<-0.0218 0.0258 0.3512 -0.9357>}
 middle_03_l {parent=middle_02_l, loc=<2.6077 0.0534 0.1373>, rot=<0.0082 0.0188 0.1708 -0.9851>}
 ring_metacarpal_l {parent=hand_l, loc=<2.2074 -0.3625 1.3638>, rot=<-0.0284 0.0722 -0.0458 -0.9959>}
 ring_01_l {parent=ring_metacarpal_l, loc=<7.0295 -0.6197 -0.1446>, rot=<-0.0489 -0.0252 0.1892 -0.9804>}
 ring_02_l {parent=ring_01_l, loc=<3.6174 0.4647 -0.1027>, rot=<0 0 0.2806 -0.9598>}
 ring_03_l {parent=ring_02_l, loc=<2.4152 0.114 0.0199>, rot=<0 0 0.1189 -0.9929>}
 pinky_metacarpal_l {parent=hand_l, loc=<1.7616 -0.9753 3.3814>, rot=<-0.0397 0.0814 -0.069 -0.9935>}
 pinky_01_l {parent=pinky_metacarpal_l, loc=<7.1709 -0.4831 -0.1094>, rot=<-0.0654 -0.0035 0.2242 -0.9723>}
 pinky_02_l {parent=pinky_01_l, loc=<2.8274 0.0893 -0.1043>, rot=<0.0037 0.0011 0.2854 -0.9584>}
 pinky_03_l {parent=pinky_02_l, loc=<2.2046 0.0307 -0.0078>, rot=<-0.0002 0.0098 0.1445 -0.9895>}
 deform_wrist_dn_base_l {parent=hand_l, loc=<0 -1.5326 0>, rot=<0 0 0 -1>}
 deform_wrist_dn_l {parent=deform_wrist_dn_base_l, loc=<0 0 0>, rot=<0 0 0 -1>}
 deform_wrist_up_base_l {parent=hand_l, loc=<0 2.3009 0>, rot=<0 0 0 -1>}
 deform_wrist_up_l {parent=deform_wrist_up_base_l, loc=<0 0 0>, rot=<0 0 0 -1>}
 hand_attach_l {parent=hand_l, loc=<8.5 -3 0>, rot=<0.7071 0.7071 0 0>}
 weapon_l {parent=hand_l, loc=<-0.1974 -3.4111 0.3811>, rot=<0 0 0 1>}
 lowerarm_twist_02_l {parent=lowerarm_l, loc=<10.6707 0 0>, rot=<0 0 0 1>}
 lowerarm_twist_01_l {parent=lowerarm_l, loc=<18.316 0 0>, rot=<0 0 0 1>}
 deform_elbow_in_base_l {parent=lowerarm_l, loc=<0.1834 3.6062 -0.2572>, rot=<0 0 0 -1>}
 deform_elbow_in_l {parent=deform_elbow_in_base_l, loc=<0 0 0>, rot=<0 0 0 -1>}
 deform_elbow_out_base_l {parent=lowerarm_l, loc=<0 -4 0>, rot=<0 0 0 1>}
 deform_elbow_out_l {parent=deform_elbow_out_base_l, loc=<0 0 0>, rot=<0 0 0 1>}
 upperarm_twist_01_l {parent=upperarm_l, loc=<7.2135 -0.1134 0>, rot=<0 0 0 1>}
 deform_upperarm_bk_base_l {parent=upperarm_twist_01_l, loc=<-6.4053 -4.3746 0.314>, rot=<0 0 0 -1>}
 deform_upperarm_bk_l {parent=deform_upperarm_bk_base_l, loc=<0 0 0>, rot=<0 0 0 -1>}
 deform_upperarm_fr_base_l {parent=upperarm_twist_01_l, loc=<-5.6312 5.092 0.3354>, rot=<0 0 0 -1>}
 deform_upperarm_fr_l {parent=deform_upperarm_fr_base_l, loc=<0 0 0>, rot=<0 0 0 -1>}
 deform_upperarm_up_base_l {parent=upperarm_twist_01_l, loc=<-8.891 0.4825 3.9517>, rot=<0 0 0 -1>}
 deform_upperarm_up_l {parent=deform_upperarm_up_base_l, loc=<0 0 0>, rot=<0 0 0 -1>}
 upperarm_twist_02_l {parent=upperarm_l, loc=<19.7998 -0.1134 0>, rot=<0 0 0 1>}
 clavicle_r {parent=spine_05, loc=<5.4831 -0.1053 3.2946>, rot=<0.7289 0.0465 -0.6829 0.009>}
 upperarm_r {parent=clavicle_r, loc=<-9.8658 1.2136 0.4105>, rot=<0.0164 -0.3245 -0.0269 -0.9454>}
 lowerarm_r {parent=upperarm_r, loc=<-26.465 0.8345 -0.259>, rot=<0.0036 0.0161 -0.2298 -0.9731>}
 hand_r {parent=lowerarm_r, loc=<-22.7355 0.3203 0.6106>, rot=<-0.6784 -0.0223 0.0735 -0.7307>}
 thumb_01_r {parent=hand_r, loc=<-1.3118 0.3191 2.459>, rot=<0.6546 -0.4042 0.0664 -0.6354>}
 thumb_02_r {parent=thumb_01_r, loc=<-5.9336 0.0448 -0.0401>, rot=<-0.0037 0.0213 0.298 -0.9543>}
 thumb_03_r {parent=thumb_02_r, loc=<-2.2223 -0.3682 0.0907>, rot=<0.0118 -0.013 0.0066 -0.9998>}
 index_metacarpal_r {parent=hand_r, loc=<-1.6535 -0.0183 2.96>, rot=<0.0158 -0.0124 -0.0173 -0.9996>}
 index_01_r {parent=index_metacarpal_r, loc=<-8.3008 -0.0357 0.0524>, rot=<-0.0044 0.0485 0.0924 -0.9945>}
 index_02_r {parent=index_01_r, loc=<-3.3587 0.3004 -0.0452>, rot=<-0.002 0.0027 0.3225 -0.9466>}
 index_03_r {parent=index_02_r, loc=<-2.5578 0.123 0.0252>, rot=<-0.002 0.0067 0.2558 -0.9667>}
 middle_metacarpal_r {parent=hand_r, loc=<-2.1856 0.1725 0.5984>, rot=<0.006 0.0168 -0.0329 -0.9993>}
 middle_01_r {parent=middle_metacarpal_r, loc=<-7.3275 0.2447 -0.0129>, rot=<-0.0208 0.0042 0.0858 -0.9961>}
 middle_02_r {parent=middle_01_r, loc=<-4.0123 -0.1401 -0.1175>, rot=<-0.0218 0.0258 0.3512 -0.9357>}
 middle_03_r {parent=middle_02_r, loc=<-2.6077 -0.0534 -0.1373>, rot=<0.0082 0.0188 0.1708 -0.9851>}
 ring_metacarpal_r {parent=hand_r, loc=<-2.2074 0.3625 -1.3638>, rot=<-0.0284 0.0722 -0.0458 -0.9959>}
 ring_01_r {parent=ring_metacarpal_r, loc=<-7.0295 0.6197 0.1446>, rot=<-0.0489 -0.0252 0.1892 -0.9804>}
 ring_02_r {parent=ring_01_r, loc=<-3.6174 -0.4647 0.1027>, rot=<0 0 0.2806 -0.9598>}
 ring_03_r {parent=ring_02_r, loc=<-2.4152 -0.114 -0.0199>, rot=<0 0 0.1189 -0.9929>}
 pinky_metacarpal_r {parent=hand_r, loc=<-1.7616 0.9753 -3.3814>, rot=<-0.0397 0.0814 -0.069 -0.9935>}
 pinky_01_r {parent=pinky_metacarpal_r, loc=<-7.1709 0.4831 0.1094>, rot=<-0.0654 -0.0035 0.2242 -0.9723>}
 pinky_02_r {parent=pinky_01_r, loc=<-2.8274 -0.0893 0.1043>, rot=<0.0037 0.0011 0.2854 -0.9584>}
 pinky_03_r {parent=pinky_02_r, loc=<-2.2046 -0.0307 0.0078>, rot=<-0.0002 0.0098 0.1445 -0.9895>}
 deform_wrist_dn_base_r {parent=hand_r, loc=<-0.0001 1.5328 0>, rot=<0 0 0 -1>}
 deform_wrist_dn_r {parent=deform_wrist_dn_base_r, loc=<0 0 0>, rot=<0 0 0 -1>}
 deform_wrist_up_base_r {parent=hand_r, loc=<-0.0002 -2.3007 -0.0001>, rot=<0 0 0 -1>}
 deform_wrist_up_r {parent=deform_wrist_up_base_r, loc=<0 0 0>, rot=<0 0 0 -1>}
 weapon_r {parent=hand_r, loc=<0.1974 3.4111 -0.3811>, rot=<0 0 0 1>}
 hand_attach_r {parent=hand_r, loc=<-8.5001 2.9999 0>, rot=<0 0 -0.7071 -0.7071>}
 lowerarm_twist_02_r {parent=lowerarm_r, loc=<-10.6707 0 0>, rot=<0 0 0 -1>}
 lowerarm_twist_01_r {parent=lowerarm_r, loc=<-18.316 0 0>, rot=<0 0 0 -1>}
 deform_elbow_in_base_r {parent=lowerarm_r, loc=<-0.1834 -3.6062 0.2572>, rot=<0 0 0 -1>}
 deform_elbow_in_r {parent=deform_elbow_in_base_r, loc=<0 0 0>, rot=<0 0 0 -1>}
 deform_elbow_out_base_r {parent=lowerarm_r, loc=<0 4 -0.0001>, rot=<0 0 0 -1>}
 deform_elbow_out_r {parent=deform_elbow_out_base_r, loc=<0 0 0>, rot=<0 0 0 1>}
 upperarm_twist_01_r {parent=upperarm_r, loc=<-7.2135 0.1134 0>, rot=<0 0 0 -1>}
 deform_upperarm_bk_base_r {parent=upperarm_twist_01_r, loc=<6.4053 4.3746 -0.314>, rot=<0 0 0 -1>}
 deform_upperarm_bk_r {parent=deform_upperarm_bk_base_r, loc=<0 0 0>, rot=<0 0 0 -1>}
 deform_upperarm_fr_base_r {parent=upperarm_twist_01_r, loc=<5.6312 -5.092 -0.3354>, rot=<0 0 0 -1>}
 deform_upperarm_fr_r {parent=deform_upperarm_fr_base_r, loc=<0 0 0>, rot=<0 0 0 -1>}
 deform_upperarm_up_base_r {parent=upperarm_twist_01_r, loc=<8.891 -0.4825 -3.9517>, rot=<0 0 0 -1>}
 deform_upperarm_up_r {parent=deform_upperarm_up_base_r, loc=<0 0 0>, rot=<0 0 0 -1>}
 upperarm_twist_02_r {parent=upperarm_r, loc=<-19.7998 0.1134 0>, rot=<0 0 0 -1>}
 attach_backpack {parent=spine_05, loc=<-0.123 -13.473 0>, rot=<-0.0093 -0.707 0.0093 -0.707>}
 Backpack_BR {parent=attach_backpack, type=Null, loc=<0 0 0>, rot=<0 0 0 1>}
 attach_cape {parent=spine_05, loc=<9.017 -11.714 0>, rot=<0.0616 -0.7044 -0.0616 -0.7044>}
 cape {parent=attach_cape, type=Null, loc=<0 0 0>, rot=<0 0 0 1>}
 deform_pec_aim_base_l {parent=spine_04, loc=<3.2456 9.3309 -1.2237>, rot=<0.0603 -0.4033 0.3318 -0.8507>}
 deform_pec_aim_l {parent=deform_pec_aim_base_l, loc=<0 0 0>, rot=<0 0 0 1>}
 deform_pec_base_l {parent=deform_pec_aim_l, loc=<9.3665 0 0>, rot=<0 0 0 1>}
 deform_pec_l {parent=deform_pec_base_l, loc=<0 0 0>, rot=<0 0 0 -1>}
 deform_pec_aim_base_r {parent=spine_04, loc=<3.2452 9.3309 1.2237>, rot=<-0.4033 -0.0603 0.8507 0.3318>}
 deform_pec_aim_r {parent=deform_pec_aim_base_r, loc=<0 0 0>, rot=<0 0 0 1>}
 deform_pec_base_r {parent=deform_pec_aim_r, loc=<-9.3667 -0.0003 -0.0004>, rot=<0 0 0 1>}
 deform_pec_r {parent=deform_pec_base_r, loc=<0 0 0>, rot=<0 0 0 -1>}
 def_belt_back_corr_r {parent=spine_04, loc=<3.8936 -7.1353 10.3371>, rot=<-0.21 0.4179 -0.7188 -0.5143>}
 def_belt_back_corr_l {parent=spine_04, loc=<3.894 -7.1353 -10.3371>, rot=<0.21 -0.4179 -0.7188 -0.5143>}
 thigh_l {parent=pelvis, loc=<-1.0855 -0.2109 -8.7053>, rot=<0.0009 0.0355 0.0155 -0.9993>}
 calf_l {parent=thigh_l, loc=<-42.5284 -0.0019 -0.7297>, rot=<0.0006 -0.0161 -0.0391 -0.9991>}
 foot_l {parent=calf_l, loc=<-46.7401 -2.5014 -0.592>, rot=<0.0427 -0.0238 0.0353 -0.9982>}
 ball_l {parent=foot_l, loc=<-3.5658 10.8066 -0.1544>, rot=<0.0001 -0.0001 -0.7063 -0.7079>}
 calf_twist_02_l {parent=calf_l, loc=<-19.1546 -1.7943 -0.2638>, rot=<0.0237 0.0095 -0.0035 -0.9997>}
 calf_twist_01_l {parent=calf_l, loc=<-35.8483 -1.9197 -0.5794>, rot=<0.0237 0.0095 -0.0035 -0.9997>}
 def_corr_foot_l {parent=calf_twist_01_l, loc=<-8.341 -3.1583 0.2514>, rot=<0.694 -0.7193 -0.0205 0.0236>}
 deform_knee_in_base_l {parent=calf_l, loc=<0.8031 -5.0231 0.2456>, rot=<0 0 0 -1>}
 deform_knee_in_l {parent=deform_knee_in_base_l, loc=<0 0 0>, rot=<0 0 0 -1>}
 deform_knee_out_base_l {parent=calf_l, loc=<0 4 0>, rot=<0 0 0 -1>}
 deform_knee_out_l {parent=deform_knee_out_base_l, loc=<0 0 0>, rot=<0 0 0 1>}
 thigh_twist_01_l {parent=thigh_l, loc=<-27.952 0 0>, rot=<0.0013 0 0 -1>}
 deform_groin_base_l {parent=thigh_l, loc=<0.4801 5.6776 -0.2486>, rot=<0 0 0 -1>}
 deform_groin_l {parent=deform_groin_base_l, loc=<0 0 0>, rot=<0 0 0 -1>}
 thigh_r {parent=pelvis, loc=<-1.0855 -0.2109 8.7053>, rot=<0.0355 -0.0009 0.9993 0.0155>}
 calf_r {parent=thigh_r, loc=<42.5284 0.0019 0.7297>, rot=<0.0006 -0.0161 -0.0391 -0.9991>}
 foot_r {parent=calf_r, loc=<46.7401 2.5014 0.592>, rot=<0.0427 -0.0238 0.0353 -0.9982>}
 ball_r {parent=foot_r, loc=<3.5658 -10.8066 0.1544>, rot=<0.0001 -0.0001 -0.7063 -0.7079>}
 calf_twist_02_r {parent=calf_r, loc=<19.1546 1.7943 0.2638>, rot=<0.0237 0.0095 -0.0035 -0.9997>}
 calf_twist_01_r {parent=calf_r, loc=<35.8483 1.9197 0.5794>, rot=<0.0237 0.0095 -0.0035 -0.9997>}
 def_corr_foot_r {parent=calf_twist_01_r, loc=<8.3411 3.1583 -0.2514>, rot=<0.7193 0.694 -0.0236 -0.0205>}
 deform_knee_in_base_r {parent=calf_r, loc=<-0.8031 5.0231 -0.2457>, rot=<0 0 0 -1>}
 deform_knee_in_r {parent=deform_knee_in_base_r, loc=<0 0 0>, rot=<0 0 0 -1>}
 deform_knee_out_base_r {parent=calf_r, loc=<0 -4 0>, rot=<0 0 0 -1>}
 deform_knee_out_r {parent=deform_knee_out_base_r, loc=<0 0 0>, rot=<0 0 0 1>}
 thigh_twist_01_r {parent=thigh_r, loc=<27.952 0 0>, rot=<0.0013 0 0 -1>}
 deform_groin_base_r {parent=thigh_r, loc=<-0.4801 -5.6776 0.2486>, rot=<0 0 0 -1>}
 deform_groin_r {parent=deform_groin_base_r, loc=<0 0 0>, rot=<0 0 0 -1>}
 deform_glute_base_l {parent=pelvis, loc=<-7.3466 -9.0213 -9.1245>, rot=<0.0172 0.031 -0.4477 -0.8935>}
 deform_glute_l {parent=deform_glute_base_l, loc=<0 0 0>, rot=<0 0 0 1>}
 deform_glute_base_r {parent=pelvis, loc=<-7.3464 -9.0213 9.1245>, rot=<0.031 -0.0172 0.8935 -0.4477>}
 deform_glute_r {parent=deform_glute_base_r, loc=<0 0 0>, rot=<0 0 0 1>}
 MSTR_skirt_fr_l {parent=pelvis, loc=<3.4185 6.4886 -7.1306>, rot=<-0.0519 -0.0034 0.9936 0.1002>}
 dyn_skirt_fr_01_l {parent=MSTR_skirt_fr_l, loc=<0 0 0>, rot=<0 0 0 1>}
 dyn_skirt_fr_02_l {parent=dyn_skirt_fr_01_l, loc=<8.831 0 0>, rot=<-0.0009 0.0059 -0.0057 -1>}
 dyn_skirt_fr_03_l {parent=dyn_skirt_fr_02_l, loc=<7.5311 0 0>, rot=<-0.0006 0.0041 -0.0138 -0.9999>}
 dyn_skirt_fr_end_l {parent=dyn_skirt_fr_03_l, loc=<8.8071 0 0>, rot=<0 0 0 1>}
 MSTR_skirt_side_l {parent=pelvis, loc=<3.9627 0.0369 -13.2192>, rot=<0.1027 -0.6996 -0.6955 -0.1277>}
 dyn_skirt_side_01_l {parent=MSTR_skirt_side_l, loc=<0 0 0>, rot=<0 0 0 1>}
 dyn_skirt_side_02_l {parent=dyn_skirt_side_01_l, loc=<8.8467 0 0>, rot=<0 0 -0.0497 -0.9988>}
 dyn_skirt_side_03_l {parent=dyn_skirt_side_02_l, loc=<7.4794 0.0001 0>, rot=<0 0 -0.0534 -0.9986>}
 dyn_skirt_side_end_l {parent=dyn_skirt_side_03_l, loc=<8.6348 0 0>, rot=<0 0 0 1>}
 MSTR_skirt_bk_l {parent=pelvis, loc=<2.7681 -9.9533 -8.8424>, rot=<0.1119 -0.9793 -0.1653 -0.0326>}
 dyn_skirt_bk_01_l {parent=MSTR_skirt_bk_l, loc=<0 0 0>, rot=<0 0 0 1>}
 dyn_skirt_bk_02_l {parent=dyn_skirt_bk_01_l, loc=<8.8392 0.0002 0>, rot=<0.0007 0.0226 -0.0751 -0.9969>}
 dyn_skirt_bk_03_l {parent=dyn_skirt_bk_02_l, loc=<7.2774 0 0>, rot=<0.0004 -0.0049 0.0004 -1>}
 dyn_skirt_bk_end_l {parent=dyn_skirt_bk_03_l, loc=<8.5006 0 0>, rot=<0 0 0 1>}
 MSTR_skirt_fr {parent=pelvis, loc=<2.6793 6.515 0.0569>, rot=<-0.011 -0.0007 0.9949 0.0998>}
 dyn_skirt_fr_01 {parent=MSTR_skirt_fr, loc=<0 0 0>, rot=<0 0 0 1>}
 dyn_skirt_fr_02 {parent=dyn_skirt_fr_01, loc=<8.8868 0 0>, rot=<-0.0038 0.0238 -0.0053 -0.9997>}
 dyn_skirt_fr_03 {parent=dyn_skirt_fr_02, loc=<7.5431 0 0>, rot=<0.0023 -0.0163 -0.0134 -0.9998>}
 dyn_skirt_fr_end {parent=dyn_skirt_fr_03, loc=<8.7766 0 0>, rot=<0 0 0 1>}
 MSTR_skirt_bk_r {parent=pelvis, loc=<2.7676 -9.9533 8.8424>, rot=<0.1119 -0.9793 0.1653 0.0326>}
 dyn_skirt_bk_01_r {parent=MSTR_skirt_bk_r, loc=<0 0 0>, rot=<0 0 0 1>}
 dyn_skirt_bk_02_r {parent=dyn_skirt_bk_01_r, loc=<8.8387 0 0>, rot=<-0.0007 -0.0226 -0.0751 -0.9969>}
 dyn_skirt_bk_03_r {parent=dyn_skirt_bk_02_r, loc=<7.2774 0 0>, rot=<-0.0004 0.0049 0.0004 -1>}
 dyn_skirt_bk_end_r {parent=dyn_skirt_bk_03_r, loc=<8.5006 0 0>, rot=<0 0 0 1>}
 MSTR_skirt_bk {parent=pelvis, loc=<2.3193 -10.6909 -0.0546>, rot=<0.1053 -0.9944 0.0005 -0.0033>}
 dyn_skirt_bk_01 {parent=MSTR_skirt_bk, loc=<0 0 0>, rot=<0 0 0 1>}
 dyn_skirt_bk_02 {parent=dyn_skirt_bk_01, loc=<8.7938 0 0>, rot=<0.0018 -0.0109 -0.0788 -0.9968>}
 dyn_skirt_bk_03 {parent=dyn_skirt_bk_02, loc=<7.2456 0 0>, rot=<-0.0007 0.0077 0.002 -1>}
 dyn_skirt_bk_end {parent=dyn_skirt_bk_03, loc=<8.4692 0 0>, rot=<0 0 0 1>}
 MSTR_skirt_fr_r {parent=pelvis, loc=<3.4183 6.4886 7.1306>, rot=<0.0519 0.0034 0.9936 0.1002>}
 dyn_skirt_fr_01_r {parent=MSTR_skirt_fr_r, loc=<0 0 0>, rot=<0 0 0 1>}
 dyn_skirt_fr_02_r {parent=dyn_skirt_fr_01_r, loc=<8.8308 0 0>, rot=<0.0009 -0.0059 -0.0057 -1>}
 dyn_skirt_fr_03_r {parent=dyn_skirt_fr_02_r, loc=<7.531 0 0>, rot=<0.0006 -0.0041 -0.0138 -0.9999>}
 dyn_skirt_fr_end_r {parent=dyn_skirt_fr_03_r, loc=<8.807 0 0>, rot=<0 0 0 1>}
 MSTR_skirt_side_r {parent=pelvis, loc=<3.9625 0.0369 13.2192>, rot=<0.1027 -0.6996 0.6955 0.1277>}
 dyn_skirt_side_01_r {parent=MSTR_skirt_side_r, loc=<0 0 0>, rot=<0 0 0 1>}
 dyn_skirt_side_02_r {parent=dyn_skirt_side_01_r, loc=<8.8465 0 0>, rot=<0 0 -0.0497 -0.9988>}
 dyn_skirt_side_03_r {parent=dyn_skirt_side_02_r, loc=<7.4793 0 0>, rot=<0 0 -0.0534 -0.9986>}
 dyn_skirt_side_end_r {parent=dyn_skirt_side_03_r, loc=<8.6348 0 0>, rot=<0 0 0 1>}
 def_belt_fr_r {parent=pelvis, loc=<3.7053 8.4507 7.9154>, rot=<0.5899 -0.6633 0.3329 0.3182>}
 dart_01 {parent=def_belt_fr_r, loc=<2.6705 0.5504 -1.13>, rot=<-0.0346 -0.1994 -0.0885 -0.9753>}
 dart_02 {parent=def_belt_fr_r, loc=<-0.499 0.3941 -0.8066>, rot=<-0.0346 -0.1994 -0.0885 -0.9753>}
 dart_03 {parent=def_belt_fr_r, loc=<-3.3991 -0.0392 0.2944>, rot=<-0.0346 -0.1994 -0.0885 -0.9753>}
 def_belt_fr_r_ctrl {controlType=EULER_TRANSFORM, parent=def_belt_fr_r, type=Control, loc=<0 0 0>, rot=<0 0 0 1>}
 def_belt_fr_rear_r {parent=pelvis, loc=<3.1055 5.0791 12.7304>, rot=<0.7372 -0.6593 0.0102 0.1476>}
 def_belt_fr_rear_r_ctrl {controlType=EULER_TRANSFORM, parent=def_belt_fr_rear_r, type=Control, loc=<0 0 0>, rot=<0 0 0 1>}
 def_belt_fr {parent=pelvis, loc=<5.1471 9.4979 0.4172>, rot=<0.4734 -0.5903 0.4951 0.4271>}
 def_belt_fr_l {parent=pelvis, loc=<7.4249 7.7886 -7.333>, rot=<0.3683 -0.5115 0.6134 0.4759>}
 def_belt_side_l {parent=pelvis, loc=<9.285 2.7285 -10.9903>, rot=<-0.1526 0.2881 -0.7482 -0.5779>}
 dyn_bag_01 {parent=def_belt_side_l, loc=<0.3427 4.0452 -0.5124>, rot=<0.6905 0.6712 -0.2686 0.0236>}
 dyn_bag_02 {parent=dyn_bag_01, loc=<4.808 -0.3159 -1.1488>, rot=<0 -0.0852 0.015 -0.9963>}
 MSTR_skirt_bk_r_ctrl {controlType=EULER_TRANSFORM, parent=pelvis, type=Control, loc=<0 0 0>, rot=<0 0 0 1>}
 MSTR_skirt_side_r_ctrl {controlType=EULER_TRANSFORM, parent=pelvis, type=Control, loc=<0 0 0>, rot=<0 0 0 1>}
 dart {parent=pelvis, type=Null, loc=<3.7053 8.4507 7.9154>, rot=<0.5899 -0.6633 0.3329 0.3182>}
 pelvis_null {parent=pelvis, type=Null, loc=<2.5 0 0>, rot=<0 0 0 1>}
]
variables [
 isFrontEndIdle {type=bool, default=False}
 isSkydiving {type=double, default=0}
 isSliding {type=bool, default=False}
 isSlidingBaseball {type=bool, default=False}
]
