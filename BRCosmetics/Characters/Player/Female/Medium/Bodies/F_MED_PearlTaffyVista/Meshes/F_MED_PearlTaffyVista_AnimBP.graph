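# Post-process-style AnimBlueprint for the PearlTaffyVista female cosmetic body, adding secondary motion to skirt, bag and dart bones.
# It copies the attached parent's pose, applies a Control Rig fed slide, skydive and front-end idle states, then back-skirt bone transforms.
# Chained spring-based AnimDynamics nodes follow; dyn_skirt_side_01_l is left unconnected.
# The event graph computes gravity override, jump rotation and front-end-dependent sim-space settings, though graph pins show constants.
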AnimBlueprint /BRCosmetics/Characters/Player/Female/Medium/Bodies/F_MED_PearlTaffyVista/Meshes/F_MED_PearlTaffyVista_AnimBP
target_skeleton=&/F_MED_PearlTaffyVista_Skeleton
graphs [
 AnimGraph {nodes [
   Anim Dynamics - Bone: dyn_skirt_side_01_l!(ComponentPose, Alpha=1, GravityOverride:Vector="0,0,-980", ExternalForce="0,0,0") -> (Pose) {
    angularDampingOverride=0.99
    angularSpringConstant=300
    bAngularSpring=true
    bOverrideAngularDamping=true
    bUseGravityOverride=true
    simulationSpace=BoneRelative
    boundBone {boneName=dyn_skirt_side_01_l}
    componentAppliedLinearAccClamp=<2000 2000 2000>
    componentLinearAccScale=<0.3 0.3 0.3>
    componentLinearVelScale=<0.3 0.3 0.5>
    gravityOverride {z=-980}
    physicsBodyDefinitions [{
      boundBone {boneName=dyn_skirt_side_01_l}
      boxExtents=<10 10 10>
      constraintSetup {angularLimitsMax=<0 0 0>, angularLimitsMin=<0 0 0>, angularTarget=<1 0 0>}
      localJointOffset=<0.3 0 0>
     }]
    relativeSpaceBone {boneName=root}
    simSpaceSettings {maxAngularAcceleration=20, maxAngularVelocity=4, simSpaceAngularAlpha=0.3}
   }
   n1 = Control Rig!(
    Source=Copy Pose From Mesh() {bUseAttachedParent=true}
    Alpha
    isSlidingBaseball:bool=Get bIsBaseballSliding()
    isSliding:bool=Get bIsKneeSliding()
    isSkydiving:double=Get isSkydiving() -> (isSkydiving:double)
    isFrontEndIdle:bool=Get isFrontEndIdle() -> (isFrontEndIdle:bool)
   ) -> (Pose) {controlRigAssetReference {blueprintRigClass=&/F_MED_PearlTaffyVista_CtrlRig.F_MED_PearlTaffyVista_CtrlRig_C}}
   n2 = "Transform (Modify) Bone - Bone: MSTR_skirt_bk"(ComponentPose=Local To Component(LocalPose=@n1), Alpha=1, Translation="0,0,0", Rotation="0,0,0", Scale="1,1,1") {rotationMode=BMM_Additive, rotationSpace=BCS_ParentBoneSpace, boneToModify {boneName=MSTR_skirt_bk}}
   n3 = Anim Dynamics - Bone: dyn_skirt_fr_01!(
    ComponentPose="Transform (Modify) Bone - Bone: MSTR_skirt_bk_l"(
     ComponentPose="Transform (Modify) Bone - Bone: MSTR_skirt_bk_r"(ComponentPose=@n2, Alpha=1, Translation="0,0,0", Rotation="0,0,0", Scale="1,1,1") {rotationMode=BMM_Additive, rotationSpace=BCS_ParentBoneSpace, boneToModify {boneName=MSTR_skirt_bk_r}}
     Alpha=1
     Translation="0,0,0"
     Rotation="0,0,0"
     Scale="1,1,1"
    ) {rotationMode=BMM_Additive, rotationSpace=BCS_ParentBoneSpace, boneToModify {boneName=MSTR_skirt_bk_l}}
    Alpha=1
    GravityOverride:Vector="0,0,-980"
    SimSpaceSettings:Anim Phys Sim Space Settings Structure=(SimSpaceAngularAlpha=0.3,MaxAngularVelocity=4,MaxAngularAcceleration=20,ExternalAngularVelocity=(X=0,Y=0,Z=0))
    ExternalForce="0,0,0"
   ) -> (Pose) {
    angularDampingOverride=0.99
    angularSpringConstant=300
    bAngularSpring=true
    bOverrideAngularDamping=true
    bUseGravityOverride=true
    simulationSpace=BoneRelative
    boundBone {boneName=dyn_skirt_fr_01}
    componentAppliedLinearAccClamp=<3000 3000 3000>
    componentLinearAccScale=<0.3 0.3 0.3>
    componentLinearVelScale=<0.3 0.3 0.5>
    gravityOverride {z=-980}
    physicsBodyDefinitions [{
      boundBone {boneName=dyn_skirt_fr_01}
      boxExtents=<10 10 10>
      constraintSetup {angularLimitsMax=<0 0 0>, angularLimitsMin=<0 0 -5>, angularTarget=<1 0 0>}
      localJointOffset=<0.5 0 0>
     }]
    relativeSpaceBone {boneName=root}
    simSpaceSettings {maxAngularAcceleration=20, maxAngularVelocity=4, simSpaceAngularAlpha=0.3}
   }
   n4 = Anim Dynamics - Bone: dyn_skirt_fr_02!(
    ComponentPose=@n3
    Alpha=1
    GravityOverride:Vector="0,0,-980"
    SimSpaceSettings:Anim Phys Sim Space Settings Structure=(SimSpaceAngularAlpha=0.3,MaxAngularVelocity=4,MaxAngularAcceleration=20,ExternalAngularVelocity=(X=0,Y=0,Z=0))
    ExternalForce="0,0,0"
   ) -> (Pose) {
    angularDampingOverride=0.99
    angularSpringConstant=300
    bAngularSpring=true
    bOverrideAngularDamping=true
    bUseGravityOverride=true
    simulationSpace=BoneRelative
    boundBone {boneName=dyn_skirt_fr_02}
    componentAppliedLinearAccClamp=<3000 3000 3000>
    componentLinearAccScale=<0.3 0.3 0.4>
    componentLinearVelScale=<0.3 0.3 0.6>
    gravityOverride {z=-980}
    physicsBodyDefinitions [{
      boundBone {boneName=dyn_skirt_fr_02}
      boxExtents=<10 10 10>
      constraintSetup {angularLimitsMax=<0 0 0>, angularLimitsMin=<0 0 -15>, angularTarget=<1 0 0>}
      localJointOffset=<0.5 0 0>
     }]
    relativeSpaceBone {boneName=root}
    simSpaceSettings {maxAngularAcceleration=20, maxAngularVelocity=4, simSpaceAngularAlpha=0.3}
   }
   n5 = Anim Dynamics - Bone: dyn_skirt_fr_03!(
    ComponentPose=@n4
    Alpha=1
    GravityOverride:Vector="0,0,-980"
    SimSpaceSettings:Anim Phys Sim Space Settings Structure=(SimSpaceAngularAlpha=0.3,MaxAngularVelocity=4,MaxAngularAcceleration=20,ExternalAngularVelocity=(X=0,Y=0,Z=0))
    ExternalForce="0,0,0"
   ) -> (Pose) {
    angularDampingOverride=0.99
    angularSpringConstant=300
    bAngularSpring=true
    bOverrideAngularDamping=true
    bUseGravityOverride=true
    simulationSpace=BoneRelative
    boundBone {boneName=dyn_skirt_fr_03}
    componentAppliedLinearAccClamp=<3000 3000 3000>
    componentLinearAccScale=<0.3 0.3 0.4>
    componentLinearVelScale=<0.3 0.3 0.6>
    gravityOverride {z=-980}
    physicsBodyDefinitions [{
      boundBone {boneName=dyn_skirt_fr_03}
      boxExtents=<10 10 10>
      constraintSetup {angularLimitsMax=<0 0 0>, angularLimitsMin=<0 0 -30>, angularTarget=<1 0 0>}
      localJointOffset=<0.5 0 0>
     }]
    relativeSpaceBone {boneName=root}
    simSpaceSettings {maxAngularAcceleration=20, maxAngularVelocity=4, simSpaceAngularAlpha=0.3}
   }
   n6 = Anim Dynamics - Bone: dyn_skirt_fr_01_r!(
    ComponentPose=@n5
    Alpha=1
    GravityOverride:Vector="0,0,-980"
    SimSpaceSettings:Anim Phys Sim Space Settings Structure=(SimSpaceAngularAlpha=0.3,MaxAngularVelocity=4,MaxAngularAcceleration=20,ExternalAngularVelocity=(X=0,Y=0,Z=0))
    ExternalForce="0,0,0"
   ) -> (Pose) {
    angularDampingOverride=0.99
    angularSpringConstant=300
    bAngularSpring=true
    bOverrideAngularDamping=true
    bUseGravityOverride=true
    simulationSpace=BoneRelative
    boundBone {boneName=dyn_skirt_fr_01_r}
    componentAppliedLinearAccClamp=<3000 3000 3000>
    componentLinearAccScale=<0.3 0.3 0.3>
    componentLinearVelScale=<0.3 0.3 0.5>
    gravityOverride {z=-980}
    physicsBodyDefinitions [{
      boundBone {boneName=dyn_skirt_fr_01_r}
      boxExtents=<10 10 10>
      constraintSetup {angularLimitsMax=<0 0 0>, angularLimitsMin=<0 0 -5>, angularTarget=<1 0 0>}
      localJointOffset=<0.5 0 0>
     }]
    relativeSpaceBone {boneName=root}
    simSpaceSettings {maxAngularAcceleration=20, maxAngularVelocity=4, simSpaceAngularAlpha=0.3}
   }
   n7 = Anim Dynamics - Bone: dyn_skirt_fr_02_r!(
    ComponentPose=@n6
    Alpha=1
    GravityOverride:Vector="0,0,-980"
    SimSpaceSettings:Anim Phys Sim Space Settings Structure=(SimSpaceAngularAlpha=0.3,MaxAngularVelocity=4,MaxAngularAcceleration=20,ExternalAngularVelocity=(X=0,Y=0,Z=0))
    ExternalForce="0,0,0"
   ) -> (Pose) {
    angularDampingOverride=0.99
    angularSpringConstant=300
    bAngularSpring=true
    bOverrideAngularDamping=true
    bUseGravityOverride=true
    simulationSpace=BoneRelative
    boundBone {boneName=dyn_skirt_fr_02_r}
    componentAppliedLinearAccClamp=<3000 3000 3000>
    componentLinearAccScale=<0.3 0.3 0.4>
    componentLinearVelScale=<0.3 0.3 0.6>
    gravityOverride {z=-980}
    physicsBodyDefinitions [{
      boundBone {boneName=dyn_skirt_fr_02_r}
      boxExtents=<10 10 10>
      constraintSetup {angularLimitsMax=<0 0 0>, angularLimitsMin=<0 0 -15>, angularTarget=<1 0 0>}
      localJointOffset=<0.5 0 0>
     }]
    relativeSpaceBone {boneName=root}
    simSpaceSettings {maxAngularAcceleration=20, maxAngularVelocity=4, simSpaceAngularAlpha=0.3}
   }
   n8 = Anim Dynamics - Bone: dyn_skirt_fr_03_r!(
    ComponentPose=@n7
    Alpha=1
    GravityOverride:Vector="0,0,-980"
    SimSpaceSettings:Anim Phys Sim Space Settings Structure=(SimSpaceAngularAlpha=0.3,MaxAngularVelocity=4,MaxAngularAcceleration=20,ExternalAngularVelocity=(X=0,Y=0,Z=0))
    ExternalForce="0,0,0"
   ) -> (Pose) {
    angularDampingOverride=0.99
    angularSpringConstant=300
    bAngularSpring=true
    bOverrideAngularDamping=true
    bUseGravityOverride=true
    simulationSpace=BoneRelative
    boundBone {boneName=dyn_skirt_fr_03_r}
    componentAppliedLinearAccClamp=<3000 3000 3000>
    componentLinearAccScale=<0.3 0.3 0.4>
    componentLinearVelScale=<0.3 0.3 0.6>
    gravityOverride {z=-980}
    physicsBodyDefinitions [{
      boundBone {boneName=dyn_skirt_fr_03_r}
      boxExtents=<10 10 10>
      constraintSetup {angularLimitsMax=<0 0 0>, angularLimitsMin=<0 0 -30>, angularTarget=<1 0 0>}
      localJointOffset=<0.5 0 0>
     }]
    relativeSpaceBone {boneName=root}
    simSpaceSettings {maxAngularAcceleration=20, maxAngularVelocity=4, simSpaceAngularAlpha=0.3}
   }
   n9 = Anim Dynamics - Bone: dyn_skirt_fr_01_l!(
    ComponentPose=@n8
    Alpha=1
    GravityOverride:Vector="0,0,-980"
    SimSpaceSettings:Anim Phys Sim Space Settings Structure=(SimSpaceAngularAlpha=0.3,MaxAngularVelocity=4,MaxAngularAcceleration=20,ExternalAngularVelocity=(X=0,Y=0,Z=0))
    ExternalForce="0,0,0"
   ) -> (Pose) {
    angularDampingOverride=0.99
    angularSpringConstant=300
    bAngularSpring=true
    bOverrideAngularDamping=true
    bUseGravityOverride=true
    simulationSpace=BoneRelative
    boundBone {boneName=dyn_skirt_fr_01_l}
    componentAppliedLinearAccClamp=<3000 3000 3000>
    componentLinearAccScale=<0.3 0.3 0.3>
    componentLinearVelScale=<0.3 0.3 0.5>
    gravityOverride {z=-980}
    physicsBodyDefinitions [{
      boundBone {boneName=dyn_skirt_fr_01_l}
      boxExtents=<10 10 10>
      constraintSetup {angularLimitsMax=<0 0 0>, angularLimitsMin=<0 0 -5>, angularTarget=<1 0 0>}
      localJointOffset=<0.5 0 0>
     }]
    relativeSpaceBone {boneName=root}
    simSpaceSettings {maxAngularAcceleration=20, maxAngularVelocity=4, simSpaceAngularAlpha=0.3}
   }
   n10 = Anim Dynamics - Bone: dyn_skirt_fr_02_l!(
    ComponentPose=@n9
    Alpha=1
    GravityOverride:Vector="0,0,-980"
    SimSpaceSettings:Anim Phys Sim Space Settings Structure=(SimSpaceAngularAlpha=0.3,MaxAngularVelocity=4,MaxAngularAcceleration=20,ExternalAngularVelocity=(X=0,Y=0,Z=0))
    ExternalForce="0,0,0"
   ) -> (Pose) {
    angularDampingOverride=0.99
    angularSpringConstant=300
    bAngularSpring=true
    bOverrideAngularDamping=true
    bUseGravityOverride=true
    simulationSpace=BoneRelative
    boundBone {boneName=dyn_skirt_fr_02_l}
    componentAppliedLinearAccClamp=<3000 3000 3000>
    componentLinearAccScale=<0.3 0.3 0.4>
    componentLinearVelScale=<0.3 0.3 0.6>
    gravityOverride {z=-980}
    physicsBodyDefinitions [{
      boundBone {boneName=dyn_skirt_fr_02_l}
      boxExtents=<10 10 10>
      constraintSetup {angularLimitsMax=<0 0 0>, angularLimitsMin=<0 0 -15>, angularTarget=<1 0 0>}
      localJointOffset=<0.5 0 0>
     }]
    relativeSpaceBone {boneName=root}
    simSpaceSettings {maxAngularAcceleration=20, maxAngularVelocity=4, simSpaceAngularAlpha=0.3}
   }
   n11 = Anim Dynamics - Bone: dyn_skirt_fr_03_l!(
    ComponentPose=@n10
    Alpha=1
    GravityOverride:Vector="0,0,-980"
    SimSpaceSettings:Anim Phys Sim Space Settings Structure=(SimSpaceAngularAlpha=0.3,MaxAngularVelocity=4,MaxAngularAcceleration=20,ExternalAngularVelocity=(X=0,Y=0,Z=0))
    ExternalForce="0,0,0"
   ) -> (Pose) {
    angularDampingOverride=0.99
    angularSpringConstant=300
    bAngularSpring=true
    bOverrideAngularDamping=true
    bUseGravityOverride=true
    simulationSpace=BoneRelative
    boundBone {boneName=dyn_skirt_fr_03_l}
    componentAppliedLinearAccClamp=<3000 3000 3000>
    componentLinearAccScale=<0.3 0.3 0.4>
    componentLinearVelScale=<0.3 0.3 0.6>
    gravityOverride {z=-980}
    physicsBodyDefinitions [{
      boundBone {boneName=dyn_skirt_fr_03_l}
      boxExtents=<10 10 10>
      constraintSetup {angularLimitsMax=<0 0 0>, angularLimitsMin=<0 0 -30>, angularTarget=<1 0 0>}
      localJointOffset=<0.5 0 0>
     }]
    relativeSpaceBone {boneName=root}
    simSpaceSettings {maxAngularAcceleration=20, maxAngularVelocity=4, simSpaceAngularAlpha=0.3}
   }
   n12 = Anim Dynamics - Bone: dyn_skirt_side_02_l!(
    ComponentPose=@n11
    Alpha=1
    GravityOverride:Vector="0,0,-980"
    SimSpaceSettings:Anim Phys Sim Space Settings Structure=(SimSpaceAngularAlpha=0.3,MaxAngularVelocity=4,MaxAngularAcceleration=20,ExternalAngularVelocity=(X=0,Y=0,Z=0))
    ExternalForce="0,0,0"
   ) -> (Pose) {
    angularDampingOverride=0.99
    angularSpringConstant=300
    bAngularSpring=true
    bOverrideAngularDamping=true
    bUseGravityOverride=true
    simulationSpace=BoneRelative
    boundBone {boneName=dyn_skirt_side_02_l}
    componentAppliedLinearAccClamp=<3000 3000 3000>
    componentLinearAccScale=<0.3 0.3 0.4>
    componentLinearVelScale=<0.3 0.3 0.6>
    gravityOverride {z=-980}
    physicsBodyDefinitions [{
      boundBone {boneName=dyn_skirt_side_02_l}
      boxExtents=<10 10 10>
      constraintSetup {angularLimitsMax=<0 0 0>, angularLimitsMin=<0 0 -8>, angularTarget=<1 0 0>}
      localJointOffset=<0.5 0 0>
     }]
    relativeSpaceBone {boneName=root}
    simSpaceSettings {maxAngularAcceleration=20, maxAngularVelocity=4, simSpaceAngularAlpha=0.3}
   }
   n13 = Anim Dynamics - Bone: dyn_skirt_side_03_l!(
    ComponentPose=@n12
    Alpha=1
    GravityOverride:Vector="0,0,-980"
    SimSpaceSettings:Anim Phys Sim Space Settings Structure=(SimSpaceAngularAlpha=0.3,MaxAngularVelocity=4,MaxAngularAcceleration=20,ExternalAngularVelocity=(X=0,Y=0,Z=0))
    ExternalForce="0,0,0"
   ) -> (Pose) {
    angularDampingOverride=0.99
    angularSpringConstant=300
    bAngularSpring=true
    bOverrideAngularDamping=true
    bUseGravityOverride=true
    simulationSpace=BoneRelative
    boundBone {boneName=dyn_skirt_side_03_l}
    componentAppliedLinearAccClamp=<3000 3000 3000>
    componentLinearAccScale=<0.3 0.3 0.4>
    componentLinearVelScale=<0.3 0.3 0.6>
    gravityOverride {z=-980}
    physicsBodyDefinitions [{
      boundBone {boneName=dyn_skirt_side_03_l}
      boxExtents=<10 10 10>
      constraintSetup {angularLimitsMax=<0 0 0>, angularLimitsMin=<0 0 -25>, angularTarget=<1 0 0>}
      localJointOffset=<0.5 0 0>
     }]
    relativeSpaceBone {boneName=root}
    simSpaceSettings {maxAngularAcceleration=20, maxAngularVelocity=4, simSpaceAngularAlpha=0.3}
   }
   n14 = Anim Dynamics - Bone: dyn_skirt_side_01_r!(
    ComponentPose=@n13
    Alpha=1
    GravityOverride:Vector="0,0,-980"
    SimSpaceSettings:Anim Phys Sim Space Settings Structure=(SimSpaceAngularAlpha=0.3,MaxAngularVelocity=4,MaxAngularAcceleration=20,ExternalAngularVelocity=(X=0,Y=0,Z=0))
    ExternalForce="0,0,0"
   ) -> (Pose) {
    angularDampingOverride=0.99
    angularSpringConstant=300
    bAngularSpring=true
    bOverrideAngularDamping=true
    bUseGravityOverride=true
    simulationSpace=BoneRelative
    boundBone {boneName=dyn_skirt_side_01_r}
    componentAppliedLinearAccClamp=<3000 3000 3000>
    componentLinearAccScale=<0.3 0.3 0.3>
    componentLinearVelScale=<0.3 0.3 0.5>
    gravityOverride {z=-980}
    physicsBodyDefinitions [{
      boundBone {boneName=dyn_skirt_side_01_r}
      boxExtents=<10 10 10>
      constraintSetup {angularLimitsMax=<0 2 0>, angularLimitsMin=<0 -2 -5>, angularTarget=<1 0 0>}
      localJointOffset=<0.5 0 0>
     }]
    relativeSpaceBone {boneName=root}
    simSpaceSettings {maxAngularAcceleration=20, maxAngularVelocity=4, simSpaceAngularAlpha=0.3}
   }
   n15 = Anim Dynamics - Bone: dyn_skirt_side_02_r!(
    ComponentPose=@n14
    Alpha=1
    GravityOverride:Vector="0,0,-980"
    SimSpaceSettings:Anim Phys Sim Space Settings Structure=(SimSpaceAngularAlpha=0.3,MaxAngularVelocity=4,MaxAngularAcceleration=20,ExternalAngularVelocity=(X=0,Y=0,Z=0))
    ExternalForce="0,0,0"
   ) -> (Pose) {
    angularDampingOverride=0.99
    angularSpringConstant=300
    bAngularSpring=true
    bOverrideAngularDamping=true
    bUseGravityOverride=true
    simulationSpace=BoneRelative
    boundBone {boneName=dyn_skirt_side_02_r}
    componentAppliedLinearAccClamp=<3000 3000 3000>
    componentLinearAccScale=<0.3 0.3 0.4>
    componentLinearVelScale=<0.3 0.3 0.6>
    gravityOverride {z=-980}
    physicsBodyDefinitions [{
      boundBone {boneName=dyn_skirt_side_02_r}
      boxExtents=<10 10 10>
      constraintSetup {angularLimitsMax=<0 4 0>, angularLimitsMin=<0 -4 -15>, angularTarget=<1 0 0>}
      localJointOffset=<0.5 0 0>
     }]
    relativeSpaceBone {boneName=root}
    simSpaceSettings {maxAngularAcceleration=20, maxAngularVelocity=4, simSpaceAngularAlpha=0.3}
   }
   n16 = Anim Dynamics - Bone: dyn_skirt_side_03_r!(
    ComponentPose=@n15
    Alpha=1
    GravityOverride:Vector="0,0,-980"
    SimSpaceSettings:Anim Phys Sim Space Settings Structure=(SimSpaceAngularAlpha=0.3,MaxAngularVelocity=4,MaxAngularAcceleration=20,ExternalAngularVelocity=(X=0,Y=0,Z=0))
    ExternalForce="0,0,0"
   ) -> (Pose) {
    angularDampingOverride=0.99
    angularSpringConstant=300
    bAngularSpring=true
    bOverrideAngularDamping=true
    bUseGravityOverride=true
    simulationSpace=BoneRelative
    boundBone {boneName=dyn_skirt_side_03_r}
    componentAppliedLinearAccClamp=<3000 3000 3000>
    componentLinearAccScale=<0.3 0.3 0.4>
    componentLinearVelScale=<0.3 0.3 0.6>
    gravityOverride {z=-980}
    physicsBodyDefinitions [{
      boundBone {boneName=dyn_skirt_side_03_r}
      boxExtents=<10 10 10>
      constraintSetup {angularLimitsMax=<0 5 0>, angularLimitsMin=<0 -5 -30>, angularTarget=<1 0 0>}
      localJointOffset=<0.5 0 0>
     }]
    relativeSpaceBone {boneName=root}
    simSpaceSettings {maxAngularAcceleration=20, maxAngularVelocity=4, simSpaceAngularAlpha=0.3}
   }
   n17 = Anim Dynamics - Bone: dyn_skirt_bk_01_l!(
    ComponentPose=@n16
    Alpha=1
    GravityOverride:Vector="0,0,-980"
    SimSpaceSettings:Anim Phys Sim Space Settings Structure=(SimSpaceAngularAlpha=0.3,MaxAngularVelocity=4,MaxAngularAcceleration=20,ExternalAngularVelocity=(X=0,Y=0,Z=0))
    ExternalForce="0,0,0"
   ) -> (Pose) {
    angularDampingOverride=0.99
    angularSpringConstant=300
    bAngularSpring=true
    bOverrideAngularDamping=true
    bUseGravityOverride=true
    simulationSpace=BoneRelative
    alphaScaleBiasClamp {bInterpResult=true}
    boundBone {boneName=dyn_skirt_bk_01_l}
    componentAppliedLinearAccClamp=<3000 3000 3000>
    componentLinearAccScale=<0.3 0.3 0.3>
    componentLinearVelScale=<0.3 0.3 0.5>
    gravityOverride {z=-980}
    physicsBodyDefinitions [{
      boundBone {boneName=dyn_skirt_bk_01_l}
      boxExtents=<10 10 10>
      constraintSetup {angularLimitsMax=<0 5 0>, angularLimitsMin=<0 -5 -10>, angularTarget=<1 0 0>}
      localJointOffset=<0.5 0 0>
     }]
    relativeSpaceBone {boneName=root}
    simSpaceSettings {maxAngularAcceleration=20, maxAngularVelocity=4, simSpaceAngularAlpha=0.3}
   }
   n18 = Anim Dynamics - Bone: dyn_skirt_bk_02_l!(
    ComponentPose=@n17
    Alpha=1
    GravityOverride:Vector="0,0,-980"
    SimSpaceSettings:Anim Phys Sim Space Settings Structure=(SimSpaceAngularAlpha=0.3,MaxAngularVelocity=4,MaxAngularAcceleration=20,ExternalAngularVelocity=(X=0,Y=0,Z=0))
    ExternalForce="0,0,0"
   ) -> (Pose) {
    angularDampingOverride=0.99
    angularSpringConstant=300
    bAngularSpring=true
    bOverrideAngularDamping=true
    bUseGravityOverride=true
    simulationSpace=BoneRelative
    alphaScaleBiasClamp {bInterpResult=true}
    boundBone {boneName=dyn_skirt_bk_02_l}
    componentAppliedLinearAccClamp=<3000 3000 3000>
    componentLinearAccScale=<0.3 0.3 0.5>
    componentLinearVelScale=<0.3 0.3 0.7>
    gravityOverride {z=-980}
    physicsBodyDefinitions [{
      boundBone {boneName=dyn_skirt_bk_02_l}
      boxExtents=<10 10 7>
      constraintSetup {angularLimitsMax=<0 3 0>, angularLimitsMin=<0 -3 -20>, angularTarget=<1 0 0>}
      localJointOffset=<0.5 0 0>
     }]
    relativeSpaceBone {boneName=root}
    simSpaceSettings {maxAngularAcceleration=20, maxAngularVelocity=4, simSpaceAngularAlpha=0.3}
   }
   n19 = Anim Dynamics - Bone: dyn_skirt_bk_03_l!(
    ComponentPose=@n18
    Alpha=1
    GravityOverride:Vector="0,0,-980"
    SimSpaceSettings:Anim Phys Sim Space Settings Structure=(SimSpaceAngularAlpha=0.3,MaxAngularVelocity=4,MaxAngularAcceleration=20,ExternalAngularVelocity=(X=0,Y=0,Z=0))
    ExternalForce="0,0,0"
   ) -> (Pose) {
    angularDampingOverride=0.99
    angularSpringConstant=300
    bAngularSpring=true
    bOverrideAngularDamping=true
    bUseGravityOverride=true
    simulationSpace=BoneRelative
    alphaScaleBiasClamp {bInterpResult=true}
    boundBone {boneName=dyn_skirt_bk_03_l}
    componentAppliedLinearAccClamp=<3000 3000 3000>
    componentLinearAccScale=<0.3 0.3 0.5>
    componentLinearVelScale=<0.3 0.3 0.7>
    gravityOverride {z=-980}
    physicsBodyDefinitions [{
      boundBone {boneName=dyn_skirt_bk_03_l}
      boxExtents=<10 10 7>
      constraintSetup {angularLimitsMax=<0 5 10>, angularLimitsMin=<0 -5 -30>, angularTarget=<1 0 0>}
      localJointOffset=<0.5 0 0>
     }]
    relativeSpaceBone {boneName=root}
    simSpaceSettings {maxAngularAcceleration=20, maxAngularVelocity=4, simSpaceAngularAlpha=0.3}
   }
   n20 = Anim Dynamics - Bone: dyn_skirt_bk_01_r!(
    ComponentPose=@n19
    Alpha=1
    GravityOverride:Vector="0,0,-980"
    SimSpaceSettings:Anim Phys Sim Space Settings Structure=(SimSpaceAngularAlpha=0.3,MaxAngularVelocity=4,MaxAngularAcceleration=20,ExternalAngularVelocity=(X=0,Y=0,Z=0))
    ExternalForce="0,0,0"
   ) -> (Pose) {
    angularDampingOverride=0.99
    angularSpringConstant=300
    bAngularSpring=true
    bOverrideAngularDamping=true
    bUseGravityOverride=true
    simulationSpace=BoneRelative
    alphaScaleBiasClamp {bInterpResult=true}
    boundBone {boneName=dyn_skirt_bk_01_r}
    componentAppliedLinearAccClamp=<3000 3000 3000>
    componentLinearAccScale=<0.3 0.3 0.3>
    componentLinearVelScale=<0.3 0.3 0.5>
    gravityOverride {z=-980}
    physicsBodyDefinitions [{
      boundBone {boneName=dyn_skirt_bk_01_r}
      boxExtents=<10 10 10>
      constraintSetup {angularLimitsMax=<0 5 0>, angularLimitsMin=<0 -5 -10>, angularTarget=<1 0 0>}
      localJointOffset=<0.5 0 0>
     }]
    relativeSpaceBone {boneName=root}
    simSpaceSettings {maxAngularAcceleration=20, maxAngularVelocity=4, simSpaceAngularAlpha=0.3}
   }
   n21 = Anim Dynamics - Bone: dyn_skirt_bk_02_r!(
    ComponentPose=@n20
    Alpha=1
    GravityOverride:Vector="0,0,-980"
    SimSpaceSettings:Anim Phys Sim Space Settings Structure=(SimSpaceAngularAlpha=0.3,MaxAngularVelocity=4,MaxAngularAcceleration=20,ExternalAngularVelocity=(X=0,Y=0,Z=0))
    ExternalForce="0,0,0"
   ) -> (Pose) {
    angularDampingOverride=0.99
    angularSpringConstant=300
    bAngularSpring=true
    bOverrideAngularDamping=true
    bUseGravityOverride=true
    simulationSpace=BoneRelative
    alphaScaleBiasClamp {bInterpResult=true}
    boundBone {boneName=dyn_skirt_bk_02_r}
    componentAppliedLinearAccClamp=<3000 3000 3000>
    componentLinearAccScale=<0.3 0.3 0.5>
    componentLinearVelScale=<0.3 0.3 0.7>
    gravityOverride {z=-980}
    physicsBodyDefinitions [{
      boundBone {boneName=dyn_skirt_bk_02_r}
      boxExtents=<10 10 7>
      constraintSetup {angularLimitsMax=<0 3 0>, angularLimitsMin=<0 -3 -20>, angularTarget=<1 0 0>}
      localJointOffset=<0.5 0 0>
     }]
    relativeSpaceBone {boneName=root}
    simSpaceSettings {maxAngularAcceleration=20, maxAngularVelocity=4, simSpaceAngularAlpha=0.3}
   }
   n22 = Anim Dynamics - Bone: dyn_skirt_bk_03_r!(
    ComponentPose=@n21
    Alpha=1
    GravityOverride:Vector="0,0,-980"
    SimSpaceSettings:Anim Phys Sim Space Settings Structure=(SimSpaceAngularAlpha=0.3,MaxAngularVelocity=4,MaxAngularAcceleration=20,ExternalAngularVelocity=(X=0,Y=0,Z=0))
    ExternalForce="0,0,0"
   ) -> (Pose) {
    angularDampingOverride=0.99
    angularSpringConstant=300
    bAngularSpring=true
    bOverrideAngularDamping=true
    bUseGravityOverride=true
    simulationSpace=BoneRelative
    alphaScaleBiasClamp {bInterpResult=true}
    boundBone {boneName=dyn_skirt_bk_03_r}
    componentAppliedLinearAccClamp=<3000 3000 3000>
    componentLinearAccScale=<0.3 0.3 0.5>
    componentLinearVelScale=<0.3 0.3 0.7>
    gravityOverride {z=-980}
    physicsBodyDefinitions [{
      boundBone {boneName=dyn_skirt_bk_03_r}
      boxExtents=<10 10 7>
      constraintSetup {angularLimitsMax=<0 5 10>, angularLimitsMin=<0 -5 -30>, angularTarget=<1 0 0>}
      localJointOffset=<0.5 0 0>
     }]
    relativeSpaceBone {boneName=root}
    simSpaceSettings {maxAngularAcceleration=20, maxAngularVelocity=4, simSpaceAngularAlpha=0.3}
   }
   n23 = Anim Dynamics - Bone: dyn_skirt_bk_01!(
    ComponentPose=@n22
    Alpha=1
    GravityOverride:Vector="0,0,-980"
    SimSpaceSettings:Anim Phys Sim Space Settings Structure=(SimSpaceAngularAlpha=0.3,MaxAngularVelocity=4,MaxAngularAcceleration=20,ExternalAngularVelocity=(X=0,Y=0,Z=0))
    ExternalForce="0,0,0"
   ) -> (Pose) {
    angularDampingOverride=0.99
    angularSpringConstant=300
    bAngularSpring=true
    bOverrideAngularDamping=true
    bUseGravityOverride=true
    simulationSpace=BoneRelative
    alphaScaleBiasClamp {bInterpResult=true}
    boundBone {boneName=dyn_skirt_bk_01}
    componentAppliedLinearAccClamp=<3000 3000 3000>
    componentLinearAccScale=<0.3 0.3 0.3>
    componentLinearVelScale=<0.3 0.3 0.5>
    gravityOverride {z=-980}
    physicsBodyDefinitions [{
      boundBone {boneName=dyn_skirt_bk_01}
      boxExtents=<10 10 10>
      constraintSetup {angularLimitsMax=<0 3 0>, angularLimitsMin=<0 -3 -10>, angularTarget=<1 0 0>}
      localJointOffset=<0.5 0 0>
     }]
    relativeSpaceBone {boneName=root}
    simSpaceSettings {maxAngularAcceleration=20, maxAngularVelocity=4, simSpaceAngularAlpha=0.3}
   }
   n24 = Anim Dynamics - Bone: dyn_skirt_bk_02!(
    ComponentPose=@n23
    Alpha=1
    GravityOverride:Vector="0,0,-980"
    SimSpaceSettings:Anim Phys Sim Space Settings Structure=(SimSpaceAngularAlpha=0.3,MaxAngularVelocity=4,MaxAngularAcceleration=20,ExternalAngularVelocity=(X=0,Y=0,Z=0))
    ExternalForce="0,0,0"
   ) -> (Pose) {
    angularDampingOverride=0.99
    angularSpringConstant=300
    bAngularSpring=true
    bOverrideAngularDamping=true
    bUseGravityOverride=true
    simulationSpace=BoneRelative
    alphaScaleBiasClamp {bInterpResult=true}
    boundBone {boneName=dyn_skirt_bk_02}
    componentAppliedLinearAccClamp=<3000 3000 3000>
    componentLinearAccScale=<0.3 0.3 0.5>
    componentLinearVelScale=<0.3 0.3 0.7>
    gravityOverride {z=-980}
    physicsBodyDefinitions [{
      boundBone {boneName=dyn_skirt_bk_02}
      boxExtents=<10 10 7>
      constraintSetup {angularLimitsMax=<0 2 0>, angularLimitsMin=<0 -2 -20>, angularTarget=<1 0 0>}
      localJointOffset=<0.5 0 0>
     }]
    relativeSpaceBone {boneName=root}
    simSpaceSettings {maxAngularAcceleration=20, maxAngularVelocity=4, simSpaceAngularAlpha=0.3}
   }
   n25 = Anim Dynamics - Bone: dyn_skirt_bk_03!(
    ComponentPose=@n24
    Alpha=1
    GravityOverride:Vector="0,0,-980"
    SimSpaceSettings:Anim Phys Sim Space Settings Structure=(SimSpaceAngularAlpha=0.3,MaxAngularVelocity=4,MaxAngularAcceleration=20,ExternalAngularVelocity=(X=0,Y=0,Z=0))
    ExternalForce="0,0,0"
   ) -> (Pose) {
    angularDampingOverride=0.99
    angularSpringConstant=300
    bAngularSpring=true
    bOverrideAngularDamping=true
    bUseGravityOverride=true
    simulationSpace=BoneRelative
    alphaScaleBiasClamp {bInterpResult=true}
    boundBone {boneName=dyn_skirt_bk_03}
    componentAppliedLinearAccClamp=<3000 3000 3000>
    componentLinearAccScale=<0.3 0.3 0.5>
    componentLinearVelScale=<0.3 0.3 0.7>
    gravityOverride {z=-980}
    physicsBodyDefinitions [{
      boundBone {boneName=dyn_skirt_bk_03}
      boxExtents=<10 10 7>
      constraintSetup {angularLimitsMax=<0 5 10>, angularLimitsMin=<0 -5 -30>, angularTarget=<1 0 0>}
      localJointOffset=<0.5 0 0>
     }]
    relativeSpaceBone {boneName=root}
    simSpaceSettings {maxAngularAcceleration=20, maxAngularVelocity=4, simSpaceAngularAlpha=0.3}
   }
   n26 = Anim Dynamics - Bone: dyn_bag_01!(
    ComponentPose=@n25
    Alpha=1
    GravityOverride:Vector="0,0,-980"
    SimSpaceSettings:Anim Phys Sim Space Settings Structure=(SimSpaceAngularAlpha=0.3,MaxAngularVelocity=4,MaxAngularAcceleration=20,ExternalAngularVelocity=(X=0,Y=0,Z=0))
    ExternalForce="0,0,0"
   ) -> (Pose) {
    angularDampingOverride=0.99
    angularSpringConstant=300
    bAngularSpring=true
    bOverrideAngularDamping=true
    bUseGravityOverride=true
    simulationSpace=BoneRelative
    boundBone {boneName=dyn_bag_01}
    componentAppliedLinearAccClamp=<2000 2000 8000>
    componentLinearAccScale=<0.3 0.3 0.5>
    componentLinearVelScale=<0.3 0.3 1>
    gravityOverride {z=-980}
    physicsBodyDefinitions [{
      boundBone {boneName=dyn_bag_01}
      boxExtents=<10 10 5>
      constraintSetup {angularLimitsMax=<0 2 5>, angularLimitsMin=<0 -25 -5>, angularTarget=<1 0 0>}
      localJointOffset=<0.3 0 0>
     }]
    relativeSpaceBone {boneName=root}
    simSpaceSettings {maxAngularAcceleration=20, maxAngularVelocity=4, simSpaceAngularAlpha=0.3}
   }
   n27 = Anim Dynamics - Bone: dyn_bag_02!(
    ComponentPose=@n26
    Alpha=1
    GravityOverride:Vector="0,0,-980"
    SimSpaceSettings:Anim Phys Sim Space Settings Structure=(SimSpaceAngularAlpha=0.3,MaxAngularVelocity=4,MaxAngularAcceleration=20,ExternalAngularVelocity=(X=0,Y=0,Z=0))
    ExternalForce="0,0,0"
   ) -> (Pose) {
    angularDampingOverride=0.99
    angularSpringConstant=300
    bAngularSpring=true
    bOverrideAngularDamping=true
    bUseGravityOverride=true
    simulationSpace=BoneRelative
    boundBone {boneName=dyn_bag_02}
    componentAppliedLinearAccClamp=<2000 2000 8000>
    componentLinearAccScale=<0.3 0.3 0.5>
    componentLinearVelScale=<0.3 0.3 1>
    gravityOverride {z=-980}
    physicsBodyDefinitions [{
      boundBone {boneName=dyn_bag_02}
      boxExtents=<10 10 5>
      constraintSetup {angularLimitsMax=<2 3 5>, angularLimitsMin=<-2 -30 -5>, angularTarget=<1 0 0>}
      localJointOffset=<0.3 0 0>
     }]
    relativeSpaceBone {boneName=root}
    simSpaceSettings {maxAngularAcceleration=20, maxAngularVelocity=4, simSpaceAngularAlpha=0.3}
   }
   n28 = Anim Dynamics - Bone: dart_01!(
    ComponentPose=@n27
    Alpha=1
    GravityOverride:Vector="0,0,-980"
    SimSpaceSettings:Anim Phys Sim Space Settings Structure=(SimSpaceAngularAlpha=0.3,MaxAngularVelocity=4,MaxAngularAcceleration=20,ExternalAngularVelocity=(X=0,Y=0,Z=0))
    ExternalForce="0,0,0"
   ) -> (Pose) {
    angularDampingOverride=0.99
    angularSpringConstant=300
    bAngularSpring=true
    bLinearSpring=true
    bOverrideAngularDamping=true
    bOverrideLinearDamping=true
    bUseGravityOverride=true
    linearDampingOverride=0.999
    linearSpringConstant=600
    simulationSpace=RootRelative
    boundBone {boneName=dart_01}
    componentAppliedLinearAccClamp=<2000 2000 8000>
    componentLinearAccScale=<0.3 0.3 0.5>
    componentLinearVelScale=<0.3 0.3 1>
    gravityOverride {z=-980}
    physicsBodyDefinitions [{
      boundBone {boneName=dart_01}
      boxExtents=<6 7 7>
      constraintSetup {
       angularTargetAxis=AxisY
       twistAxis=AxisY
       angularLimitsMax=<0 2 5>
       angularLimitsMin=<0 -2 -5>
       angularTarget=<1 0 0>
       linearAxesMax=<0 0.1 0.2>
       linearAxesMin=<0 -0.4 -0.2>
      }
      localJointOffset=<0 0 0.3>
     }]
    relativeSpaceBone {boneName=pelvis}
    simSpaceSettings {maxAngularAcceleration=20, maxAngularVelocity=4, simSpaceAngularAlpha=0.3}
   }
   n29 = Anim Dynamics - Bone: dart_02!(
    ComponentPose=@n28
    Alpha=1
    GravityOverride:Vector="0,0,-980"
    SimSpaceSettings:Anim Phys Sim Space Settings Structure=(SimSpaceAngularAlpha=0.3,MaxAngularVelocity=4,MaxAngularAcceleration=20,ExternalAngularVelocity=(X=0,Y=0,Z=0))
    ExternalForce="0,0,0"
   ) -> (Pose) {
    angularDampingOverride=0.99
    angularSpringConstant=300
    bAngularSpring=true
    bLinearSpring=true
    bOverrideAngularDamping=true
    bOverrideLinearDamping=true
    bUseGravityOverride=true
    linearDampingOverride=0.999
    linearSpringConstant=600
    simulationSpace=RootRelative
    boundBone {boneName=dart_02}
    componentAppliedLinearAccClamp=<2000 2000 8000>
    componentLinearAccScale=<0.3 0.3 0.5>
    componentLinearVelScale=<0.3 0.3 1>
    gravityOverride {z=-980}
    physicsBodyDefinitions [{
      boundBone {boneName=dart_02}
      boxExtents=<7 7 7>
      constraintSetup {
       angularTargetAxis=AxisY
       twistAxis=AxisY
       angularLimitsMax=<0 2 3>
       angularLimitsMin=<0 -2 -5>
       angularTarget=<1 0 0>
       linearAxesMax=<0 0.2 0>
       linearAxesMin=<0 -0.4 0>
      }
      localJointOffset=<0 0 0.2>
     }]
    relativeSpaceBone {boneName=pelvis}
    simSpaceSettings {maxAngularAcceleration=20, maxAngularVelocity=4, simSpaceAngularAlpha=0.3}
   }
   n30 = Anim Dynamics - Bone: dart_03!(
    ComponentPose=@n29
    Alpha=1
    GravityOverride:Vector="0,0,-980"
    SimSpaceSettings:Anim Phys Sim Space Settings Structure=(SimSpaceAngularAlpha=0.3,MaxAngularVelocity=4,MaxAngularAcceleration=20,ExternalAngularVelocity=(X=0,Y=0,Z=0))
    ExternalForce="0,0,0"
   ) -> (Pose) {
    angularDampingOverride=0.99
    angularSpringConstant=300
    bAngularSpring=true
    bLinearSpring=true
    bOverrideAngularDamping=true
    bOverrideLinearDamping=true
    bUseGravityOverride=true
    linearDampingOverride=0.999
    linearSpringConstant=600
    simulationSpace=RootRelative
    boundBone {boneName=dart_03}
    componentAppliedLinearAccClamp=<2000 2000 8000>
    componentLinearAccScale=<0.3 0.3 0.5>
    componentLinearVelScale=<0.3 0.3 1>
    gravityOverride {z=-980}
    physicsBodyDefinitions [{
      boundBone {boneName=dart_03}
      boxExtents=<5 7 7>
      constraintSetup {
       angularTargetAxis=AxisY
       twistAxis=AxisY
       angularLimitsMax=<0 2 5>
       angularLimitsMin=<0 -2 -5>
       angularTarget=<1 0 0>
       linearAxesMax=<0 0.1 0>
       linearAxesMin=<0 -0.3 0>
      }
      localJointOffset=<0 0 0.4>
     }]
    relativeSpaceBone {boneName=pelvis}
    simSpaceSettings {maxAngularAcceleration=20, maxAngularVelocity=4, simSpaceAngularAlpha=0.3}
   }
   Output Pose(Result=Component To Local(ComponentPose=@n30)) {layerGroup=None}
  ]}
 EventGraph {nodes [
   n1 = Event BlueprintUpdateAnimation()
   > n2 = "Gravity Override State Function (Parameter Asset) (Nativized)"(
    CharacterPartAnimInstance=Self-Reference()
    GameTimeInSeconds=GetGameTimeInSeconds()
    Parameters=&/F_MED_PearlTaffyVista_Parameters
   )
   n3 = "To String (Name)"(InName=@n2.OutCurrentState)
   n4 = float * float(
    A="To Float (Boolean)"(InBool=Contains(SearchIn=@n3, Substring=OnGround_))
    B=MapRangeClamped(Value=Get PawnVelocityZ(), InRangeA=0, InRangeB=-300, OutRangeB=1680)
   ) {is=title K2Node_PromotableOperator}
   > Set GravityOverride(GravityOverride=vector + vector(A=@n2, B=MakeVector(Y=0, Z=@n4)))
   n5 = Get PawnVelocityZ()
   > Set!(
    execute
    RotateWhileJump_Roll:float=0
    RotateWhileJump_Pitch:float=0
    RotateWhileJump_Yaw:float=FInterpTo(
     Current=Get!() -> (RotateWhileJump_Roll:float, RotateWhileJump_Pitch:float, RotateWhileJump_Yaw:float).RotateWhileJump_Yaw
     Target=MapRangeClamped(Value=@n5, InRangeA=0, InRangeB=-600, OutRangeB=25)
     DeltaTime=@n1.DeltaTimeX
     InterpSpeed=30
    )
   ) -> (then, Output_Get:Rotator)
   n6 = Get bIsPlayingEmote()
   > Set isFrontEndIdle(isFrontEndIdle:bool=AND Boolean(A=NOT Boolean(A=@n6), B=Get bIsInFrontEnd())) -> (Output_Get:bool)
   Event OnBeginSkydiving()
   > Set isSkydiving(isSkydiving:double=1) -> (Output_Get:double)
   Event OnEndSkydiving()
   > Set isSkydiving(isSkydiving=0)
   Event OnBeginSliding() {is=title K2Node_Event}
   > Set isSlidingDynamicsOff(isSlidingDynamicsOff:double=0) -> (Output_Get:double)
   Event OnEndSliding() {is=title K2Node_Event}
   > Set isSlidingDynamicsOff(isSlidingDynamicsOff=1)
   n7 = Event OnCharacterCustomizationLoaded()
   Set AnimDynamicsSimSpaceSetting(
    execute=[@n7.then, @n8.then]
    AnimDynamicsSimSpaceSetting:Anim Phys Sim Space Settings Structure=Select(
     Option 0:Anim Phys Sim Space Settings Structure=Make Anim Phys Sim Space Settings!(SimSpaceAngularAlpha:float=0.3, MaxAngularVelocity:float=4, MaxAngularAcceleration:float=20, ExternalAngularVelocity:Vector="0,0,0") -> (AnimPhysSimSpaceSettings:Anim Phys Sim Space Settings Structure) {is=title K2Node_MakeStruct}
     Option 1:Anim Phys Sim Space Settings Structure=Make Anim Phys Sim Space Settings(SimSpaceAngularAlpha=0.8, MaxAngularVelocity=7, MaxAngularAcceleration=35, ExternalAngularVelocity="0,0,0")
     Index=Get bIsInFrontEnd()
    ) -> (ReturnValue:Anim Phys Sim Space Settings Structure)
   ) -> (Output_Get:Anim Phys Sim Space Settings Structure)
   n8 = Event BlueprintInitializeAnimation()
  ]}
]
variables [
 AnimDynamicsSimSpaceSetting {
  type=AnimPhysSimSpaceSettings
  default {maxAngularAcceleration=10000, maxAngularVelocity=10000, simSpaceAngularAlpha=0, externalAngularVelocity=<0 0 0>}
 }
 GravityOverride {type=Vector}
 isFrontEndIdle {type=boolean}
 isSkydiving {type=number}
 isSlidingDynamicsOff {type=number}
 RotateWhileJump {type=Rotator}
]
