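# Physics asset for the right-side curl face accessory on the PearlTaffyVista medium female character cosmetic, defining bodies for secondary curl motion.
# Three tapering capsules form a dynamic chain hanging off a small sphere on the MSTR_curl_r bone.
# Constraints link each segment to its parent with limited swing and locked twist.
# Larger unconstrained capsules on the upper spine, right clavicle and right upper arm approximate the body as collision shapes.
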PhysicsAsset /BRCosmetics/Characters/Player/Female/Medium/Bodies/F_MED_PearlTaffyVista/Meshes/Parts/F_MED_PearlTaffyVista_FaceAcc_Curl_Physics
bodies {
 MSTR_curl_r [{shapeType=Sphere, radius=0.505, shapeName=MSTR_curl_r_sphere}]
 clavicle_r [{
   shapeType=Capsule
   length=3.2717
   radius=7.3893
   shapeName=clavicle_r_capsule
   center=<-5.7416 -0.2785 3.088>
   rotation=<rot -85.9538 -2.1765 -177.8243>
  }]
 dyn_curl_01_r [{
   shapeType=Capsule
   length=2.9394
   radius=1.0611
   shapeName=dyn_curl_01_r_capsule
   center=<1.9753 -0.083 -0.0198>
   rotation=<rot 0 -90.1319 -86.8997>
  }]
 dyn_curl_02_r [{
   shapeType=Capsule
   length=3.1753
   radius=0.8127
   shapeName=dyn_curl_02_r_capsule
   center=<1.3497 0.018 -0.0303>
   rotation=<rot 0 -91.312 -86.1557>
  }]
 dyn_curl_03_r [{
   shapeType=Capsule
   length=2.6329
   radius=0.6548
   shapeName=dyn_curl_03_r_capsule
   center=<1.4924 0.0477 0.0865>
   rotation=<rot 0 -92.3308 -83.0187>
  }]
 spine_04 [{shapeType=Capsule, length=6.5363, radius=9.4737, shapeName=spine_05_capsule, center=<-0.169 1.935 0>}]
 spine_05 [{shapeType=Capsule, length=8.9763, radius=8.592, shapeName=spine_05_capsule, center=<0.9332 -0.1751 0>}]
 upperarm_r [{
   shapeType=Capsule
   length=5.831
   radius=6.6679
   shapeName=upperarm_r_capsule
   center=<-5.1242 0.0049 0.5622>
   rotation=<rot -88.9538 105.5797 75.1322>
  }]
}
constraints [
 {bone1Name=dyn_curl_01_r, bone2Name=MSTR_curl_r, swing1LimitDegrees=20, swing2LimitDegrees=20, twistLimitDegrees=45, twistMotion=Locked}
 {bone1Name=dyn_curl_02_r, bone2Name=dyn_curl_01_r, swing1LimitDegrees=30, swing2LimitDegrees=30, twistLimitDegrees=45, twistMotion=Locked}
 {bone1Name=dyn_curl_03_r, bone2Name=dyn_curl_02_r, swing1LimitDegrees=15, swing2LimitDegrees=30, twistLimitDegrees=45, twistMotion=Locked}
]
