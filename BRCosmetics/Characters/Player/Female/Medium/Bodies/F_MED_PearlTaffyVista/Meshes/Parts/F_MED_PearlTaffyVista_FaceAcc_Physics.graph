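# Physics asset for the back hair in the FaceAcc part of the PearlTaffyVista female medium character cosmetic.
# Center, left and right three-bone capsule chains hang from small MSTR sphere anchors, with narrow swing limits and locked twist.
# Angularly free constraints tie both side tips to the center tip, linking the three chains.
# Upper spine, clavicle and upper-arm capsules provide body shapes for the hair to collide with.
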PhysicsAsset /BRCosmetics/Characters/Player/Female/Medium/Bodies/F_MED_PearlTaffyVista/Meshes/Parts/F_MED_PearlTaffyVista_FaceAcc_Physics
bodies {
 MSTR_hair_bk [{shapeType=Sphere, radius=0.5, shapeName=MSTR_hair_bk_sphere}]
 MSTR_hair_bk_l [{shapeType=Sphere, radius=0.5, shapeName=MSTR_hair_bk_l_sphere}]
 MSTR_hair_bk_r [{shapeType=Sphere, radius=0.5, shapeName=MSTR_hair_bk_r_sphere}]
 clavicle_l [{
   shapeType=Capsule
   length=3.2717
   radius=7.3893
   shapeName=clavicle_l_capsule
   center=<5.7416 0.2785 -3.088>
   rotation=<rot -85.9538 -2.1765 2.1757>
  }]
 clavicle_r [{
   shapeType=Capsule
   length=3.2717
   radius=7.3893
   shapeName=clavicle_r_capsule
   center=<-5.7416 -0.2785 3.088>
   rotation=<rot -85.9538 -2.1765 -177.8243>
  }]
 dyn_hair_bk_01 [{
   shapeType=Capsule
   length=12.1666
   radius=3.6646
   shapeName=dyn_hair_bk_01_r_capsule
   center=<8.878 -0.6706 -0.066>
   rotation=<rot 90 -9.9997 0>
  }]
 dyn_hair_bk_01_l [{
   shapeType=Capsule
   length=9.1666
   radius=3.6646
   shapeName=dyn_hair_bk_01_l_capsule
   center=<5.4854 -0.6302 0.11>
   rotation=<rot 88.8516 179.9751 -174.2819>
  }]
 dyn_hair_bk_01_r [{
   shapeType=Capsule
   length=8.5294
   radius=3.6646
   shapeName=dyn_hair_bk_01_r_capsule
   center=<6.755 -0.6797 0.0035>
   rotation=<rot 88.8516 179.9772 3.5065>
  }]
 dyn_hair_bk_02 [{
   shapeType=Capsule
   length=4.4464
   radius=3.6646
   shapeName=dyn_hair_bk_01_r_capsule
   center=<4.7253 -2.342 -0.178>
   rotation=<rot -0.6428 -177.8268 -0.0441>
  }]
 dyn_hair_bk_02_l [{
   shapeType=Capsule
   length=3.242
   radius=3.6646
   shapeName=dyn_hair_bk_01_l_capsule
   center=<4.787 -2.2173 -0.9034>
   rotation=<rot -3.5129 -177.8242 175.9408>
  }]
 dyn_hair_bk_02_r [{
   shapeType=Capsule
   length=2.1637
   radius=3.6726
   shapeName=dyn_hair_bk_01_r_capsule
   center=<4.7538 -2.1768 0.2483>
   rotation=<rot 4.6553 -177.7942 -5.8436>
  }]
 dyn_hair_bk_03 [{
   shapeType=Capsule
   length=3.4271
   radius=4.0679
   shapeName=dyn_hair_bk_01_r_capsule
   center=<1.2875 -1.2581 -0.0639>
   rotation=<rot 2.3612 -0.6881 179.6232>
  }]
 dyn_hair_bk_03_l [{
   shapeType=Capsule
   length=2.2681
   radius=4.0679
   shapeName=dyn_hair_bk_01_l_capsule
   center=<0.0041 -0.8741 0.1221>
   rotation=<rot 20.0644 -0.5592 -0.3956>
  }]
 dyn_hair_bk_03_r [{
   shapeType=Capsule
   length=2.3714
   radius=4.0679
   shapeName=dyn_hair_bk_01_r_capsule
   center=<-0.0281 -0.9538 -0.116>
   rotation=<rot -24.3536 0.572 179.6102>
  }]
 spine_04 [{shapeType=Capsule, length=6.2978, radius=9.7122, shapeName=spine_05_capsule, center=<1.4315 0.2708 0>}]
 spine_05 [{shapeType=Capsule, length=8.914, radius=8.6544, shapeName=spine_05_capsule, center=<0.9332 -0.0555 0>}]
 upperarm_l [{
   shapeType=Capsule
   length=5.831
   radius=6.6679
   shapeName=upperarm_l_capsule
   center=<5.1242 -0.0049 -0.5622>
   rotation=<rot -88.9538 105.5797 -104.8678>
  }]
 upperarm_r [{
   shapeType=Capsule
   length=5.831
   radius=6.6679
   shapeName=upperarm_r_capsule
   center=<-5.1242 0.0049 0.5622>
   rotation=<rot -88.9538 105.5797 75.1322>
  }]
}
constraints [
 {
  bone1Name=dyn_hair_bk_03_l
  bone2Name=dyn_hair_bk_02_l
  swing1LimitDegrees=30
  swing2LimitDegrees=3
  twistLimitDegrees=45
  twistMotion=Locked
 }
 {
  bone1Name=dyn_hair_bk_02_l
  bone2Name=dyn_hair_bk_01_l
  swing1LimitDegrees=30
  swing2LimitDegrees=5
  twistLimitDegrees=45
  twistMotion=Locked
 }
 {bone1Name=dyn_hair_bk_01_l, bone2Name=MSTR_hair_bk_l, swing1LimitDegrees=15, swing2LimitDegrees=5, twistLimitDegrees=45, twistMotion=Locked}
 {bone1Name=dyn_hair_bk_01_r, bone2Name=MSTR_hair_bk_r, swing1LimitDegrees=10, swing2LimitDegrees=5, twistLimitDegrees=45, twistMotion=Locked}
 {
  bone1Name=dyn_hair_bk_02_r
  bone2Name=dyn_hair_bk_01_r
  swing1LimitDegrees=30
  swing2LimitDegrees=5
  twistLimitDegrees=45
  twistMotion=Locked
 }
 {
  bone1Name=dyn_hair_bk_03_r
  bone2Name=dyn_hair_bk_02_r
  swing1LimitDegrees=30
  swing2LimitDegrees=3
  twistLimitDegrees=45
  twistMotion=Locked
 }
 {bone1Name=dyn_hair_bk_02, bone2Name=dyn_hair_bk_01, swing1LimitDegrees=30, swing2LimitDegrees=5, twistLimitDegrees=45, twistMotion=Locked}
 {bone1Name=dyn_hair_bk_03, bone2Name=dyn_hair_bk_02, swing1LimitDegrees=30, swing2LimitDegrees=10, twistLimitDegrees=45, twistMotion=Locked}
 {bone1Name=dyn_hair_bk_01, bone2Name=MSTR_hair_bk, swing1LimitDegrees=10, swing2LimitDegrees=7, twistLimitDegrees=45, twistMotion=Locked}
 {
  bone1Name=dyn_hair_bk_03_r
  bone2Name=dyn_hair_bk_03
  swing1LimitDegrees=45
  swing1Motion=Free
  swing2LimitDegrees=45
  swing2Motion=Free
  twistLimitDegrees=45
  twistMotion=Free
 }
 {
  bone1Name=dyn_hair_bk_03_l
  bone2Name=dyn_hair_bk_03
  swing1LimitDegrees=45
  swing1Motion=Free
  swing2LimitDegrees=45
  swing2Motion=Free
  twistLimitDegrees=45
  twistMotion=Free
 }
]
