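# Post-process style AnimBlueprint for the PearlTaffyVista female face accessory part, adding secondary motion to hair.
# The attached parent's pose passes through a Control Rig and front-hair AnimDynamics chain into a cached pose.
# That cache feeds front-end or gameplay RigidBody pairs, selected by bIsInFrontEnd; back-hair AnimDynamics nodes never reach the output.
# The EventGraph computes state-driven gravity, jump rotation, backbling/rigid-body flags and sim-space settings.
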AnimBlueprint /BRCosmetics/Characters/Player/Female/Medium/Bodies/F_MED_PearlTaffyVista/Meshes/Parts/F_MED_PearlTaffyVista_FaceAcc_AnimBP
target_skeleton=&/F_MED_PearlTaffyVista_FaceAcc_Skeleton
graphs [
 AnimGraph {nodes [
   n1 = "Transform (Modify) Bone - Bone: dyn_hair_bk_02_l"(
    ComponentPose="Transform (Modify) Bone - Bone: MSTR_hair"(Alpha=1, Translation="0,0,0", Rotation="0,0,0", Scale="1,1,1") {rotationMode=BMM_Additive, rotationSpace=BCS_BoneSpace, boneToModify {boneName=MSTR_hair}}
    Alpha=1
    Translation="0,0,0"
    Rotation="0,0,0"
    Scale="1,1,1"
   ) {rotationMode=BMM_Additive, rotationSpace=BCS_BoneSpace, boneToModify {boneName=dyn_hair_bk_02_l}}
   n2 = Anim Dynamics - Bone: dyn_hair_bk_01_l!(
    ComponentPose="Transform (Modify) Bone - Bone: dyn_hair_bk_02_r"(
     ComponentPose="Transform (Modify) Bone - Bone: dyn_hair_bk_02"(ComponentPose=@n1, Alpha=1, Translation="0,0,0", Rotation="0,0,0", Scale="1,1,1") {rotationMode=BMM_Additive, rotationSpace=BCS_BoneSpace, boneToModify {boneName=dyn_hair_bk_02}}
     Alpha=1
     Translation="0,0,0"
     Rotation="0,0,0"
     Scale="1,1,1"
    ) {rotationMode=BMM_Additive, rotationSpace=BCS_BoneSpace, boneToModify {boneName=dyn_hair_bk_02_r}}
    Alpha=1
    GravityOverride:Vector="0,0,-980"
    ExternalForce="0,0,0"
   ) -> (Pose) {
    angularDampingOverride=0.99
    angularSpringConstant=300
    bAngularSpring=true
    bOverrideAngularDamping=true
    bUseGravityOverride=true
    simulationSpace=RootRelative
    boundBone {boneName=dyn_hair_bk_01_l}
    componentAppliedLinearAccClamp=<3000 3000 8000>
    componentLinearAccScale=<0.3 0.3 0.3>
    componentLinearVelScale=<0.3 0.3 0.5>
    gravityOverride {z=-980}
    physicsBodyDefinitions [{
      boundBone {boneName=dyn_hair_bk_01_l}
      boxExtents=<10 10 10>
      constraintSetup {angularLimitsMax=<0 2 0>, angularLimitsMin=<0 -2 -15>, angularTarget=<1 0 0>}
      localJointOffset=<0.5 0 0>
     }]
    relativeSpaceBone {boneName=spine_05}
    simSpaceSettings {maxAngularAcceleration=40, maxAngularVelocity=8, simSpaceAngularAlpha=0.3}
   }
   n3 = Anim Dynamics - Bone: dyn_hair_bk_02_l!(ComponentPose=@n2, Alpha=1, GravityOverride:Vector="0,0,-980", ExternalForce="0,0,0") -> (Pose) {
    angularDampingOverride=0.99
    angularSpringConstant=300
    bAngularSpring=true
    bOverrideAngularDamping=true
    bUseGravityOverride=true
    simulationSpace=RootRelative
    boundBone {boneName=dyn_hair_bk_02_l}
    componentAppliedLinearAccClamp=<3000 3000 8000>
    componentLinearAccScale=<0.3 0.3 0.3>
    componentLinearVelScale=<0.3 0.3 0.5>
    gravityOverride {z=-980}
    physicsBodyDefinitions [{
      boundBone {boneName=dyn_hair_bk_02_l}
      boxExtents=<10 10 10>
      constraintSetup {angularLimitsMax=<3 10 2>, angularLimitsMin=<-3 -10 -45>, angularTarget=<1 0 0>}
      localJointOffset=<0.5 0 0>
     }]
    relativeSpaceBone {boneName=spine_05}
    simSpaceSettings {maxAngularAcceleration=40, maxAngularVelocity=8, simSpaceAngularAlpha=0.3}
   }
   n4 = Anim Dynamics - Bone: dyn_hair_bk_03_l!(ComponentPose=@n3, Alpha=1, GravityOverride:Vector="0,0,-980", ExternalForce="0,0,0") -> (Pose) {
    angularDampingOverride=0.99
    angularSpringConstant=300
    bAngularSpring=true
    bOverrideAngularDamping=true
    bUseGravityOverride=true
    simulationSpace=RootRelative
    boundBone {boneName=dyn_hair_bk_03_l}
    componentAppliedLinearAccClamp=<3000 3000 8000>
    componentLinearAccScale=<0.3 0.3 0.3>
    componentLinearVelScale=<0.3 0.3 0.5>
    gravityOverride {z=-980}
    physicsBodyDefinitions [{
      boundBone {boneName=dyn_hair_bk_03_l}
      boxExtents=<10 10 10>
      constraintSetup {angularLimitsMax=<3 15 0>, angularLimitsMin=<-3 -15 -60>, angularTarget=<1 0 0>}
      localJointOffset=<0.5 0 0>
     }]
    relativeSpaceBone {boneName=spine_05}
    simSpaceSettings {maxAngularAcceleration=40, maxAngularVelocity=8, simSpaceAngularAlpha=0.3}
   }
   n5 = Anim Dynamics - Bone: dyn_hair_bk_01_r!(ComponentPose=@n4, Alpha=1, GravityOverride:Vector="0,0,-980", ExternalForce="0,0,0") -> (Pose) {
    angularDampingOverride=0.99
    angularSpringConstant=300
    bAngularSpring=true
    bOverrideAngularDamping=true
    bUseGravityOverride=true
    simulationSpace=RootRelative
    boundBone {boneName=dyn_hair_bk_01_r}
    componentAppliedLinearAccClamp=<3000 3000 8000>
    componentLinearAccScale=<0.3 0.3 0.3>
    componentLinearVelScale=<0.3 0.3 0.5>
    gravityOverride {z=-980}
    physicsBodyDefinitions [{
      boundBone {boneName=dyn_hair_bk_01_r}
      boxExtents=<10 10 10>
      constraintSetup {angularLimitsMax=<0 2 0>, angularLimitsMin=<0 -2 -15>, angularTarget=<1 0 0>}
      localJointOffset=<0.5 0 0>
     }]
    relativeSpaceBone {boneName=spine_05}
    simSpaceSettings {maxAngularAcceleration=40, maxAngularVelocity=8, simSpaceAngularAlpha=0.3}
   }
   n6 = Anim Dynamics - Bone: dyn_hair_bk_02_r!(ComponentPose=@n5, Alpha=1, GravityOverride:Vector="0,0,-980", ExternalForce="0,0,0") -> (Pose) {
    angularDampingOverride=0.99
    angularSpringConstant=300
    bAngularSpring=true
    bOverrideAngularDamping=true
    bUseGravityOverride=true
    simulationSpace=RootRelative
    boundBone {boneName=dyn_hair_bk_02_r}
    componentAppliedLinearAccClamp=<3000 3000 8000>
    componentLinearAccScale=<0.3 0.3 0.3>
    componentLinearVelScale=<0.3 0.3 0.5>
    gravityOverride {z=-980}
    physicsBodyDefinitions [{
      boundBone {boneName=dyn_hair_bk_02_r}
      boxExtents=<10 10 10>
      constraintSetup {angularLimitsMax=<3 10 2>, angularLimitsMin=<-3 -10 -45>, angularTarget=<1 0 0>}
      localJointOffset=<0.5 0 0>
     }]
    relativeSpaceBone {boneName=spine_05}
    simSpaceSettings {maxAngularAcceleration=40, maxAngularVelocity=8, simSpaceAngularAlpha=0.3}
   }
   n7 = Anim Dynamics - Bone: dyn_hair_bk_03_r!(ComponentPose=@n6, Alpha=1, GravityOverride:Vector="0,0,-980", ExternalForce="0,0,0") -> (Pose) {
    angularDampingOverride=0.99
    angularSpringConstant=300
    bAngularSpring=true
    bOverrideAngularDamping=true
    bUseGravityOverride=true
    simulationSpace=RootRelative
    boundBone {boneName=dyn_hair_bk_03_r}
    componentAppliedLinearAccClamp=<3000 3000 8000>
    componentLinearAccScale=<0.3 0.3 0.3>
    componentLinearVelScale=<0.3 0.3 0.5>
    gravityOverride {z=-980}
    physicsBodyDefinitions [{
      boundBone {boneName=dyn_hair_bk_03_r}
      boxExtents=<10 10 10>
      constraintSetup {angularLimitsMax=<3 15 0>, angularLimitsMin=<-3 -15 -60>, angularTarget=<1 0 0>}
      localJointOffset=<0.5 0 0>
     }]
    relativeSpaceBone {boneName=spine_05}
    simSpaceSettings {maxAngularAcceleration=40, maxAngularVelocity=8, simSpaceAngularAlpha=0.3}
   }
   n8 = Anim Dynamics - Bone: dyn_hair_bk_01!(ComponentPose=@n7, Alpha=1, GravityOverride:Vector="0,0,-980", ExternalForce="0,0,0") -> (Pose) {
    angularDampingOverride=0.99
    angularSpringConstant=300
    bAngularSpring=true
    bOverrideAngularDamping=true
    bUseGravityOverride=true
    simulationSpace=RootRelative
    boundBone {boneName=dyn_hair_bk_01}
    componentAppliedLinearAccClamp=<3000 3000 8000>
    componentLinearAccScale=<0.3 0.3 0.3>
    componentLinearVelScale=<0.3 0.3 0.5>
    gravityOverride {z=-980}
    physicsBodyDefinitions [{
      boundBone {boneName=dyn_hair_bk_01}
      boxExtents=<10 10 10>
      constraintSetup {angularLimitsMax=<0 2 0>, angularLimitsMin=<0 -2 -15>, angularTarget=<1 0 0>}
      localJointOffset=<0.5 0 0>
     }]
    relativeSpaceBone {boneName=spine_05}
    simSpaceSettings {maxAngularAcceleration=40, maxAngularVelocity=8, simSpaceAngularAlpha=0.3}
   }
   n9 = Anim Dynamics - Bone: dyn_hair_bk_02!(ComponentPose=@n8, Alpha=1, GravityOverride:Vector="0,0,-980", ExternalForce="0,0,0") -> (Pose) {
    angularDampingOverride=0.99
    angularSpringConstant=300
    bAngularSpring=true
    bOverrideAngularDamping=true
    bUseGravityOverride=true
    simulationSpace=RootRelative
    boundBone {boneName=dyn_hair_bk_02}
    componentAppliedLinearAccClamp=<3000 3000 8000>
    componentLinearAccScale=<0.3 0.3 0.3>
    componentLinearVelScale=<0.3 0.3 0.5>
    gravityOverride {z=-980}
    physicsBodyDefinitions [{
      boundBone {boneName=dyn_hair_bk_02}
      boxExtents=<10 10 10>
      constraintSetup {angularLimitsMax=<3 10 2>, angularLimitsMin=<-3 -10 -45>, angularTarget=<1 0 0>}
      localJointOffset=<0.5 0 0>
     }]
    relativeSpaceBone {boneName=spine_05}
    simSpaceSettings {maxAngularAcceleration=40, maxAngularVelocity=8, simSpaceAngularAlpha=0.3}
   }
   Anim Dynamics - Bone: dyn_hair_bk_03!(ComponentPose=@n9, Alpha=1, GravityOverride:Vector="0,0,-980", ExternalForce="0,0,0") -> (Pose) {
    angularDampingOverride=0.99
    angularSpringConstant=300
    bAngularSpring=true
    bOverrideAngularDamping=true
    bUseGravityOverride=true
    simulationSpace=RootRelative
    boundBone {boneName=dyn_hair_bk_03}
    componentAppliedLinearAccClamp=<3000 3000 8000>
    componentLinearAccScale=<0.3 0.3 0.3>
    componentLinearVelScale=<0.3 0.3 0.5>
    gravityOverride {z=-980}
    physicsBodyDefinitions [{
      boundBone {boneName=dyn_hair_bk_03}
      boxExtents=<10 10 10>
      constraintSetup {angularLimitsMax=<3 15 0>, angularLimitsMin=<-3 -15 -60>, angularTarget=<1 0 0>}
      localJointOffset=<0.5 0 0>
     }]
    relativeSpaceBone {boneName=spine_05}
    simSpaceSettings {maxAngularAcceleration=40, maxAngularVelocity=8, simSpaceAngularAlpha=0.3}
   }
   n10 = RigidBody!(ComponentPose, Alpha, OverrideWorldGravity:Vector="0,0,-980", ExternalForce) -> (Pose) {
    bClampLinearTranslationLimitToRefPose=true
    bOverrideWorldGravity=true
    overridePhysicsAsset=&/F_MED_PearlTaffyVista_FaceAcc_FrontEnd_Physics
    simulationSpace=BaseBoneSpace
    baseBoneRef {boneName=spine_05}
    simSpaceSettings {
     dampingAlpha=0
     maxAngularAcceleration=40
     maxAngularVelocity=8
     maxLinearAcceleration=450
     maxLinearVelocity=40
     velocityScaleZ=0.7
     worldAlpha=1
     externalLinearDragV {x=5}
    }
   }
   RigidBody(ComponentPose=@n10, OverrideWorldGravity="0,0,-980") {
    bClampLinearTranslationLimitToRefPose=true
    bOverrideWorldGravity=true
    overridePhysicsAsset=&/F_MED_PearlTaffyVista_FaceAcc_Curl_Physics
    simulationSpace=BaseBoneSpace
    baseBoneRef {boneName=spine_05}
    simSpaceSettings {
     dampingAlpha=0
     maxAngularAcceleration=40
     maxAngularVelocity=8
     maxLinearAcceleration=450
     maxLinearVelocity=40
     velocityScaleZ=0.7
     worldAlpha=1
     externalLinearDragV {x=5}
    }
   }
   n11 = RigidBody(ComponentPose=Local To Component(LocalPose=Use cached pose 'DeformAndDynamics'()), Alpha=0.2, OverrideWorldGravity="0,0,-980") {
    bClampLinearTranslationLimitToRefPose=true
    bOverrideWorldGravity=true
    overridePhysicsAsset=&F_MED_PearlTaffyVista_FaceAcc_FrontEnd_Physics
    simulationSpace=WorldSpace
    baseBoneRef {boneName=spine_05}
    componentAppliedLinearAccClamp=<3000 3000 3000>
    simSpaceSettings {
     dampingAlpha=0.3
     maxAngularAcceleration=30
     maxAngularVelocity=6
     maxLinearAcceleration=450
     maxLinearVelocity=40
     velocityScaleZ=0.7
     worldAlpha=0.5
    }
   }
   n12 = RigidBody(ComponentPose=@n11, OverrideWorldGravity="0,0,-980") {
    bClampLinearTranslationLimitToRefPose=true
    bOverrideWorldGravity=true
    overridePhysicsAsset=&/F_MED_PearlTaffyVista_FaceAcc_Curl_Physics_FE
    simulationSpace=WorldSpace
    baseBoneRef {boneName=root}
    componentAppliedLinearAccClamp=<1000 1000 1000>
    simSpaceSettings {maxAngularAcceleration=3, maxAngularVelocity=1, maxLinearAcceleration=30, maxLinearVelocity=1, velocityScaleZ=0.7, worldAlpha=0.2}
   }
   n13 = RigidBody!(
    ComponentPose=Local To Component(LocalPose=Use cached pose 'DeformAndDynamics'())
    bAlphaBoolEnabled:bool
    OverrideWorldGravity="0,0,-980"
    ExternalForce
   ) -> (Pose) {
    alphaInputType=Bool
    bClampLinearTranslationLimitToRefPose=true
    bOverrideWorldGravity=true
    overridePhysicsAsset=&/F_MED_PearlTaffyVista_FaceAcc_Physics
    simulationSpace=BaseBoneSpace
    baseBoneRef {boneName=spine_05}
    simSpaceSettings {
     dampingAlpha=0
     maxAngularAcceleration=20
     maxAngularVelocity=4
     maxLinearAcceleration=450
     maxLinearVelocity=40
     velocityScaleZ=0.7
     worldAlpha=1
    }
   }
   n14 = RigidBody!(ComponentPose=@n13, Alpha:float, OverrideWorldGravity="0,0,-980", ExternalForce) -> (Pose) {
    bClampLinearTranslationLimitToRefPose=true
    bOverrideWorldGravity=true
    overridePhysicsAsset=&F_MED_PearlTaffyVista_FaceAcc_Curl_Physics
    simulationSpace=BaseBoneSpace
    baseBoneRef {boneName=spine_05}
    simSpaceSettings {
     dampingAlpha=0
     maxAngularAcceleration=20
     maxAngularVelocity=4
     maxLinearAcceleration=450
     maxLinearVelocity=40
     velocityScaleZ=0.7
     worldAlpha=1
    }
   }
   Output Pose(Result=Blend Poses by bool(BlendPose_0=Component To Local(ComponentPose=@n12), BlendPose_1=Component To Local(ComponentPose=@n14), bActiveValue=Get bIsInFrontEnd())) {layerGroup=None}
   n15 = Control Rig!(Source=Copy Pose From Mesh() {bUseAttachedParent=true}, Alpha, isBackPackOrRBAN:bool=Get isBackpackOrRBANoff() -> (isBackpackOrRBANoff:bool)) -> (Pose) {controlRigAssetReference {blueprintRigClass=&/F_MED_PearlTaffyVista_FaceAcc_CtrlRig.F_MED_PearlTaffyVista_FaceAcc_CtrlRig_C}}
   n16 = "Transform (Modify) Bone - Bone: dyn_hair_fr_03"(ComponentPose=Local To Component(LocalPose=@n15), Alpha=1, Translation="0,-1,1", Rotation="0.000,-0.001,20.000", Scale="1,1,1") {
    rotationMode=BMM_Additive
    rotationSpace=BCS_BoneSpace
    translationMode=BMM_Additive
    translationSpace=BCS_BoneSpace
    boneToModify {boneName=dyn_hair_fr_03}
    rotation=<rot 0 -0.001 20>
    translation {y=-1, z=1}
   }
   n17 = Anim Dynamics - Bone: dyn_hair_fr_01!(
    ComponentPose=@n16
    Alpha=1
    GravityOverride:Vector="0,0,-980"
    SimSpaceSettings:Anim Phys Sim Space Settings Structure=(SimSpaceAngularAlpha=0.3,MaxAngularVelocity=30,MaxAngularAcceleration=120,ExternalAngularVelocity=(X=0,Y=0,Z=0))
    ExternalForce="0,0,0"
   ) -> (Pose) {
    angularDampingOverride=0.99
    angularSpringConstant=300
    bAngularSpring=true
    bOverrideAngularDamping=true
    bUseGravityOverride=true
    simulationSpace=RootRelative
    boundBone {boneName=dyn_hair_fr_01}
    componentAppliedLinearAccClamp=<3000 3000 15000>
    componentLinearAccScale=<0.3 0.3 0.5>
    componentLinearVelScale=<0.3 0.3 1>
    gravityOverride {z=-980}
    physicsBodyDefinitions [{
      boundBone {boneName=dyn_hair_fr_01}
      boxExtents=<12 12 12>
      constraintSetup {angularLimitsMax=<0 0 0>, angularLimitsMin=<0 0 -15>, angularTarget=<1 0 0>}
      localJointOffset=<0.5 0 0>
     }]
    relativeSpaceBone {boneName=spine_05}
    simSpaceSettings {maxAngularAcceleration=120, maxAngularVelocity=30, simSpaceAngularAlpha=0.3}
   }
   n18 = Anim Dynamics - Bone: dyn_hair_fr_02!(
    ComponentPose=@n17
    Alpha=1
    GravityOverride:Vector="0,0,-980"
    SimSpaceSettings:Anim Phys Sim Space Settings Structure=(SimSpaceAngularAlpha=0.3,MaxAngularVelocity=30,MaxAngularAcceleration=120,ExternalAngularVelocity=(X=0,Y=0,Z=0))
    ExternalForce="0,0,0"
   ) -> (Pose) {
    angularDampingOverride=0.99
    angularSpringConstant=300
    bAngularSpring=true
    bOverrideAngularDamping=true
    bUseGravityOverride=true
    simulationSpace=RootRelative
    boundBone {boneName=dyn_hair_fr_02}
    componentAppliedLinearAccClamp=<3000 3000 15000>
    componentLinearAccScale=<0.3 0.3 0.5>
    componentLinearVelScale=<0.3 0.3 1>
    gravityOverride {z=-980}
    physicsBodyDefinitions [{
      boundBone {boneName=dyn_hair_fr_02}
      boxExtents=<12 12 12>
      constraintSetup {angularLimitsMax=<0 0 2>, angularLimitsMin=<0 0 -20>, angularTarget=<1 0 0>}
      localJointOffset=<0.5 0 0>
     }]
    relativeSpaceBone {boneName=spine_05}
    simSpaceSettings {maxAngularAcceleration=120, maxAngularVelocity=30, simSpaceAngularAlpha=0.3}
   }
   n19 = Anim Dynamics - Bone: dyn_hair_fr_03!(
    ComponentPose=@n18
    Alpha=1
    GravityOverride:Vector="0,0,-980"
    SimSpaceSettings:Anim Phys Sim Space Settings Structure=(SimSpaceAngularAlpha=0.3,MaxAngularVelocity=30,MaxAngularAcceleration=120,ExternalAngularVelocity=(X=0,Y=0,Z=0))
    ExternalForce="0,0,0"
   ) -> (Pose) {
    angularDampingOverride=0.99
    angularSpringConstant=300
    bAngularSpring=true
    bOverrideAngularDamping=true
    bUseGravityOverride=true
    simulationSpace=RootRelative
    boundBone {boneName=dyn_hair_fr_03}
    componentAppliedLinearAccClamp=<3000 3000 15000>
    componentLinearAccScale=<0.3 0.3 0.5>
    componentLinearVelScale=<0.3 0.3 1>
    gravityOverride {z=-980}
    physicsBodyDefinitions [{
      boundBone {boneName=dyn_hair_fr_03}
      boxExtents=<12 12 12>
      constraintSetup {angularLimitsMax=<0 2 0>, angularLimitsMin=<0 -2 -25>, angularTarget=<1 0 0>}
      localJointOffset=<0.5 0 0>
     }]
    relativeSpaceBone {boneName=spine_05}
    simSpaceSettings {maxAngularAcceleration=120, maxAngularVelocity=30, simSpaceAngularAlpha=0.3}
   }
   n20 = Anim Dynamics - Bone: dyn_hair_fr_04!(
    ComponentPose=@n19
    Alpha=1
    GravityOverride:Vector="0,0,-980"
    SimSpaceSettings:Anim Phys Sim Space Settings Structure=(SimSpaceAngularAlpha=0.3,MaxAngularVelocity=30,MaxAngularAcceleration=120,ExternalAngularVelocity=(X=0,Y=0,Z=0))
    ExternalForce="0,0,0"
   ) -> (Pose) {
    angularDampingOverride=0.99
    angularSpringConstant=300
    bAngularSpring=true
    bOverrideAngularDamping=true
    bUseGravityOverride=true
    simulationSpace=RootRelative
    boundBone {boneName=dyn_hair_fr_04}
    componentAppliedLinearAccClamp=<3000 3000 15000>
    componentLinearAccScale=<0.3 0.3 0.5>
    componentLinearVelScale=<0.3 0.3 1>
    gravityOverride {z=-980}
    physicsBodyDefinitions [{
      boundBone {boneName=dyn_hair_fr_04}
      boxExtents=<12 12 12>
      constraintSetup {angularLimitsMax=<0 3 0>, angularLimitsMin=<0 -3 -30>, angularTarget=<1 0 0>}
      localJointOffset=<0.5 0 0>
     }]
    relativeSpaceBone {boneName=spine_05}
    simSpaceSettings {maxAngularAcceleration=120, maxAngularVelocity=30, simSpaceAngularAlpha=0.3}
   }
   Save cached pose 'DeformAndDynamics'(Pose=Component To Local(ComponentPose=@n20)) {cacheName=DeformAndDynamics}
  ]}
 EventGraph {nodes [
   n1 = Event BlueprintUpdateAnimation()
   n2 = Get PawnVelocityZ()
   > Set!(
    execute
    RotateWhileJump_Roll:float=0
    RotateWhileJump_Pitch:float=0
    RotateWhileJump_Yaw:float=FInterpTo(Current=0, Target=MapRangeClamped(Value=@n2, InRangeA=0, InRangeB=-600, OutRangeB=-10), DeltaTime=@n1.DeltaTimeX, InterpSpeed=30)
   ) -> (then, Output_Get:Rotator)
   > n3 = "Gravity Override State Function (Parameter Asset) (Nativized)"(
    CharacterPartAnimInstance=Self-Reference()
    GameTimeInSeconds=GetGameTimeInSeconds()
    Parameters=/BRCosmetics/Characters/Player/Female/Medium/Bodies/F_MED_PearlTaffyVista/Meshes/F_MED_PearlTaffyVista_Parameters
   )
   n4 = "To String (Name)"(InName=@n3.OutCurrentState)
   n5 = float * float(
    A="To Float (Boolean)"(InBool=Contains(SearchIn=@n4, Substring=OnGround_))
    B=MapRangeClamped(Value=Get PawnVelocityZ(), InRangeA=0, InRangeB=-300, OutRangeB=1280)
   ) {is=title K2Node_PromotableOperator}
   > Set Gravity Override(Gravity Override=vector + vector(A=@n3, B=MakeVector(Y=0, Z=@n5)))
   n6 = Get bRigidBodyEnabled()
   n7 = Get Is Not Wearing Backbling() -> (Is Not Wearing Backbling:bool)
   > Set isBackpackOrRBANoff(isBackpackOrRBANoff:bool=OR Boolean(A=NOT Boolean(A=@n6), B=NOT Boolean(A=@n7))) -> (Output_Get:bool)
   n8 = Event OnCharacterCustomizationLoaded()
   n9 = Self-Reference()
   Set Is Not Wearing Backbling(execute=[@n8.then, @n10.then], Is Not Wearing Backbling:bool=NOT Boolean(A=CheckForBackbling(self=@n9))) -> (Output_Get:bool)
   > Set AnimDynamicsSimSpaceSettings(AnimDynamicsSimSpaceSettings:Anim Phys Sim Space Settings Structure=Select(
     Option 0:Anim Phys Sim Space Settings Structure=Make Anim Phys Sim Space Settings!(SimSpaceAngularAlpha:float=0.3, MaxAngularVelocity:float=8, MaxAngularAcceleration:float=40, ExternalAngularVelocity:Vector="0,0,0") -> (AnimPhysSimSpaceSettings:Anim Phys Sim Space Settings Structure) {is=title K2Node_MakeStruct}
     Option 1:Anim Phys Sim Space Settings Structure=Make Anim Phys Sim Space Settings(SimSpaceAngularAlpha=0.5, MaxAngularVelocity=15, MaxAngularAcceleration=90, ExternalAngularVelocity="0,0,0")
     Index=Get bIsInFrontEnd()
    ) -> (ReturnValue:Anim Phys Sim Space Settings Structure)) -> (Output_Get:Anim Phys Sim Space Settings Structure)
   n10 = Event BlueprintInitializeAnimation()
   Set SimSpaceSetting(SimSpaceSetting:Sim Space Settings Structure=Select(
     Option 0:Sim Space Settings Structure=Make Sim Space Settings!(
      WorldAlpha:float=1
      VelocityScaleZ:float=0.7
      DampingAlpha:float=0
      MaxLinearVelocity:float=40
      MaxAngularVelocity:float=4
      MaxLinearAcceleration:float=450
      MaxAngularAcceleration:float=20
      ExternalLinearDragV:Vector="0,0,0"
      ExternalLinearVelocity:Vector="0,0,0"
      ExternalAngularVelocity:Vector="0,0,0"
     ) -> (SimSpaceSettings:Sim Space Settings Structure) {is=title K2Node_MakeStruct}
     Option 1:Sim Space Settings Structure=Make Sim Space Settings(
      WorldAlpha=1
      VelocityScaleZ=0.7
      DampingAlpha=0
      MaxLinearVelocity=40
      MaxAngularVelocity=8
      MaxLinearAcceleration=450
      MaxAngularAcceleration=30
      ExternalLinearDragV="0,0,0"
      ExternalLinearVelocity="0,0,0"
      ExternalAngularVelocity="0,0,0"
     )
     Index=Get bIsInFrontEnd()
    ) -> (ReturnValue:Sim Space Settings Structure)) -> (Output_Get:Sim Space Settings Structure)
  ]}
]
variables [
 AnimDynamicsSimSpaceSettings {
  type=AnimPhysSimSpaceSettings
  default {maxAngularAcceleration=10000, maxAngularVelocity=10000, simSpaceAngularAlpha=0, externalAngularVelocity=<0 0 0>}
 }
 Gravity Override {type=Vector}
 GravityOverrideAnimDynamics {type=Vector}
 Is Not Wearing Backbling {type=boolean}
 isBackpackOrRBANoff {type=boolean}
 RotateWhileJump {type=Rotator}
 SimSpaceSetting {
  type=SimSpaceSettings
  default {
   dampingAlpha=1
   maxAngularAcceleration=10000
   maxAngularVelocity=10000
   maxLinearAcceleration=10000
   maxLinearVelocity=10000
   velocityScaleZ=1
   worldAlpha=0
   externalLinearDragV=<0 0 0>
  }
 }
]
